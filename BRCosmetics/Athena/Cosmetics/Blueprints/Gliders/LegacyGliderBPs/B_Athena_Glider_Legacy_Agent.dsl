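; Legacy Agent glider Blueprint that drives its trail particle effect, including side engines, from parachute movement.
; Each trail update flags sideways movement beyond a threshold as an extra VFX activation condition, then calls the parent.
; Once the glider anim is fully deployed, it sends Forward, Right and Left values as float and vector particle parameters.
; Forward and Right are dot products offset by one, while Left reduces to negated RightDot.
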(event EventOnChuteVisibilityChanged
  (params (bParachuteVisibility :type Boolean))
  (|CallParentFunction bParachuteVisibility))

(event Parachute|EventOnParachuteTrailUpdated
  (params (MovementLengthSquared :type Float%20%28single-precision%29) (ForwardDot :type Float%20%28single-precision%29) (RightDot :type Float%20%28single-precision%29) (RotationalMovementAcceleration :type Float%20%28single-precision%29))
  (bind _returnvalue (CallFunction|GetTrailVFXComponent))
  (bind _returnvalue_2 (- 1.0 (+ RightDot 1.0)))
  (bind _returnvalue_3 (+ ForwardDot 1.0))
  (bind _returnvalue_4 (+ RightDot 1.0))
  (Variables|LegacyVFX|SetAdditional%20Condition%20for%20Movement%20Activation (or (>= RightDot 0.2) (<= RightDot -0.2)))
  (|CallParentFunction MovementLengthSquared ForwardDot RightDot RotationalMovementAcceleration)
  (if (CallFunction|IsGliderFullyDeployedAccordingtoAnim)
    ;[1 1 1] Moving forward is 1, moving backward is 0
    (Effects|Components|ParticleSystem|SetFloatParameter _returnvalue "ForwardS" _returnvalue_3)
    (Effects|Components|ParticleSystem|SetVectorParameter _returnvalue "ForwardV" (Math|Conversions|ToVector(Float) _returnvalue_3))
    ;[1 1 1] Moving right or forward, this value becomes 1.  Turn on left side engines in particle system.
    (Effects|Components|ParticleSystem|SetFloatParameter _returnvalue "RightS" _returnvalue_4)
    (Effects|Components|ParticleSystem|SetVectorParameter _returnvalue "RightV" (Math|Conversions|ToVector(Float) _returnvalue_4))
    ;[1 1 1] Moving left or forward, this value becomes 1.  Turn on right side engines in particle system
    (Effects|Components|ParticleSystem|SetFloatParameter _returnvalue "LeftS" _returnvalue_2)
    (Effects|Components|ParticleSystem|SetVectorParameter _returnvalue "LeftV" (Math|Conversions|ToVector(Float) _returnvalue_2))))

(fn ConstructionScript
  (|CallParentFunction))
