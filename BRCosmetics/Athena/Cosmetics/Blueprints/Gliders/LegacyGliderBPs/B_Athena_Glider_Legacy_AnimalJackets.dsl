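; Legacy AnimalJackets glider Blueprint that drives vent particle trails as parachute movement updates.
; On each trail update, after the parent logic, it ends trails and deactivates the Vents component unless forward motion passes a threshold and the glider animation is fully deployed.
; A DoOnce begins socket-to-socket trails on Vents and Vents2 using left and right FX sockets.
; The stop path touches only Vents, and the DoOnce has no visible reset.
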(event Parachute|EventOnParachuteTrailUpdated
  (params (MovementLengthSquared :type Float%20%28single-precision%29) (ForwardDot :type Float%20%28single-precision%29) (RightDot :type Float%20%28single-precision%29) (RotationalMovementAcceleration :type Float%20%28single-precision%29))
  (bind _vents (Variables|Default|GetVents))
  (|CallParentFunction MovementLengthSquared ForwardDot RightDot RotationalMovementAcceleration)
  (if (not (and (>= ForwardDot 0.2) (CallFunction|IsGliderFullyDeployedAccordingtoAnim)))
    (Effects|Particles|Trails|EndTrails _vents)
    (Components|Activation|Deactivate _vents))
  (Utilities|FlowControl|DoOnce false
    (:Completed
      (Effects|Particles|Trails|BeginTrails (Variables|Default|GetVents) "FX_L3" "FX_L4" "ETrailWidthMode_FromSecond" 1.0)
      (Effects|Particles|Trails|BeginTrails (Variables|Default|GetVents2) "FX_R3" "FX_R4" "ETrailWidthMode_FromSecond" 1.0))))

(fn ConstructionScript
  (|CallParentFunction))
