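; Legacy glider Blueprint for the Ashton Boardwalk cosmetic, adding a cloak-in effect and jet thruster particles.
; When the chute shows, a timeline drives fade and refraction on a dynamic cloak material.
; Full deployment restores the standard material, starts airflow, and boosts a jet after a retriggerable delay.
; Trail updates lower flare size or opacity when forward motion or deployment is lacking, and use DoOnce to start the right jet.
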(event EventOnChuteVisibilityChanged
  (params (bParachuteVisibility :type Boolean))
  (bind _parachutemesh (Variables|Parachute|GetParachuteMesh))
  (bind _mid (Variables|Default|GetMID))
  (|CallParentFunction bParachuteVisibility)
  (if bParachuteVisibility
    (Rendering|Material|SetMaterial _parachutemesh 0 "/BRCosmetics/Gadgets/Assets/VinderTech_GliderChute/Glider_Ashton_Boardwalk/FX/MI_Glider_Ashton_Boardwalk_Cloak.MI_Glider_Ashton_Boardwalk_Cloak")
    (bind _returnvalue (Rendering|Material|CreateDynamicMaterialInstance _parachutemesh 0 "/BRCosmetics/Gadgets/Assets/VinderTech_GliderChute/Glider_Ashton_Boardwalk/FX/MI_Glider_Ashton_Boardwalk_Cloak.MI_Glider_Ashton_Boardwalk_Cloak" "None" :!class /Script/Engine.PrimitiveComponent))
    (Variables|Default|SetMID _returnvalue)
    (bind (_direction _cloakin _refraction _cloakin_2) (Timeline|Timeline_0 PlayFromStart)
      (:Update
        (Rendering|Material|SetScalarParameterValue _mid "FadeControl" (Math|Float|Lerp 0.01 25.0 _cloakin_2) :!class /Script/Engine.MaterialInstanceDynamic)
        (Rendering|Material|SetScalarParameterValue _mid "Refraction" _refraction :!class /Script/Engine.MaterialInstanceDynamic)))))

(event EventOnGliderFullyDeployed
  (bind _p_glider_ashton_boardwalk_jet (Variables|Default|GetP_Glider_Ashton_Boardwalk_Jet))
  (|CallParentFunction)
  (Rendering|Material|SetMaterial (Variables|Parachute|GetParachuteMesh) 0 "/BRCosmetics/Gadgets/Assets/VinderTech_GliderChute/Glider_Ashton_Boardwalk/Materials/MI_Ashton_Boardwalk.MI_Ashton_Boardwalk")
  (Components|Activation|Activate (Variables|Default|GetP_Glider_Ashton_Boardwalk_AirFlow_R) true)
  (Effects|Components|ParticleSystem|SetVectorParameter _p_glider_ashton_boardwalk_jet "ThrusterSize" "1.000000,1.000000,1.000000")
  (Effects|Components|ParticleSystem|SetVectorParameter _p_glider_ashton_boardwalk_jet "ThrusterOffset" "-25.000000, 0, 0")
  (Effects|Components|ParticleSystem|SetVectorParameter _p_glider_ashton_boardwalk_jet "FlareSize" "0.700000,0.700000,1.000000")
  (Effects|Components|ParticleSystem|SetFloatParameter _p_glider_ashton_boardwalk_jet "FlareOpacity" 0.1)
  (Effects|Components|ParticleSystem|SetFloatParameter (Variables|Default|GetP_Glider_Ashton_Boardwalk_Jet_R) "PopSpawn" 0.0)
  (Utilities|FlowControl|RetriggerableDelay 0.9)
  (Components|Activation|Activate _p_glider_ashton_boardwalk_jet true)
  (Effects|Components|ParticleSystem|SetVectorParameter _p_glider_ashton_boardwalk_jet "ThrusterSize" "1.000000,1.000000,1.300000")
  (Effects|Components|ParticleSystem|SetVectorParameter _p_glider_ashton_boardwalk_jet "ThrusterOffset" "-15.000000, 0, 0")
  (Effects|Components|ParticleSystem|SetVectorParameter _p_glider_ashton_boardwalk_jet "FlareSize" "1.000000,1.000000,1.000000")
  (Effects|Components|ParticleSystem|SetFloatParameter _p_glider_ashton_boardwalk_jet "FlareOpacity" 0.2)
  (Effects|Components|ParticleSystem|SetFloatParameter _p_glider_ashton_boardwalk_jet "PopSpawn" 1.0))

(event Parachute|EventOnParachuteTrailUpdated
  (params (MovementLengthSquared :type Float%20%28single-precision%29) (ForwardDot :type Float%20%28single-precision%29) (RightDot :type Float%20%28single-precision%29) (RotationalMovementAcceleration :type Float%20%28single-precision%29))
  (bind _p_glider_ashton_boardwalk_jet_r (Variables|Default|GetP_Glider_Ashton_Boardwalk_Jet_R))
  (|CallParentFunction MovementLengthSquared ForwardDot RightDot RotationalMovementAcceleration)
  (if (not (and (>= ForwardDot 0.2) (CallFunction|IsGliderFullyDeployedAccordingtoAnim)))
    (Effects|Components|ParticleSystem|SetVectorParameter (Variables|Default|GetP_Glider_Ashton_Boardwalk_Jet) "FlareSize" "0.700000,0.700000,1.000000")
    (Effects|Components|ParticleSystem|SetFloatParameter _p_glider_ashton_boardwalk_jet_r "FlareOpacity" 0.1))
  (Utilities|FlowControl|DoOnce false
    (:Completed
      (Components|Activation|Activate _p_glider_ashton_boardwalk_jet_r true)
      (Effects|Components|ParticleSystem|SetVectorParameter _p_glider_ashton_boardwalk_jet_r "ThrusterSize" "1.000000,1.000000,1.300000")
      (Effects|Components|ParticleSystem|SetVectorParameter _p_glider_ashton_boardwalk_jet_r "ThrusterOffset" "-15.000000, 0, 0")
      (Effects|Components|ParticleSystem|SetVectorParameter _p_glider_ashton_boardwalk_jet_r "FlareSize" "1.000000,1.000000,1.000000")
      (Effects|Components|ParticleSystem|SetFloatParameter _p_glider_ashton_boardwalk_jet_r "FlareOpacity" 0.2))))

(fn ConstructionScript
  (|CallParentFunction))
