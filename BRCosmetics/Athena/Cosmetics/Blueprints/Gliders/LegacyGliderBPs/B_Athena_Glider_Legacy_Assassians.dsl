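; Legacy Assassin glider Blueprint that animates the tail lights on the glider's bits mesh while parachuting.
; On receiving its item definition, it creates a dynamic tail-light material on the first bits mesh.
; Trail updates map right movement or actor pitch to turn and direction parameters, and backward movement to a braking parameter.
; A DoOnce-gated timeline drives the light sequence parameter, and level flight stops and rewinds it.
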(event Parachute|EventOnReceiveParachuteItemdef
  (|CallParentFunction)
  (bind _returnvalue (Rendering|Material|CreateDynamicMaterialInstance (Utilities|Array|Get(acopy) (CallFunction|GetBitsMeshComponents)) 0 "/BRCosmetics/Gadgets/Assets/VinderTech_GliderChute/Assassin/Materials/M_VinderTech_GliderChute_Assassin_TailLights_Inst.M_VinderTech_GliderChute_Assassin_TailLights_Inst" "AssassianBits_MID" :!class /Script/Engine.PrimitiveComponent))
  (Variables|Default|SetBits_MID _returnvalue))

(event Parachute|EventOnParachuteTrailUpdated
  (params (MovementLengthSquared :type Float%20%28single-precision%29) (ForwardDot :type Float%20%28single-precision%29) (RightDot :type Float%20%28single-precision%29) (RotationalMovementAcceleration :type Float%20%28single-precision%29))
  (bind _returnvalue (Transformation|GetActorRotation))
  (bind _pitch (.pitch _returnvalue))
  (bind _bits_mid (Variables|Default|GetBits_MID))
  (|CallParentFunction MovementLengthSquared ForwardDot RightDot RotationalMovementAcceleration)
  (Utilities|FlowControl|Sequence
    (:then_0
      (if (> RightDot 0.5)
        (Rendering|Material|SetScalarParameterValue (Variables|Default|GetBits_MID) "TailLightDirection" 1.0 :!class /Script/Engine.MaterialInstanceDynamic :!label _label_2)
        (elif (< RightDot -0.5)
          (Rendering|Material|SetScalarParameterValue (Variables|Default|GetBits_MID) "TailLightDirection" 0.0 :!class /Script/Engine.MaterialInstanceDynamic :!label _label)
          (elif (> _pitch 5.0)
            (goto _label)
            (elif (< _pitch -5.0)
              (goto _label_2)
              (else
                (Rendering|Material|SetScalarParameterValue (Variables|Default|GetBits_MID) "isTurning" 0.0 :!class /Script/Engine.MaterialInstanceDynamic)
                (Utilities|FlowControl|Sequence
                  (:then_0
                    (Timeline|TailLightSequence Stop))
                  (:then_1
                    (Timeline|TailLightSequence SetNewTime))
                  (:then_2
                    (Utilities|FlowControl|DoOnce false :!label _label_3
                      (:Completed
                        (bind (_direction _newtrack_0) (Timeline|TailLightSequence Play)
                          (:Update
                            (Rendering|Material|SetScalarParameterValue (Variables|Default|GetBits_MID) "TailLightSequence" _newtrack_0 :!class /Script/Engine.MaterialInstanceDynamic))))))))))))
      (Rendering|Material|SetScalarParameterValue (Variables|Default|GetBits_MID) "isTurning" 1.0 :!class /Script/Engine.MaterialInstanceDynamic)
      (goto _label_3))
    (:then_1
      (if (> MovementLengthSquared 0.1)
        (Utilities|FlowControl|Sequence
          (:then_0
            (if (< ForwardDot -0.25)
              (Rendering|Material|SetScalarParameterValue _bits_mid "isBraking" 1.0 :!class /Script/Engine.MaterialInstanceDynamic)
              (else
                _))))
        (else
          (Utilities|FlowControl|Sequence
            (:then_0
              _))))
      (Rendering|Material|SetScalarParameterValue _bits_mid "isBraking" 0.0 :!class /Script/Engine.MaterialInstanceDynamic))))

(fn ConstructionScript
  (|CallParentFunction))
