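; Legacy Bandolier glider Blueprint that manages rotor and trail effects on the parachute mesh.
; When the chute becomes visible, it attaches the blade spin effect and two trail effects to the main rotor, tail rotor and engine exhaust sockets.
; The blade spin effect activates once the glider fully deploys and deactivates when the chute is hidden.
; Both event overrides call their parent logic first.
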(event EventOnChuteVisibilityChanged
  (params (bParachuteVisibility :type Boolean))
  (|CallParentFunction bParachuteVisibility)
  (if bParachuteVisibility
    (Transformation|AttachComponentToComponent (Variables|Default|GetBladeSpinFX) (Variables|Parachute|GetParachuteMesh) "FX_Rotor" "KeepRelative" "KeepRelative" "KeepRelative" true)
    (Transformation|AttachComponentToComponent (CallFunction|GetTrailVFX2Component) (Variables|Parachute|GetParachuteMesh) "FX_Tail_Rotor" "KeepRelative" "KeepRelative" "KeepRelative" true)
    (Transformation|AttachComponentToComponent (CallFunction|GetTrailVFXComponent) (Variables|Parachute|GetParachuteMesh) "engine_exhaust" "SnapToTarget" "KeepRelative" "KeepRelative" true)
    (else
      (Components|Activation|Deactivate (Variables|Default|GetBladeSpinFX)))))

(event EventOnGliderFullyDeployed
  (|CallParentFunction)
  (Components|Activation|Activate (Variables|Default|GetBladeSpinFX) false))

(fn ConstructionScript
  (|CallParentFunction))
