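; Legacy Basketball glider Blueprint that adjusts a trail effect and drives audio when the glider fully deploys.
; After calling the parent's deploy event, it rotates the second trail VFX component to a fixed relative orientation.
; Only when the parachute is a front-end preview, it plays a FadeMusic timeline that feeds its fade value into a parameter on the legacy vehicle audio voice.
; The construction script only calls the parent.
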(event EventOnGliderFullyDeployed
  (|CallParentFunction)
  (Transformation|SetRelativeRotation (CallFunction|GetTrailVFX2Component) "0.000000,-90.000000,90.000000" false false)
  (if (Variables|FortPlayerParachute|GetbIsFrontEndPreview)
    (bind (_direction _fade) (Timeline|FadeMusic PlayFromStart)
      (:Update
        (Class|FortVehicleAudioVoice|SetParam (Variables|LegacyAudio|GetFortVehicleAudioVoice) _fade)))))

(fn ConstructionScript
  (|CallParentFunction))
