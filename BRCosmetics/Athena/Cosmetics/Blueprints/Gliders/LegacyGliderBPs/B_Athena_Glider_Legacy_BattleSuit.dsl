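; Legacy BattleSuit glider Blueprint that drives wing-flap flame materials and trail VFX on a player parachute.
; On full deployment it shows the flap material and plays a timeline on its "Glider Forward" parameter; hiding the chute zeroes MasterAlpha.
; Outside front-end preview, trail updates use forward movement and deploy state to toggle the second trail, flap opacity and a forward or reversed flame timeline.
; BeginPlay destroys the left flap light, although deployment still references it.
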(event EventBeginPlay
  (|CallParentFunction)
  (if (or (TouchPlatform|IsMobileGame) (TouchPlatform|IsNintendoSwitch))
    _
    (elif (>= (Rendering|GetRenderingDetailMode) 1)
      _))
  (Components|DestroyComponent (Variables|Default|GetFlapLight_L)))

(event EventOnGliderFullyDeployed
  (|CallParentFunction)
  (if (Variables|FortPlayerParachute|GetbIsFrontEndPreview)
    (Components|Activation|Deactivate (CallFunction|GetTrailVFXComponent)) ; Auto Activate
  )
  (Rendering|Material|SetScalarParameterValueonMaterials (Variables|Default|GetS_FX_Battelsuit_WingFlap) "MasterAlpha" 1.0)
  (if (or (TouchPlatform|IsMobileGame) (TouchPlatform|IsNintendoSwitch))
    _
    (elif (>= (Rendering|GetRenderingDetailMode) 1)
      _
      (else
        (Rendering|SetVisibility (Variables|Default|GetFlapLight_L) true false)
        (bind (_direction _lerp_controller) (Timeline|Timeline_0 PlayFromStart)
          (:Update
            (Rendering|Material|SetScalarParameterValueonMaterials (Variables|Default|GetS_FX_Battelsuit_WingFlap) "Glider Forward" (Math|Float|Lerp 0.0 2.0 _lerp_controller)))))))
  (Timeline|Timeline_0 PlayFromStart))

(event EventOnChuteVisibilityChanged
  (params (bParachuteVisibility :type Boolean))
  (|CallParentFunction bParachuteVisibility)
  (if (not bParachuteVisibility)
    (Rendering|Material|SetScalarParameterValueonMaterials (Variables|Default|GetS_FX_Battelsuit_WingFlap) "MasterAlpha" 0.0))
  (Components|Activation|Deactivate (CallFunction|GetTrailVFXComponent)) ; Auto Activate
)

(event Parachute|EventOnParachuteTrailUpdated
  (params (MovementLengthSquared :type Float%20%28single-precision%29) (ForwardDot :type Float%20%28single-precision%29) (RightDot :type Float%20%28single-precision%29) (RotationalMovementAcceleration :type Float%20%28single-precision%29))
  (bind _returnvalue (CallFunction|IsGliderFullyDeployedAccordingtoAnim))
  (bind _returnvalue_2 (CallFunction|GetTrailVFX2Component))
  (bind _s_fx_battelsuit_wingflap (Variables|Default|GetS_FX_Battelsuit_WingFlap))
  (|CallParentFunction MovementLengthSquared ForwardDot RightDot RotationalMovementAcceleration)
  (if (not (Variables|FortPlayerParachute|GetbIsFrontEndPreview))
    (Utilities|FlowControl|Sequence
      (:then_0
        (if (not (and (<= ForwardDot -0.2) _returnvalue))
          (Components|Activation|Deactivate _returnvalue_2))
        (Utilities|FlowControl|DoOnce false
          (:Completed
            (Components|Activation|Activate _returnvalue_2 false))))
      (:then_1
        (if (and _returnvalue (>= ForwardDot 0.2))
          (Utilities|FlowControl|DoOnce false
            (:Completed
              (CallFunction|FlapFlames)))
          (else
            (Utilities|FlowControl|DoOnce false
              (:Completed
                (CallFunction|FlapFlamesReverse))))))
      (:then_2
        (if (and _returnvalue (>= ForwardDot -0.1))
          (Rendering|Material|SetScalarParameterValueonMaterials _s_fx_battelsuit_wingflap "MasterAlpha" 1.0)
          (else
            (Rendering|Material|SetScalarParameterValueonMaterials _s_fx_battelsuit_wingflap "MasterAlpha" 0.0)))))))

(event Custom|FlapFlames
  (bind (_direction _lerp_controller) (Timeline|Timeline_1 PlayFromStart)
    (:Update
      (Rendering|Material|SetScalarParameterValueonMaterials (Variables|Default|GetS_FX_Battelsuit_WingFlap) "Glider Forward" (Math|Float|Lerp 0.0 1.0 _lerp_controller)))))

(event Custom|FlapFlames_Reverse
  (Timeline|Timeline_1 ReverseFromEnd))

(fn ConstructionScript
  (|CallParentFunction))
