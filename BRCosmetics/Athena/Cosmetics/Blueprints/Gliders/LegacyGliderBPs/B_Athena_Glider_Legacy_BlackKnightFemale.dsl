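; Legacy Black Knight Female glider Blueprint that extends its parent's parachute trail update with extra particle trails.
; It checks forward movement and whether the animation reports full deployment, deactivating and ending TrailVFX3 when either condition fails.
; A DoOnce gate activates TrailVFX3 and begins center-width trails on TrailVFX3 and TrailVFX4 between right and left socket pairs.
; TrailVFX4 is never explicitly activated or ended here, and the construction script only calls its parent.
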(event Parachute|EventOnParachuteTrailUpdated
  (params (MovementLengthSquared :type Float%20%28single-precision%29) (ForwardDot :type Float%20%28single-precision%29) (RightDot :type Float%20%28single-precision%29) (RotationalMovementAcceleration :type Float%20%28single-precision%29))
  (bind _trailvfx3 (Variables|Default|GetTrailVFX3))
  (|CallParentFunction MovementLengthSquared ForwardDot RightDot RotationalMovementAcceleration)
  (if (not (and (>= ForwardDot 0.2) (CallFunction|IsGliderFullyDeployedAccordingtoAnim)))
    (Components|Activation|Deactivate _trailvfx3)
    (Effects|Particles|Trails|EndTrails _trailvfx3))
  (Utilities|FlowControl|DoOnce false
    (:Completed
      (Components|Activation|Activate (Variables|Default|GetTrailVFX3) false)
      (Effects|Particles|Trails|BeginTrails (Variables|Default|GetTrailVFX3) "FX_Trail_R2" "FX_Trail_R3" "ETrailWidthMode_FromCentre" 1.0)
      (Effects|Particles|Trails|BeginTrails (Variables|Default|GetTrailVFX4) "FX_Trail_L2" "FX_Trail_L3" "ETrailWidthMode_FromCentre" 1.0))))

(fn ConstructionScript
  (|CallParentFunction))
