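; Legacy Black Widow glider Blueprint that drives its vehicle audio voice from parachute movement.
; On each trail update it calls the parent, then sets the voice parameter according to whether the animation reports the glider fully deployed.
; When deployed, a squared-movement threshold picks between a full and a half level; otherwise the parameter is zero.
; The construction script only calls the parent.
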(event Parachute|EventOnParachuteTrailUpdated
  (params (MovementLengthSquared :type Float%20%28single-precision%29) (ForwardDot :type Float%20%28single-precision%29) (RightDot :type Float%20%28single-precision%29) (RotationalMovementAcceleration :type Float%20%28single-precision%29))
  (bind _fortvehicleaudiovoice (Variables|LegacyAudio|GetFortVehicleAudioVoice))
  (|CallParentFunction MovementLengthSquared ForwardDot RightDot RotationalMovementAcceleration)
  (if (CallFunction|IsGliderFullyDeployedAccordingtoAnim)
    (Class|FortVehicleAudioVoice|SetParam _fortvehicleaudiovoice (select (>= MovementLengthSquared 0.2) 1.0 0.5))
    (else
      (Class|FortVehicleAudioVoice|SetParam _fortvehicleaudiovoice 0.0))))

(fn ConstructionScript
  (|CallParentFunction))
