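; Blueprint for the legacy Bling glider cosmetic, adding show/hide logic for a center contrail particle on top of its parent glider class.
; On full deploy, it restarts the contrail if the glider is displayed in a frontend context.
; On each parachute trail update, it activates the contrail without reset when squared movement length passes a threshold and the animation reports full deployment, otherwise deactivates it.
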(event EventOnGliderFullyDeployed
  (|CallParentFunction)
  (if (Reactive|Frontend|IsCosmeticObjectDisplayedinFrontendContext 0)
    (Components|Activation|Activate (Variables|Default|GetP_GliderContrail_Center) true)))

(event Parachute|EventOnParachuteTrailUpdated
  (params (MovementLengthSquared :type Float%20%28single-precision%29) (ForwardDot :type Float%20%28single-precision%29) (RightDot :type Float%20%28single-precision%29) (RotationalMovementAcceleration :type Float%20%28single-precision%29))
  (bind _p_glidercontrail_center (Variables|Default|GetP_GliderContrail_Center))
  (|CallParentFunction MovementLengthSquared ForwardDot RightDot RotationalMovementAcceleration)
  (if (and (>= MovementLengthSquared 0.2) (CallFunction|IsGliderFullyDeployedAccordingtoAnim))
    (Components|Activation|Activate _p_glidercontrail_center false)
    (else
      (Components|Activation|Deactivate _p_glidercontrail_center))))

(fn ConstructionScript
  (|CallParentFunction))
