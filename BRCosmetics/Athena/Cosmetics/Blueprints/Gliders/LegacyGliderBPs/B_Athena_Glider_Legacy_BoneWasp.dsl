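; Legacy Bone Wasp glider cosmetic Blueprint that adds engine exhaust and trail effects to the glider.
; When the chute becomes visible, it snaps exhaust loop and start effects to the back engine socket and two trail VFX to the left and right engine sockets.
; Full deployment activates the start exhaust, and hiding the chute deactivates it.
; Trail updates check forward movement and deployment, and include a DoOnce-gated start exhaust activation.
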(event EventOnChuteVisibilityChanged
  (params (bParachuteVisibility :type Boolean))
  (bind _parachutemesh (Variables|Parachute|GetParachuteMesh))
  (|CallParentFunction bParachuteVisibility)
  (if bParachuteVisibility
    (Transformation|AttachComponentToComponent (Variables|Default|GetExhaustLoop) _parachutemesh "S_EngineBack" "SnapToTarget" "KeepRelative" "KeepRelative" true)
    (Transformation|AttachComponentToComponent (Variables|Default|GetExhaustStart) _parachutemesh "S_EngineBack" "SnapToTarget" "KeepRelative" "KeepRelative" true)
    (Transformation|AttachComponentToComponent (CallFunction|GetTrailVFXComponent) _parachutemesh "S_Engine_L" "SnapToTarget" "KeepRelative" "KeepRelative" true)
    (Transformation|AttachComponentToComponent (CallFunction|GetTrailVFX2Component) _parachutemesh "S_Engine_R" "SnapToTarget" "KeepRelative" "KeepRelative" true)
    (else
      (Components|Activation|Deactivate (Variables|Default|GetExhaustStart)))))

(event EventOnGliderFullyDeployed
  (|CallParentFunction)
  (Components|Activation|Activate (Variables|Default|GetExhaustStart) false))

(event Parachute|EventOnParachuteTrailUpdated
  (params (MovementLengthSquared :type Float%20%28single-precision%29) (ForwardDot :type Float%20%28single-precision%29) (RightDot :type Float%20%28single-precision%29) (RotationalMovementAcceleration :type Float%20%28single-precision%29))
  (|CallParentFunction MovementLengthSquared ForwardDot RightDot RotationalMovementAcceleration)
  (if (and (>= ForwardDot 0.2) (CallFunction|IsGliderFullyDeployedAccordingtoAnim))
    _)
  (Utilities|FlowControl|DoOnce false
    (:Completed
      (Components|Activation|Activate (Variables|Default|GetExhaustStart) false))))

(fn ConstructionScript
  (|CallParentFunction))
