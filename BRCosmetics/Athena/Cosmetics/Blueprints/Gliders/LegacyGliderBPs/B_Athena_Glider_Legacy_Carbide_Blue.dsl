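; Glider Blueprint for the legacy Carbide Blue cosmetic glider in Athena, adding a glow effect tied to chute visibility.
; When chute visibility changes, it calls the parent handler, then activates the Glider_Glow component if the parachute is visible and deactivates it otherwise.
; The construction script only calls its parent.
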(event EventOnChuteVisibilityChanged
  (params (bParachuteVisibility :type Boolean))
  (bind _glider_glow (Variables|Default|GetGlider_Glow))
  (|CallParentFunction bParachuteVisibility)
  (if bParachuteVisibility
    (Components|Activation|Activate _glider_glow false)
    (else
      (Components|Activation|Deactivate _glider_glow))))

(fn ConstructionScript
  (|CallParentFunction))
