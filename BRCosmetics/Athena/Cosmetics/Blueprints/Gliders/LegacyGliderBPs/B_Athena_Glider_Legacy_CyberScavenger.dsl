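; Cyber Scavenger legacy glider Blueprint that drives the glider's cosmetic trail and tail-light particle effects.
; On showing, it attaches two trails to the cowl sockets and the tail light to the base, which activates on full deployment and deactivates on hiding.
; Forward movement while deployed triggers a DoOnce-gated boost of light-speed parameters and a timeline ramping trail rotation.
; When the ramp ends, light speeds drop and the timeline reverses.
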(event EventOnChuteVisibilityChanged
  (params (bParachuteVisibility :type Boolean))
  (|CallParentFunction bParachuteVisibility)
  (if bParachuteVisibility
    (Components|Activation|Deactivate (CallFunction|GetTrailVFXComponent)) ; Auto Activate
    (Transformation|AttachComponentToComponent (CallFunction|GetTrailVFXComponent) (Variables|Parachute|GetParachuteMesh) "cowl_l" "KeepRelative" "KeepRelative" "KeepRelative" true)
    (Transformation|AttachComponentToComponent (CallFunction|GetTrailVFX2Component) (Variables|Parachute|GetParachuteMesh) "cowl_r" "KeepRelative" "KeepRelative" "KeepRelative" true)
    (Transformation|AttachComponentToComponent (Variables|Default|GetP_CyberScavenger_TailLight) (Variables|Parachute|GetParachuteMesh) "Glider_Base" "KeepRelative" "KeepRelative" "KeepRelative" true)
    (else
      (Components|Activation|Deactivate (Variables|Default|GetP_CyberScavenger_TailLight)))))

(event EventOnGliderFullyDeployed
  (|CallParentFunction)
  (Components|Activation|Activate (Variables|Default|GetP_CyberScavenger_TailLight) false))

(event Parachute|EventOnParachuteTrailUpdated
  (params (MovementLengthSquared :type Float%20%28single-precision%29) (ForwardDot :type Float%20%28single-precision%29) (RightDot :type Float%20%28single-precision%29) (RotationalMovementAcceleration :type Float%20%28single-precision%29))
  (|CallParentFunction MovementLengthSquared ForwardDot RightDot RotationalMovementAcceleration)
  (if (and (>= ForwardDot 0.2) (CallFunction|IsGliderFullyDeployedAccordingtoAnim))
    (Utilities|FlowControl|DoOnce false
      (:Completed
        (Effects|Components|ParticleSystem|SetFloatParameter (Variables|Default|GetP_CyberScavenger_TailLight) "BrakeLightSpeed" 5.0)
        (Effects|Components|ParticleSystem|SetFloatParameter (CallFunction|GetTrailVFXComponent) "CLightSpeed" 5.0)
        (bind (_direction _newtrack_0) (Timeline|Timeline_0 PlayFromStart)
          (:Update
            (Effects|Components|ParticleSystem|SetVectorParameter (CallFunction|GetTrailVFXComponent) "RotateSpeed" (+ (* (* "0, 0,-1.000000" (- (Variables|Default|GetMax_Rotation_speed) (Variables|Default|GetMin_Rotation_speed))) _newtrack_0) (* "0, 0,-1.000000" (Variables|Default|GetMin_Rotation_speed)))))
          (:Finished
            (switch Utilities|FlowControl|Switch|SwitchonETimelineDirection _direction
              (:Forward
                (Utilities|FlowControl|DoOnce false
                  (:Completed
                    (Effects|Components|ParticleSystem|SetFloatParameter (Variables|Default|GetP_CyberScavenger_TailLight) "BrakeLightSpeed" 1.0)
                    (Effects|Components|ParticleSystem|SetFloatParameter (CallFunction|GetTrailVFXComponent) "CLightSpeed" 1.0)
                    (Timeline|Timeline_0 ReverseFromEnd))))
              (:Backward))))))
    (else
      _)))

(fn ConstructionScript
  (|CallParentFunction))
