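; Legacy glider Blueprint for the DJ Remix cosmetic, adding wave-ring particles and light trails.
; Full deployment starts the wave ring and, in front-end preview, a timeline that drives a vehicle audio parameter.
; Hiding the chute stops the ring.
; Trail updates begin six socket-to-socket trails via DoOnce, shortening life on two right-side trails, and end only Trail_Light_R1 when not deployed or moving sufficiently forward.
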(event EventOnGliderFullyDeployed
  (bind _p_glider_djremix_wavering (Variables|Default|GetP_Glider_DJRemix_WaveRing))
  (|CallParentFunction)
  (Components|Activation|Activate _p_glider_djremix_wavering false)
  (Effects|Components|ParticleSystem|SetFloatParameter _p_glider_djremix_wavering "Spawn" 1.0)
  (if (Variables|FortPlayerParachute|GetbIsFrontEndPreview)
    (bind (_direction _fade) (Timeline|Timeline_0 PlayFromStart)
      (:Update
        (Class|FortVehicleAudioVoice|SetParam (Variables|LegacyAudio|GetFortVehicleAudioVoice) _fade)))))

(event EventOnChuteVisibilityChanged
  (params (bParachuteVisibility :type Boolean))
  (bind _p_glider_djremix_wavering (Variables|Default|GetP_Glider_DJRemix_WaveRing))
  (|CallParentFunction bParachuteVisibility)
  (if (not bParachuteVisibility)
    (Effects|Components|ParticleSystem|SetFloatParameter _p_glider_djremix_wavering "Spawn" 0.0)
    (Components|Activation|Deactivate _p_glider_djremix_wavering)))

(event Parachute|EventOnParachuteTrailUpdated
  (params (MovementLengthSquared :type Float%20%28single-precision%29) (ForwardDot :type Float%20%28single-precision%29) (RightDot :type Float%20%28single-precision%29) (RotationalMovementAcceleration :type Float%20%28single-precision%29))
  (|CallParentFunction MovementLengthSquared ForwardDot RightDot RotationalMovementAcceleration)
  (if (not (and (>= ForwardDot 0.2) (CallFunction|IsGliderFullyDeployedAccordingtoAnim)))
    (Effects|Particles|Trails|EndTrails (Variables|Default|GetTrail_Light_R1)))
  (Utilities|FlowControl|DoOnce false
    (:Completed
      (Effects|Particles|Trails|BeginTrails (Variables|Default|GetTrail_Light_L1) "Anim_light_L1" "Anim_light_L2" "ETrailWidthMode_FromCentre" 1.7)
      (Effects|Particles|Trails|BeginTrails (Variables|Default|GetTrail_Light_L2) "Anim_light_L3" "Anim_light_L4" "ETrailWidthMode_FromCentre" 1.7)
      (Effects|Particles|Trails|BeginTrails (Variables|Default|GetTrail_Light_L3) "Anim_light_L5" "Anim_light_L6" "ETrailWidthMode_FromCentre" 1.7)
      (Effects|Particles|Trails|BeginTrails (Variables|Default|GetTrail_Light_R1) "Anim_light_R1" "Anim_light_R2" "ETrailWidthMode_FromCentre" 1.7)
      (Effects|Particles|Trails|BeginTrails (Variables|Default|GetTrail_Light_R2) "Anim_light_R3" "Anim_light_R4" "ETrailWidthMode_FromCentre" 1.7)
      (Effects|Particles|Trails|BeginTrails (Variables|Default|GetTrail_Light_R3) "Anim_light_R5" "Anim_light_R6" "ETrailWidthMode_FromCentre" 1.7)
      (Effects|Components|ParticleSystem|SetFloatParameter (Variables|Default|GetTrail_Light_R2) "LightTrail_Life" 0.07)
      (Effects|Components|ParticleSystem|SetFloatParameter (Variables|Default|GetTrail_Light_R3) "LightTrail_Life" 0.04))))

(fn ConstructionScript
  (|CallParentFunction))
