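; Legacy Dark Viking dragon glider Blueprint that drives storm effects, wing-flap sounds, trails and fire-breath.
; While visible, it binds to anim-instance roar events that toggle flame particles and a roar sound, skipped for KnightRemix.
; A timeline fades and scales the flame spline mesh, feeding interpolated drift into its material's BendOffset for lagging bend.
; Socket-based trails run during deployed forward movement, with one removed on mobile.
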(event EventOnChuteVisibilityChanged
  (params (bParachuteVisibility :type Boolean))
  (|CallParentFunction bParachuteVisibility)
  (if bParachuteVisibility
    (bind _storm (Variables|Default|GetStorm)
      (:then
        (Components|Activation|Activate _storm false))
      (:else
        _))
    (bind _asdv_dragon_glider_anim_bp (Utilities|Casting|CastToDV_Dragon_Glider_AnimBP (Components|SkeletalMesh|GetAnimInstance (Variables|Parachute|GetParachuteMesh)))
      (:then
        (Default|BindEventtoRoarEvent _asdv_dragon_glider_anim_bp (EventDispatchers|CreateEvent :!event Custom|Roar%20On))
        (Default|BindEventtoRoarEventEnd _asdv_dragon_glider_anim_bp (EventDispatchers|CreateEvent :!event Custom|Roar%20Off)))
      (:CastFailed
        (Development|PrintString "AnimINstance Cast Failed")))
    (else
      (bind _storm_2 (Variables|Default|GetStorm)
        (:then
          (Components|Activation|Deactivate _storm_2))
        (:else
          _))
      (bind _flamesaudiocomp (Variables|Default|GetFlamesAudioComp)
        (:then
          (Audio|Components|Audio|FadeOut _flamesaudiocomp 0.25 0.0 "Linear" :!class /Script/Engine.AudioComponent)
          (bind _asdv_dragon_glider_anim_bp_2 (Utilities|Casting|CastToDV_Dragon_Glider_AnimBP (Components|SkeletalMesh|GetAnimInstance (Variables|Parachute|GetParachuteMesh))))
          (Default|UnbindEventfromRoarEvent _asdv_dragon_glider_anim_bp_2 (EventDispatchers|CreateEvent :!event Custom|Roar%20On))
          (Default|UnbindEventfromRoarEventEnd _asdv_dragon_glider_anim_bp_2 (EventDispatchers|CreateEvent :!event Custom|Roar%20Off)))
        (:else)))))

(event EventBeginPlay
  (|CallParentFunction)
  (if (TouchPlatform|IsMobileGame)
    (Components|DestroyComponent (Variables|Default|GetTrailParticles5))
    (else
      _))
  (Variables|Default|SetFlame%20Drift (Math|Transform|InverseTransformLocation (Transformation|GetActorTransform) (Transformation|GetWorldLocation (Variables|Default|GetS_DarkViking_Glider_FlameSpline))))
  (for (_array_element _array_index) (Variables|Default|GetAnimTrails)
    (Utilities|IsValid _array_element
      (:Is%20Valid
        (Effects|Components|ParticleSystem|SetFloatParameter _array_element "Lifetime" (Utilities|Select :Index _array_index :Option%200 0.14 :Option%201 0.14 :Option%202 0.07 :Option%203 0.07 :Option%204 0.035 :Option%205 0.035))))))

(event Custom|On%20Flap%20State%20Changed
  (params (New%20Flap%20State :type Boolean) (Can%20Flap :type Boolean) (Fast%20Flap :type Boolean))
  (if Can%20Flap
    (if New%20Flap%20State
      (bind _flapsoundaudiocomp (Variables|Default|GetFlapSoundAudioComp)
        (:then
          (Audio|Components|Audio|FadeOut _flapsoundaudiocomp 0.5 0.0 "Linear" :!class /Script/Engine.AudioComponent))
        (:else
          _))
      (bind _returnvalue (Audio|SpawnSoundAttached (select Fast%20Flap "/Game/Sounds/Fort_Gadgets/Athena/Athena_Glider_DarkViking_Dragon_Wing_Flap_02_Cue.Athena_Glider_DarkViking_Dragon_Wing_Flap_02_Cue" "/Game/Sounds/Fort_Gadgets/Athena/Athena_Glider_DarkViking_Dragon_Wing_Flap_Cue.Athena_Glider_DarkViking_Dragon_Wing_Flap_Cue") (Variables|Parachute|GetParachuteMesh)))
      (Variables|Default|SetFlapSoundAudioComp _returnvalue)
      (else
        (bind _flapsoundaudiocomp_2 (Variables|Default|GetFlapSoundAudioComp)
          (:then
            (Audio|Components|Audio|FadeOut _flapsoundaudiocomp_2 0.5 0.0 "Linear" :!class /Script/Engine.AudioComponent))
          (:else))))))

(event Custom|Roar%20On
  (Utilities|FlowControl|DoOnce false
    (:Completed
      (CallFunction|FlamesOn/off false)
      (if (not (Utilities|String|Contains (Utilities|GetDisplayName (CosmeticLoadout|GetItemDefinitionfromLoadoutwithActor (CallFunction|GetFortPlayerPawn) "(TagName=\"LoadoutSlot.Glider\")")) "Glider_ID_165_KnightRemix"))
        (bind _flamesaudiocomp (Variables|Default|GetFlamesAudioComp)
          (:then
            (Audio|Components|Audio|Stop _flamesaudiocomp))
          (:else
            _))
        (bind _returnvalue (Audio|SpawnSoundAttached "/Game/Sounds/Fort_Gadgets/Athena/Athena_Glider_DarkViking_Dragon_Fire_Roar_Cue.Athena_Glider_DarkViking_Dragon_Fire_Roar_Cue" (Variables|Parachute|GetParachuteMesh)))
        (Variables|Default|SetFlamesAudioComp _returnvalue)))))

(event Custom|Reset_On
  (Utilities|FlowControl|DoOnce false
    (:Completed
      (CallFunction|FlamesOn/off false)
      (if (not (Utilities|String|Contains (Utilities|GetDisplayName (CosmeticLoadout|GetItemDefinitionfromLoadoutwithActor (CallFunction|GetFortPlayerPawn) "(TagName=\"LoadoutSlot.Glider\")")) "Glider_ID_165_KnightRemix"))
        (bind _flamesaudiocomp (Variables|Default|GetFlamesAudioComp)
          (:then
            (Audio|Components|Audio|Stop _flamesaudiocomp))
          (:else
            _))
        (bind _returnvalue (Audio|SpawnSoundAttached "/Game/Sounds/Fort_Gadgets/Athena/Athena_Glider_DarkViking_Dragon_Fire_Roar_Cue.Athena_Glider_DarkViking_Dragon_Fire_Roar_Cue" (Variables|Parachute|GetParachuteMesh)))
        (Variables|Default|SetFlamesAudioComp _returnvalue)))))

(event Custom|Roar%20Off
  (Utilities|FlowControl|DoOnce true
    (:Completed
      (CallFunction|FlamesOn/off true))))

(event Custom|Reset_Off
  (Utilities|FlowControl|DoOnce true
    (:Completed
      (CallFunction|FlamesOn/off true))))

(event Custom|Reset_Trail
  (Utilities|FlowControl|DoOnce false
    (:Completed
      (for (_array_element _array_index) (Variables|Default|GetAnimTrails)
        (Utilities|IsValid _array_element
          (:Is%20Valid
            (Components|Activation|Activate _array_element false)
            (Effects|Particles|Trails|BeginTrails _array_element (Utilities|Map|Find (Variables|Default|GetFirst%20Socket%20Name) _array_index) (Utilities|Map|Find (Variables|Default|GetSecond%20Socket%20Name) _array_index) "ETrailWidthMode_FromCentre" 1.0))))
      (CallFunction|ResetCloseTrail))))

(event Parachute|EventOnParachuteTrailUpdated
  (params (MovementLengthSquared :type Float%20%28single-precision%29) (ForwardDot :type Float%20%28single-precision%29) (RightDot :type Float%20%28single-precision%29) (RotationalMovementAcceleration :type Float%20%28single-precision%29))
  (if (and (>= ForwardDot 0.2) (CallFunction|IsGliderFullyDeployedAccordingtoAnim))
    (Utilities|FlowControl|DoOnce false
      (:Completed
        (for (_array_element _array_index) (Variables|Default|GetAnimTrails)
          (Utilities|IsValid _array_element
            (:Is%20Valid
              (Components|Activation|Activate _array_element false)
              (Effects|Particles|Trails|BeginTrails _array_element (Utilities|Map|Find (Variables|Default|GetFirst%20Socket%20Name) _array_index) (Utilities|Map|Find (Variables|Default|GetSecond%20Socket%20Name) _array_index) "ETrailWidthMode_FromCentre" 1.0))))
        (CallFunction|ResetCloseTrail)))
    (else
      (Utilities|FlowControl|DoOnce true
        (:Completed
          (for _array_element_2 (Variables|Default|GetAnimTrails)
            (Utilities|IsValid _array_element_2
              (:Is%20Valid
                (Components|Activation|Deactivate _array_element_2)
                (Effects|Particles|Trails|EndTrails _array_element_2))))
          (CallFunction|ResetTrail)
          (Class|FortVehicleAudioVoice|SetParam (Variables|LegacyAudio|GetFortVehicleAudioVoice) 0.0))))))

(event Custom|Reset_Close_Trail
  (Utilities|FlowControl|DoOnce true
    (:Completed
      (for _array_element (Variables|Default|GetAnimTrails)
        (Utilities|IsValid _array_element
          (:Is%20Valid
            (Components|Activation|Deactivate _array_element)
            (Effects|Particles|Trails|EndTrails _array_element))))
      (CallFunction|ResetTrail)
      (Class|FortVehicleAudioVoice|SetParam (Variables|LegacyAudio|GetFortVehicleAudioVoice) 0.0))))

(event Custom|Flames%20on/off
  (params (Turn%20Off :type Boolean))
  (bind _p_darkvikingglider_flames (Variables|Default|GetP_DarkVikingGlider_Flames))
  (bind _s_darkviking_glider_flamespline (Variables|Default|GetS_DarkViking_Glider_FlameSpline))
  ;[1 1 1] For Fading Flames in and out
  (if Turn%20Off
    (Utilities|FlowControl|Sequence
      (:then_0
        (bind (_direction _alpha) (Timeline|FlameFade ReverseFromEnd)
          (:Update
            (CallFunction|FlameSplineControl)
            (Rendering|Material|SetScalarParameterValueonMaterials _s_darkviking_glider_flamespline "Alpha" (Math|Float|Lerp 0.0 (Variables|Default|GetMax%20Alpha%20Value) _alpha))
            (switch Utilities|FlowControl|Switch|SwitchonETimelineDirection _direction
              (:Forward
                (Transformation|SetWorldScale3D _s_darkviking_glider_flamespline (Math|Vector|Lerp(Vector) "0,1.000000,1.000000" "2.000000,1.000000,1.000000" _alpha)))))
          (:Finished
            (switch Utilities|FlowControl|Switch|SwitchonETimelineDirection _direction
              (:Forward
                (CallFunction|ResetOff))
              (:Backward
                (CallFunction|ResetOn)
                (Transformation|SetWorldScale3D _s_darkviking_glider_flamespline "0, 0, 0"))))))
      (:then_1
        (Components|Activation|SetActive _p_darkvikingglider_flames false false)))
    (else
      (Components|Activation|SetActive _p_darkvikingglider_flames true false)
      (Timeline|FlameFade PlayFromStart))))

(event Custom|Flame%20Spline%20Control
  (bind _flametarget (Variables|Default|GetFlameTarget))
  (bind _returnvalue (Transformation|GetWorldLocation _flametarget))
  (bind _returnvalue_2 (Transformation|GetActorTransform))
  (bind _returnvalue_3 (Math|Transform|InverseTransformLocation _returnvalue_2 _returnvalue))
  ;[1 1 1] Update Flame Spline Mesh WPO
  (Networking|CanExecuteCosmeticEvents
    (:True
      (bind _output_get (Variables|Default|SetFlame%20Drift (Math|Interpolation|VInterpTo (Variables|Default|GetFlame%20Drift) _returnvalue_3 (Utilities|Time|GetWorldDeltaSeconds) 6.0)))
      (Rendering|Material|SetVectorParameterValueonMaterials (Variables|Default|GetS_DarkViking_Glider_FlameSpline) "BendOffset" (* (- _output_get _returnvalue_3) (Variables|Default|GetOffsetScalar))))))

(fn ConstructionScript
  (|CallParentFunction)
  (Variables|Default|SetAnimTrails (Utilities|Array|MakeArray (Variables|Default|GetTrailParticles1) (Variables|Default|GetTrailParticles2) (Variables|Default|GetTrailParticles3) (Variables|Default|GetTrailParticles4) (Variables|Default|GetTrailParticles5) (Variables|Default|GetTrailParticles6))))
