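; Legacy DieselPunk glider cosmetic Blueprint that manages the glider's headlight and trail effects.
; On full deployment it shows the headlight, attaches the second trail VFX component to the parachute mesh's "Light" socket and activates the first trail.
; When the chute becomes visible it deactivates that trail, except in a fully deployed front-end preview.
; When the chute hides, the headlight is hidden.
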(event EventOnGliderFullyDeployed
  (|CallParentFunction)
  (Rendering|SetVisibility (Variables|Default|GetSM_DieselPunk_Glider_HeadLight_Light) true false)
  (Transformation|AttachComponentToComponent (CallFunction|GetTrailVFX2Component) (Variables|Parachute|GetParachuteMesh) "Light" "KeepRelative" "KeepRelative" "KeepRelative" true)
  (Components|Activation|Activate (CallFunction|GetTrailVFXComponent) false))

(event EventOnChuteVisibilityChanged
  (params (bParachuteVisibility :type Boolean))
  (|CallParentFunction bParachuteVisibility)
  (if bParachuteVisibility
    (if (not (and (CallFunction|IsGliderFullyDeployedAccordingtoAnim) (Variables|FortPlayerParachute|GetbIsFrontEndPreview)))
      (Components|Activation|Deactivate (CallFunction|GetTrailVFXComponent)))
    (else
      (Rendering|SetVisibility (Variables|Default|GetSM_DieselPunk_Glider_HeadLight_Light) false false))))

(fn ConstructionScript
  (|CallParentFunction))
