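; Legacy Athena glider Blueprint for the DonutPlate cosmetic, controlling two trail VFX components.
; When the parachute becomes visible, it attaches both trails to the mesh's Glider_Base socket and activates TrailVFX2.
; TrailVFX is deactivated on every visibility change, working around an auto-activate issue on show, and activates only once the glider fully deploys.
; Hiding the parachute leaves TrailVFX2 untouched in this graph.
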(event EventOnChuteVisibilityChanged
  (params (bParachuteVisibility :type Boolean))
  (|CallParentFunction bParachuteVisibility)
  (if bParachuteVisibility
    (Components|Activation|Deactivate (CallFunction|GetTrailVFXComponent)) ; Auto Activate Issue
    (Transformation|AttachComponentToComponent (CallFunction|GetTrailVFXComponent) (Variables|Parachute|GetParachuteMesh) "Glider_Base" "KeepRelative" "KeepRelative" "KeepRelative" true)
    (Transformation|AttachComponentToComponent (CallFunction|GetTrailVFX2Component) (Variables|Parachute|GetParachuteMesh) "Glider_Base" "KeepRelative" "KeepRelative" "KeepRelative" true)
    (Components|Activation|Activate (CallFunction|GetTrailVFX2Component) false)
    (else
      (Components|Activation|Deactivate (CallFunction|GetTrailVFXComponent)))))

(event EventOnGliderFullyDeployed
  (|CallParentFunction)
  (Components|Activation|Activate (CallFunction|GetTrailVFXComponent) false))

(fn ConstructionScript
  (|CallParentFunction))
