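; Legacy Drift glider Blueprint that extends its parent's chute and trail events with hand and trail VFX.
; When the chute shows, it attaches TrailVFX2 to the pawn's right hand and deactivates it; delayed do-once events reactivate it until full deployment.
; Trail updates switch off Fan_L and zero trail opacity without forward motion and full deployment, and use a do-once timeline to animate opacity and size.
; "No Player Pawn" is true when a pawn exists, despite its name.
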(event EventOnChuteVisibilityChanged
  (params (bParachuteVisibility :type Boolean))
  (bind _returnvalue (CallFunction|GetFortPlayerPawn))
  (|CallParentFunction bParachuteVisibility)
  (if bParachuteVisibility
    (Utilities|IsValid _returnvalue
      (:Is%20Valid
        (Transformation|AttachComponentToComponent (CallFunction|GetTrailVFX2Component) (Variables|/Script/Engine.Character|GetMesh _returnvalue) "Hand_r" "SnapToTarget" "SnapToTarget" "KeepWorld" true)
        (Variables|Default|SetNo%20Player%20Pawn true))
      (:Is%20Not%20Valid
        (Variables|Default|SetNo%20Player%20Pawn false)))
    (Components|Activation|Deactivate (CallFunction|GetTrailVFX2Component))
    (else
      (CallFunction|Reset))))

(event EventOnGliderFullyDeployed
  (|CallParentFunction)
  (Components|Activation|Deactivate (CallFunction|GetTrailVFX2Component)))

(event Parachute|EventOnParachuteTrailUpdated
  (params (MovementLengthSquared :type Float%20%28single-precision%29) (ForwardDot :type Float%20%28single-precision%29) (RightDot :type Float%20%28single-precision%29) (RotationalMovementAcceleration :type Float%20%28single-precision%29))
  (bind _returnvalue (CallFunction|GetTrailVFXComponent))
  (bind _returnvalue_2 (CallFunction|GetTrailVFXComponent))
  (|CallParentFunction MovementLengthSquared ForwardDot RightDot RotationalMovementAcceleration)
  (CallFunction|PlayerDeployEffect)
  (if (not (and (>= ForwardDot 0.2) (CallFunction|IsGliderFullyDeployedAccordingtoAnim)))
    (Components|Activation|Deactivate (Variables|Default|GetFan_L))
    (Effects|Components|ParticleSystem|SetFloatParameter _returnvalue "Opacity" 0.0)
    (Effects|Components|ParticleSystem|SetFloatParameter _returnvalue "Opacity_a" 0.0))
  (Utilities|FlowControl|DoOnce false
    (:Completed
      (Components|Activation|Activate (Variables|Default|GetFan_L) false)
      (bind (_direction _opacity _opacity_a _size) (Timeline|Timeline_0 PlayFromStart)
        (:Update
          (Effects|Components|ParticleSystem|SetFloatParameter _returnvalue_2 "Opacity" _opacity)
          (Effects|Components|ParticleSystem|SetFloatParameter _returnvalue_2 "Opacity_a" _opacity_a)
          (Effects|Components|ParticleSystem|SetVectorParameter _returnvalue_2 "Size" (Math|Vector|MakeVector _size 1.0 1.0)))))))

(event Custom|PlayerDeployEffect
  (Utilities|FlowControl|DoOnce false
    (:Completed
      (Utilities|FlowControl|Delay 0.52)
      (if (Variables|Default|GetNo%20Player%20Pawn)
        (Components|Activation|Activate (CallFunction|GetTrailVFX2Component) true)))))

(event Custom|Reset
  (Utilities|FlowControl|DoOnce false
    (:Completed
      (Utilities|FlowControl|Delay 0.52)
      (if (Variables|Default|GetNo%20Player%20Pawn)
        (Components|Activation|Activate (CallFunction|GetTrailVFX2Component) true)))))

(fn ConstructionScript
  (|CallParentFunction))
