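; Durrburger Hero legacy glider cosmetic Blueprint that drives two trail VFX components and a movement audio parameter.
; Trail updates use forward/right dot and rotational acceleration to switch trails by turn direction and tweak ribbon offset once the glider is deployed.
; One timeline ramps velocity and offset on activation; another shrinks, erodes and drops the trail when movement stops.
; DoOnce gates, re-armed through mutual reset events, limit retriggering.
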(event Custom|ResetActivateR
  (bind _returnvalue (CallFunction|GetTrailVFXComponent))
  (bind _returnvalue_2 (CallFunction|GetTrailVFXComponent))
  (Utilities|FlowControl|DoOnce false
    (:Completed
      (Components|Activation|Activate (CallFunction|GetTrailVFX2Component) false)
      (Components|Activation|Deactivate (CallFunction|GetTrailVFXComponent))
      (CallFunction|ResetActivateL)
      (Effects|Components|ParticleSystem|SetVectorParameter _returnvalue_2 "SizeByDeactivate" "1.000000,1.000000,1.000000")
      (Effects|Components|ParticleSystem|SetFloatParameter _returnvalue_2 "ErodeByBP" 1.0)
      (Effects|Components|ParticleSystem|SetFloatParameter (CallFunction|GetTrailVFX2Component) "LifeTime" 0.18)
      (bind (_direction _velocity _locationoffset) (Timeline|Timeline_0 PlayFromStart)
        (:Update
          (Effects|Components|ParticleSystem|SetVectorParameter _returnvalue "Velocity" (Math|Vector|Lerp(Vector) "0, 0, 0" "1000.000000, 0,0.000000" _velocity))
          (Effects|Components|ParticleSystem|SetVectorParameter _returnvalue "LocationOffset" (Math|Vector|Lerp(Vector) "-15.000000,0.000000,0.000000" "-5.000000, 0, 0" _locationoffset)))
        (:Finished
          (CallFunction|ResetDeactivate))))))

(event Custom|ResetActivateL
  (bind _returnvalue (CallFunction|GetTrailVFXComponent))
  (bind _returnvalue_2 (CallFunction|GetTrailVFXComponent))
  (Utilities|FlowControl|DoOnce false
    (:Completed
      (Components|Activation|Activate (CallFunction|GetTrailVFXComponent) false)
      (Components|Activation|Deactivate (CallFunction|GetTrailVFX2Component))
      (CallFunction|ResetActivateR)
      (Effects|Components|ParticleSystem|SetVectorParameter _returnvalue_2 "SizeByDeactivate" "1.000000,1.000000,1.000000")
      (Effects|Components|ParticleSystem|SetFloatParameter _returnvalue_2 "ErodeByBP" 1.0)
      (Effects|Components|ParticleSystem|SetFloatParameter (CallFunction|GetTrailVFX2Component) "LifeTime" 0.18)
      (Timeline|Timeline_0 PlayFromStart))))

(event Custom|ResetActivate
  (bind _returnvalue (CallFunction|GetTrailVFXComponent))
  (Utilities|FlowControl|DoOnce false
    (:Completed
      (Timeline|Timeline_0 PlayFromStart))))

(event Parachute|EventOnParachuteTrailUpdated
  (params (MovementLengthSquared :type Float%20%28single-precision%29) (ForwardDot :type Float%20%28single-precision%29) (RightDot :type Float%20%28single-precision%29) (RotationalMovementAcceleration :type Float%20%28single-precision%29))
  (bind _returnvalue (CallFunction|GetTrailVFXComponent))
  (bind _returnvalue_2 (CallFunction|GetTrailVFXComponent))
  (bind _returnvalue_3 (CallFunction|GetTrailVFX2Component))
  (bind _returnvalue_4 (CallFunction|GetTrailVFXComponent))
  (bind _forward_dot (Variables|Default|GetForward%20Dot))
  (bind _returnvalue_5 (>= _forward_dot 0.2))
  (bind _right_dot (Variables|Default|GetRight%20Dot))
  (bind _returnvalue_6 (< _right_dot -0.2))
  (bind _returnvalue_7 (> _right_dot 0.2))
  (bind _returnvalue_8 (Math|Boolean|ORBoolean _returnvalue_5 _returnvalue_6 _returnvalue_7))
  (Variables|Default|SetForward%20Dot ForwardDot)
  (Variables|Default|SetRight%20Dot RightDot)
  (Variables|Default|SetRotMoveAcc RotationalMovementAcceleration)
  (CallFunction|RibbonTrailLocationTweak)
  (CallFunction|UpdateAudio _returnvalue_8)
  (if (and _returnvalue_8 (CallFunction|IsGliderFullyDeployedAccordingtoAnim))
    (if (or (< (Variables|Default|GetRight%20Dot) -0.2) (Math|Float|InRange(Float) (Variables|Default|GetRotMoveAcc) 0.2 3.5e+02))
      (Utilities|FlowControl|DoOnce false
        (:Completed
          (Components|Activation|Activate (CallFunction|GetTrailVFX2Component) false)
          (Components|Activation|Deactivate (CallFunction|GetTrailVFXComponent))
          (CallFunction|ResetActivateL)))
      (else
        (if (or (> (Variables|Default|GetRight%20Dot) 0.2) (Math|Float|InRange(Float) (Variables|Default|GetRotMoveAcc) -3.5e+02 -0.2))
          (Utilities|FlowControl|DoOnce false
            (:Completed
              (Components|Activation|Activate (CallFunction|GetTrailVFXComponent) false)
              (Components|Activation|Deactivate (CallFunction|GetTrailVFX2Component))
              (CallFunction|ResetActivateR)))
          (else
            (Components|Activation|Activate (CallFunction|GetTrailVFXComponent) false)
            (Effects|Components|ParticleSystem|SetVectorParameter _returnvalue_3 "SizeByDeactivate" "1.000000,1.000000,1.000000")
            (Effects|Components|ParticleSystem|SetFloatParameter _returnvalue_3 "ErodeByBP" 1.0)
            (Effects|Components|ParticleSystem|SetFloatParameter _returnvalue_3 "LifeTime" 0.18)
            (Utilities|FlowControl|DoOnce false
              (:Completed
                (Timeline|Timeline_0 PlayFromStart)))))))
    (Effects|Components|ParticleSystem|SetVectorParameter _returnvalue_4 "SizeByDeactivate" "1.000000,1.000000,1.000000")
    (Effects|Components|ParticleSystem|SetFloatParameter _returnvalue_4 "ErodeByBP" 1.0)
    (Effects|Components|ParticleSystem|SetFloatParameter _returnvalue_3 "LifeTime" 0.18)
    (Timeline|Timeline_0 PlayFromStart)
    (else
      (if (CallFunction|IsGliderFullyDeployedAccordingtoAnim)
        (Utilities|FlowControl|DoOnce false
          (:Completed
            (bind (_direction _size _velocity) (Timeline|Timeline_1 PlayFromStart)
              (:Update
                (Effects|Components|ParticleSystem|SetVectorParameter _returnvalue_2 "SizeByDeactivate" (Math|Conversions|ToVector(Float) _size))
                (Effects|Components|ParticleSystem|SetVectorParameter _returnvalue_2 "Velocity" (Math|Vector|Lerp(Vector) "500.000000, 0, 0" "0.000000, 0,-800.000000" _velocity))
                (Effects|Components|ParticleSystem|SetFloatParameter _returnvalue_2 "ErodeByBP" (Math|Float|Lerp 1.0 1e+01 _velocity))
                (Effects|Components|ParticleSystem|SetFloatParameter _returnvalue_2 "LifeTime" (Math|Float|Lerp 0.18 0.5 _velocity)))
              (:Finished
                (Components|Activation|Deactivate (CallFunction|GetTrailVFXComponent))
                (Utilities|FlowControl|Delay 0.18)
                (CallFunction|ResetActivate)
                (CallFunction|ResetActivateL)
                (CallFunction|ResetActivateR)))))
        (else
          (Components|Activation|Deactivate (CallFunction|GetTrailVFXComponent)))))))

(event Custom|ResetDeactivate
  (bind _returnvalue (CallFunction|GetTrailVFXComponent))
  (Utilities|FlowControl|DoOnce false
    (:Completed
      (Timeline|Timeline_1 PlayFromStart))))

(fn RibbonTrailLocationTweak
  (bind _returnvalue (CallFunction|GetTrailVFXComponent))
  (if (or (or (< (Variables|Default|GetRight%20Dot) -0.2) (> (Variables|Default|GetRight%20Dot) 0.2)) (and (>= (Variables|Default|GetForward%20Dot) 0.3) (or (Math|Float|InRange(Float) (Variables|Default|GetRotMoveAcc) 0.2 3.5e+02) (Math|Float|InRange(Float) (Variables|Default|GetRotMoveAcc) -3.5e+02 -0.2))))
    (Effects|Components|ParticleSystem|SetVectorParameter _returnvalue "TruningLocationOffset" "6.200000,0.000000,0.000000")
    (else
      (Effects|Components|ParticleSystem|SetVectorParameter _returnvalue "TruningLocationOffset" "3.000000, 0, 0"))))

(fn UpdateAudio
  (params (Moving :type Boolean))
  (Class|FortVehicleAudioVoice|SetParam (Variables|LegacyAudio|GetFortVehicleAudioVoice) (select (and Moving (CallFunction|IsGliderFullyDeployedAccordingtoAnim)) 1.0 0.0)))

(fn ConstructionScript
  (|CallParentFunction))
