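; Legacy Farmer glider Blueprint for Athena cosmetics, which drives two blurred propeller meshes on the parachute.
; When the chute becomes visible, it attaches the left and right blur meshes to their propeller sockets and sets their relative transforms.
; When fully deployed, a timeline animates the right blur mesh's Opacity material parameter, which resets to zero when the chute hides.
; The left blur mesh's opacity is never changed here.
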(event EventOnChuteVisibilityChanged
  (params (bParachuteVisibility :type Boolean))
  (|CallParentFunction bParachuteVisibility)
  (if bParachuteVisibility
    (Transformation|AttachComponentToComponent (Variables|Default|GetSM_Propeller_Blur_L) (Variables|Parachute|GetParachuteMesh) "Propeller_l" "KeepRelative" "KeepRelative" "KeepRelative" true)
    (Transformation|SetRelativeTransform (Variables|Default|GetSM_Propeller_Blur_L) (struct :Location "-1.000000,0.000000,0.000000" :Rotation "0.000000,180.000000,0.000000" :Scale "1.200000,0.460000,0.460000") false false)
    (Transformation|AttachComponentToComponent (Variables|Default|GetSM_Propeller_Blur_R) (Variables|Parachute|GetParachuteMesh) "Propeller_r" "KeepRelative" "KeepRelative" "KeepRelative" true)
    (Transformation|SetRelativeTransform (Variables|Default|GetSM_Propeller_Blur_R) (struct :Location "-2.000000,-0.500000,-0.500000" :Rotation "0, 0, 0" :Scale "1.200000,0.460000,0.460000") false false)
    (else
      (Rendering|Material|SetScalarParameterValueonMaterials (Variables|Default|GetSM_Propeller_Blur_R) "Opacity" 0.0))))

(event EventOnGliderFullyDeployed
  (|CallParentFunction)
  (bind (_direction _newtrack_0) (Timeline|Timeline_0 PlayFromStart)
    (:Update
      (Rendering|Material|SetScalarParameterValueonMaterials (Variables|Default|GetSM_Propeller_Blur_R) "Opacity" _newtrack_0))))

(fn ConstructionScript
  (|CallParentFunction))
