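; Legacy DJ glider Blueprint that plays music with reactive lights and sound-wave particles while deployed.
; An Intensity timeline animates the dynamic material, while a BPM-scaled timeline cycles left and right particle emitters selected by steering direction.
; Forward movement sets music volume and an interpolated low-pass filter, and deployment registers the music with MusicEventSubsystem.
; Frontend preview skips the filter, loops the effects and resets them when the chute hides.
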(event EventBeginPlay
  (bind _mid (Variables|Default|GetMID))
  ;[0 0 0] initial
  (Utilities|FlowControl|Delay 0.1)
  (Utilities|FlowControl|Sequence
    (:then_0
      (Utilities|FlowControl|DoOnce false
        (:Completed
          (bind _returnvalue (Rendering|Material|CreateDynamicMaterialInstance (Variables|Parachute|GetParachuteMesh) 0 (Rendering|Material|GetMaterial (Variables|Parachute|GetParachuteMesh) 0) "None" :!class /Script/Engine.PrimitiveComponent))
          (bind _output_get (Variables|Default|SetMID _returnvalue))
          (Rendering|Material|SetScalarParameterValue _output_get "HDRIntensity" 0.0 :!class /Script/Engine.MaterialInstanceDynamic)
          (Rendering|Material|SetScalarParameterValue _output_get "Light Frequency" 1e+01 :!class /Script/Engine.MaterialInstanceDynamic)
          (Components|Timeline|SetPlayRate (Variables|B_Athena_Glider_Legacy_FortniteDJ|GetTimeline_0) (* (/ 1.0 (/ 6e+01 (Variables|Default|GetMusicBPM))) 0.25)))))
    (:then_1
      (Utilities|FlowControl|Delay 1.6)
      (if (Variables|FortPlayerParachute|GetbIsFrontEndPreview)
        (Effects|Components|ParticleSystem|SetVectorParameter (Variables|Default|GetP_Light) "FE_Offset" "1.000000,1.000000,1.000000")
        (Utilities|FlowControl|Delay 0.5)
        (Utilities|FlowControl|Sequence
          (:then_0
            (CallFunction|CEFrontEffect true :!label _label)
            (Utilities|FlowControl|Delay 0.5)
            (goto _label))
          (:then_1
            _))
        (else
          (Components|Activation|Activate (Variables|Default|GetP_Glider_Sound_R1) true)))
      (bind (_direction _newtrack_0 _light_frequency) (Timeline|Intensity PlayFromStart)
        (:Update
          (Rendering|Material|SetScalarParameterValue _mid "HDRIntensity" _newtrack_0 :!class /Script/Engine.MaterialInstanceDynamic)
          (Rendering|Material|SetScalarParameterValue _mid "Light Frequency" _light_frequency :!class /Script/Engine.MaterialInstanceDynamic))))))

(event EventOnChuteVisibilityChanged
  (params (bParachuteVisibility :type Boolean))
  (bind _mid (Variables|Default|GetMID))
  ;[1 1 1] Fix for effects not showing up properly when redeploying more than once in frontend.
  (if (and (Variables|FortPlayerParachute|GetbIsFrontEndPreview) (== bParachuteVisibility false))
    (Components|Activation|Deactivate (Variables|Default|GetP_Glider_Sound_L2))
    (Rendering|Material|SetScalarParameterValue _mid "HDRIntensity" 0.0 :!class /Script/Engine.MaterialInstanceDynamic)
    (Rendering|Material|SetScalarParameterValue _mid "Light Frequency" 1e+01 :!class /Script/Engine.MaterialInstanceDynamic)
    (Utilities|FlowControl|Delay 0.5)
    (CallFunction|Rest)))

(event Parachute|EventOnParachuteTrailUpdated
  (params (MovementLengthSquared :type Float%20%28single-precision%29) (ForwardDot :type Float%20%28single-precision%29) (RightDot :type Float%20%28single-precision%29) (RotationalMovementAcceleration :type Float%20%28single-precision%29))
  (bind _returnvalue (CallFunction|IsGliderFullyDeployedAccordingtoAnim))
  (bind _bisfrontendpreview (Variables|FortPlayerParachute|GetbIsFrontEndPreview))
  (bind _returnvalue_2 (or _bisfrontendpreview _returnvalue))
  (bind _mid (Variables|Default|GetMID))
  (Utilities|FlowControl|Sequence
    (:then_0
      (if _returnvalue_2
        (Utilities|FlowControl|Sequence
          (:then_0
            (Utilities|FlowControl|DoOnce false :!label _label
              (:Completed
                (Components|Activation|Deactivate (Variables|Default|GetP_Glider_Sound_L2))
                (Rendering|Material|SetScalarParameterValue _mid "HDRIntensity" 0.0 :!class /Script/Engine.MaterialInstanceDynamic)
                (Rendering|Material|SetScalarParameterValue _mid "Light Frequency" 1e+01 :!class /Script/Engine.MaterialInstanceDynamic)
                (Utilities|FlowControl|Delay 0.5)
                (CallFunction|Rest))))
          (:then_1
            _))
        (else
          (Utilities|FlowControl|Sequence
            (:then_0
              _)
            (:then_1
              (goto _label)))))
      (Utilities|FlowControl|DoOnce false
        (:Completed
          (Components|Activation|Activate (Variables|Default|GetP_Glider_Sound_R1) true)
          (Timeline|Intensity PlayFromStart))))
    (:then_1
      (if _returnvalue_2
        (if (Math|Float|InRange(Float) RightDot -0.01 0.01)
          (CallFunction|CELeftMoving 0)
          (CallFunction|CEFrontEffect true)
          (elif (> RightDot 0.0)
            (CallFunction|CELeftMoving 1)
            (else
              (CallFunction|CELeftMoving 2))))
        (CallFunction|CEFrontEffect true)
        (else
          (CallFunction|CELeftMoving 0)
          (CallFunction|CEFrontEffect false))))
    (:then_2
      (CallFunction|AudioUpdate ForwardDot))))

(event Custom|Rest
  (bind _mid (Variables|Default|GetMID))
  (Utilities|FlowControl|DoOnce false
    (:Completed
      (Components|Activation|Activate (Variables|Default|GetP_Glider_Sound_R1) true)
      (Timeline|Intensity PlayFromStart))))

(event Custom|CE_LeftMoving
  (params (NewParam :type Integer))
  ;[0 0 0] FX update
  (if (not (== NewParam (Variables|Default|GetLeftSoundFXState)))
    (Variables|Default|SetLeftSoundFXState NewParam)))

(event Custom|CE_FrontEffect
  (params (Play :type Boolean))
  ;[0 0 0] FX update
  (if Play
    (Utilities|FlowControl|Sequence
      (:then_0
        (Utilities|FlowControl|DoOnce false :!label _label
          (:Completed
            (Timeline|Timeline_0 Stop
              (:VFX
                (if (CallFunction|IsGliderFullyDeployedAccordingtoAnim)
                  (bind _output_get (Variables|Default|SetLoop2 (Math|Integer|%25%28Integer%29 (+ (Variables|Default|GetLoop2) 1) 3)))
                  (Utilities|FlowControl|Sequence
                    (:then_0
                      (if (Utilities|Select :Index (Variables|Default|GetLeftSoundFXState) :Option%200 true :Option%201 false :Option%202 true)
                        (Components|Activation|Activate (Utilities|Array|Get(acopy) (Utilities|Array|MakeArray (Variables|Default|GetP_Glider_Sound_L3) (Variables|Default|GetP_Glider_Sound_L2) (Variables|Default|GetP_Glider_Sound_L1)) _output_get) true)))
                    (:then_1
                      (if (Utilities|Select :Index (Variables|Default|GetLeftSoundFXState) :Option%200 true :Option%201 true :Option%202 false)
                        (Components|Activation|Activate (Utilities|Array|Get(acopy) (Utilities|Array|MakeArray (Variables|Default|GetP_Glider_Sound_R3) (Variables|Default|GetP_Glider_Sound_R2) (Variables|Default|GetP_Glider_Sound_R1)) _output_get) true))))))))))
      (:then_1
        _))
    (else
      (Utilities|FlowControl|Sequence
        (:then_0
          _)
        (:then_1
          (goto _label)))))
  (Utilities|FlowControl|DoOnce false
    (:Completed
      (Variables|Default|SetLoop2 -1)
      (Timeline|Timeline_0 PlayFromStart))))

(event Custom|Audio%20Update
  (params (Forward%20Dot :type Float%20%28double-precision%29))
  ;[1 1 1] Audio Logic
  (Utilities|FlowControl|Sequence
    (:then_0
      (Default|AudioUpdateMusic Forward%20Dot 0.6 1.0))
    (:then_1
      ;;[0.6 0.4 0.8] In game we want to adjust lowpass
      (Variables|FortPlayerParachute|GetbIsFrontEndPreview
        (:then
          ;;[0.6 0.4 0.8] No LPF in Front End
          (Audio|Components|Audio|SetLowPassFilterEnabled (Variables|Default|GetMusic) false :!class /Script/Engine.AudioComponent))
        (:else
          (Variables|/Script/Engine.AudioComponent|GetbEnableLowPassFilter (Variables|Default|GetMusic)
            (:then
              _)
            (:else
              (Audio|Components|Audio|SetLowPassFilterEnabled (Variables|Default|GetMusic) true :!class /Script/Engine.AudioComponent)))
          (bind _output_get (Variables|Default|SetLPFFreq (Math|Interpolation|FInterpTo (Variables|Default|GetLPFFreq) (select (>= Forward%20Dot 0.2) 2e+04 1e+03) (Utilities|Time|GetWorldDeltaSeconds) 4.0)))
          (Audio|Components|Audio|SetLowPassFilterFrequency (Variables|Default|GetMusic) _output_get :!class /Script/Engine.AudioComponent))))))

(event EventOnGliderFullyDeployed
  ;[1 1 1] Audio Logic
  (|CallParentFunction)
  (Audio|Components|Audio|FadeIn (Variables|Default|GetMusic) 0.25 1.0 0.0 "Linear" :!class /Script/Engine.AudioComponent)
  (Audio|AddEventbyComponent (GameInstanceSubsystems|GetMusicEventSubsystem) (Variables|Default|GetMusic) "(TagName=\"MusicEventSubsystem.Event.Glider\")" "(GameplayTags=)" 0 ""))

(event EventCheckForMusicEvent
  (params (GliderFullyDeployed :type Boolean))
  ;[1 1 1] Audio Logic
  (if (not GliderFullyDeployed)
    (Audio|RemoveEvent (GameInstanceSubsystems|GetMusicEventSubsystem) (Variables|Default|GetMusic))))

(fn AudioUpdateMusic
  (params (ForwardDot :type Float%20%28double-precision%29) (MusicVolMin :type Float%20%28double-precision%29) (MusicVolMax :type Float%20%28double-precision%29))
  (bind _music (Variables|Default|GetMusic))
  (Variables|/Script/Engine.AudioComponent|GetSound (Variables|Default|GetMusic)
    (:then
      (if (CallFunction|IsGliderFullyDeployedAccordingtoAnim)
        (Class|FortVehicleAudioVoice|SetParam _music (select (>= (select (Variables|LegacyAudio|GetAbsolute%20Movement_Audio) (Math|Float|Absolute(Float) ForwardDot) ForwardDot) 0.2) MusicVolMax MusicVolMin))
        (else
          (Class|FortVehicleAudioVoice|SetParam _music 0.0))))
    (:else)))

(fn ConstructionScript
  (|CallParentFunction))
