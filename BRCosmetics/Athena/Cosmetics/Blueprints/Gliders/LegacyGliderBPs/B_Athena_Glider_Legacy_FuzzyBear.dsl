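; Fuzzy Bear glider cosmetic Blueprint for Athena that drives the glider's particle effects and emissive glow.
; Full deployment starts the idle effect with a medium glow, and hiding the chute stops that effect and zeroes the glow.
; Trail updates use forward movement and deployment state to switch between brighter engine trails and the dimmer idle look.
; DoOnce gates keep both switches from re-triggering on every update.
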(event EventOnGliderFullyDeployed
  (|CallParentFunction)
  (Components|Activation|Activate (Variables|Default|GetP_Fuzzy_Bear_Glider_Idle_Left) true)
  (Rendering|Material|SetVectorParameterValueonMaterials (Variables|Parachute|GetParachuteMesh) "EmissiveMultiplier" "0.650000,0.650000,0.650000"))

(event EventOnChuteVisibilityChanged
  (params (bParachuteVisibility :type Boolean))
  (|CallParentFunction bParachuteVisibility)
  (if (not bParachuteVisibility)
    (Components|Activation|Deactivate (Variables|Default|GetP_Fuzzy_Bear_Glider_Idle_Left))
    (Rendering|Material|SetVectorParameterValueonMaterials (Variables|Parachute|GetParachuteMesh) "EmissiveMultiplier" "0, 0, 0")))

(event Parachute|EventOnParachuteTrailUpdated
  (params (MovementLengthSquared :type Float%20%28single-precision%29) (ForwardDot :type Float%20%28single-precision%29) (RightDot :type Float%20%28single-precision%29) (RotationalMovementAcceleration :type Float%20%28single-precision%29))
  (|CallParentFunction MovementLengthSquared ForwardDot RightDot RotationalMovementAcceleration)
  (if (and (>= ForwardDot 0.2) (CallFunction|IsGliderFullyDeployedAccordingtoAnim))
    (Utilities|FlowControl|DoOnce false
      (:Completed
        (Components|Activation|Deactivate (Variables|Default|GetP_Fuzzy_Bear_Glider_Idle_Left))
        (Components|Activation|Activate (Variables|Default|GetP_Fuzzy_Bear_Glider_Trail_Engines_Left) true)
        (Rendering|Material|SetVectorParameterValueonMaterials (Variables|Parachute|GetParachuteMesh) "EmissiveMultiplier" "1.500000,1.500000,1.500000")))
    (else
      _))
  (Utilities|FlowControl|DoOnce false
    (:Completed
      (Components|Activation|Deactivate (Variables|Default|GetP_Fuzzy_Bear_Glider_Trail_Engines_Left))
      (Components|Activation|Activate (Variables|Default|GetP_Fuzzy_Bear_Glider_Idle_Left) true)
      (Rendering|Material|SetVectorParameterValueonMaterials (Variables|Parachute|GetParachuteMesh) "EmissiveMultiplier" "0.650000,0.650000,0.650000"))))

(fn ConstructionScript
  (|CallParentFunction))
