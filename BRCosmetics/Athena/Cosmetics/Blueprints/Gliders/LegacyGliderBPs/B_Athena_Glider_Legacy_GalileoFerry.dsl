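; Legacy Athena glider Blueprint for the GalileoFerry cosmetic, adding deploy trail effects and movement-reactive audio.
; When the chute becomes visible, it activates the deploy trails and a timeline drives their ribbon SpawnRate.
; While the animation reports full deployment, audio updates set a vehicle voice parameter high for forward or sideways movement, low otherwise, and feed TurnAmount from absolute right dot.
; When not fully deployed, audio updates stop all voice audio.
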(event EventOnChuteVisibilityChanged
  (params (bParachuteVisibility :type Boolean))
  (|CallParentFunction bParachuteVisibility)
  (if bParachuteVisibility
    (Components|Activation|Activate (Variables|Default|GetDeployTrails) false)
    (bind (_direction _spawn_rate_for_ribbons) (Timeline|Timeline_0 PlayFromStart)
      (:Update
        (Effects|Components|ParticleSystem|SetFloatParameter (Variables|Default|GetDeployTrails) "SpawnRate" _spawn_rate_for_ribbons)))))

(event EventAudioUpdated
  (params (Forward%20Dot :type Float%20%28double-precision%29) (Right%20Dot :type Float%20%28double-precision%29))
  (bind _fortvehicleaudiovoice (Variables|LegacyAudio|GetFortVehicleAudioVoice))
  (if (CallFunction|IsGliderFullyDeployedAccordingtoAnim)
    (Class|FortVehicleAudioVoice|SetParam _fortvehicleaudiovoice (select (Math|Boolean|ORBoolean (>= Forward%20Dot 0.2) (>= Right%20Dot 0.2) (<= Right%20Dot -0.2)) 1.0 0.25))
    (Class|FortVehicleAudioVoice|SetParamExplicit _fortvehicleaudiovoice "TurnAmount" (Math|Float|Absolute(Float) Right%20Dot))
    (else
      (Class|FortVehicleAudioVoice|StopAllAudio (Variables|LegacyAudio|GetFortVehicleAudioVoice)))))

(fn ConstructionScript
  (|CallParentFunction))
