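; Legacy glider cosmetic Blueprint for GalileoKayak, driving its airbrake effect, wing trails and flight audio.
; The Deploy_Airbrake component activates on full deployment and deactivates when the chute hides.
; Trail updates begin left and right socket trails on forward or sideways movement, otherwise a timeline fades and ends only the first trail component, both branches gated by DoOnce.
; They also feed forward and turn amounts to the vehicle audio voice, which stops unless fully deployed.
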(event EventOnChuteVisibilityChanged
  (params (bParachuteVisibility :type Boolean))
  (|CallParentFunction bParachuteVisibility)
  (if (not bParachuteVisibility)
    (Components|Activation|Deactivate (Variables|Default|GetDeploy_Airbrake))))

(event EventOnGliderFullyDeployed
  (|CallParentFunction)
  (Components|Activation|Activate (Variables|Default|GetDeploy_Airbrake) false))

(event Parachute|EventOnParachuteTrailUpdated
  (params (MovementLengthSquared :type Float%20%28single-precision%29) (ForwardDot :type Float%20%28single-precision%29) (RightDot :type Float%20%28single-precision%29) (RotationalMovementAcceleration :type Float%20%28single-precision%29))
  (bind _returnvalue (CallFunction|GetTrailVFXComponent))
  (CallFunction|AudioUpdated ForwardDot RightDot)
  (if (and (or (>= ForwardDot 0.2) (> (Math|Float|Absolute(Float) RightDot) 0.0)) (CallFunction|IsGliderFullyDeployedAccordingtoAnim))
    (Utilities|FlowControl|DoOnce false
      (:Completed
        (Components|Activation|Activate (CallFunction|GetTrailVFXComponent) false)
        (Effects|Components|ParticleSystem|SetFloatParameter (CallFunction|GetTrailVFX2Component) "TrailAlpha" 1.0)
        (Effects|Particles|Trails|BeginTrails (Utilities|Casting|CastToParticleSystemComponent (CallFunction|GetTrailVFXComponent)) "FX_Wing_L_01" "FX_Wing_L_02" "ETrailWidthMode_FromCentre" 1.0)
        (Effects|Particles|Trails|BeginTrails (Utilities|Casting|CastToParticleSystemComponent (CallFunction|GetTrailVFX2Component)) "FX_Wing_R_01" "FX_Wing_R_02" "ETrailWidthMode_FromCentre" 1.0)))
    (else
      (Utilities|FlowControl|DoOnce false
        (:Completed
          (bind (_direction _lerp_control) (Timeline|Timeline_0 PlayFromStart)
            (:Update
              (Effects|Components|ParticleSystem|SetFloatParameter (CallFunction|GetTrailVFXComponent) "TrailAlpha" _lerp_control))
            (:Finished
              (Components|Activation|Deactivate _returnvalue)
              (Effects|Particles|Trails|EndTrails (Utilities|Casting|CastToParticleSystemComponent _returnvalue)))))))))

(event EventAudioUpdated
  (params (Forward%20Dot :type Float%20%28double-precision%29) (Right%20Dot :type Float%20%28double-precision%29))
  (bind _fortvehicleaudiovoice (Variables|LegacyAudio|GetFortVehicleAudioVoice))
  (if (CallFunction|IsGliderFullyDeployedAccordingtoAnim)
    (Class|FortVehicleAudioVoice|SetParam _fortvehicleaudiovoice (select (>= Forward%20Dot 0.2) 2.0 0.5))
    (Class|FortVehicleAudioVoice|SetParamExplicit _fortvehicleaudiovoice "TurnAmount" (Math|Float|Absolute(Float) Right%20Dot))
    (else
      (Class|FortVehicleAudioVoice|StopAllAudio (Variables|LegacyAudio|GetFortVehicleAudioVoice)))))

(fn ConstructionScript
  (|CallParentFunction))
