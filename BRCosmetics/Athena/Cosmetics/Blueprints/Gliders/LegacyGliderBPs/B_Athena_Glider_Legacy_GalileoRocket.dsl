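; Legacy glider Blueprint for the Galileo Rocket cosmetic, driving its deploy effects and movement-reactive engine glow and trails.
; On full deploy it spawns a front-end preview burst or adds a local-player camera lens effect, then after a delay a timeline eases down emissive, flare, light and spawn rate.
; Once the timeline finishes, trail updates scale engine glow and flares with forward movement and set trail spawning from sideways movement.
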(event EventOnChuteVisibilityChanged
  (params (bParachuteVisibility :type Boolean))
  (|CallParentFunction bParachuteVisibility)
  (if bParachuteVisibility
    (Transformation|AttachComponentToComponent (CallFunction|GetTrailVFX2Component) (Variables|Parachute|GetParachuteMesh) "Body_Offset" "KeepRelative" "KeepRelative" "KeepRelative" true)))

(event EventOnGliderFullyDeployed
  (bind _returnvalue (CallFunction|GetTrailVFX2Component))
  (bind _playerpawn (Variables|FortPlayerParachute|GetPlayerPawn))
  (if (Variables|FortPlayerParachute|GetbIsFrontEndPreview)
    (Effects|SpawnEmitteratLocation "/BRCosmetics/Gadgets/Assets/VinderTech_GliderChute/Glider_Galileo_Rocket/FX/P_Galileo_Rocket_Deploy_FE.P_Galileo_Rocket_Deploy_FE" (.location (Transformation|GetSocketTransform (Variables|Parachute|GetParachuteMesh) "None" "RTS_World")) (Math|Rotator|CombineRotators (.rotation (Transformation|GetSocketTransform (Variables|Parachute|GetParachuteMesh) "None" "RTS_World")) "0,90.000000, 0"))
    (else
      (Variables|Default|SetHypersapce%20Deployed false)
      (if (Pawn|IsLocallyControlled _playerpawn)
        (LensEffect|IsValidCameraLens (Variables|Default|GetCameraLensInterfaceFX)
          (:Invalid
            (bind _returnvalue_2 (Camera|AddGenericCameraLensEffect (Variables|/Script/Engine.PlayerController|GetPlayerCameraManager (Fortnite|GetFortPlayerControllerfromActor _playerpawn)) "/BRCosmetics/Gadgets/Assets/VinderTech_GliderChute/Glider_Galileo_Rocket/FX/BP_CameraLens_Galileo_Deploy.BP_CameraLens_Galileo_Deploy_C"))
            (Variables|Default|SetCameraLensInterfaceFX _returnvalue_2)))
        (else
          _))))
  (Utilities|FlowControl|Delay (Variables|Default|GetHOLD!))
  (Components|Activation|Activate (CallFunction|GetTrailVFXComponent) true)
  (bind (_direction _lerp_control _spawn_rate) (Timeline|Timeline_0 PlayFromStart)
    (:Update
      (Rendering|Material|SetScalarParameterValueonMaterials (Variables|Parachute|GetParachuteMesh) "PG_GlowStrength" (Math|Float|Lerp (Variables|Default|GetMax%20Engine%20Emissive_Deploy) (Variables|Default|GetMin%20Engine%20Emissive) _lerp_control))
      (Effects|Components|ParticleSystem|SetFloatParameter _returnvalue "FireLineControl" (Math|Float|Lerp 0.5 1.0 _lerp_control))
      (Effects|Components|ParticleSystem|SetFloatParameter _returnvalue "WPOFlare" (Math|Float|Lerp 5e+02 1e+01 _lerp_control))
      (Effects|Components|ParticleSystem|SetFloatParameter _returnvalue "WPO_SmallFlare" (Math|Float|Lerp 3.5e+02 5.0 _lerp_control))
      (Effects|Components|ParticleSystem|SetVectorParameter _returnvalue "LightIntensity" (Math|Vector|Lerp(Vector) "1.000000,1.000000,1.000000" "0.100000,0.100000,0.100000" _lerp_control))
      (Effects|Components|ParticleSystem|SetFloatParameter (CallFunction|GetTrailVFXComponent) "SpawnRate" (Math|Float|Lerp 3e+01 0.0 _spawn_rate)))
    (:Finished
      (Variables|Default|SetHypersapce%20Deployed true)
      (if (not (Variables|FortPlayerParachute|GetbIsFrontEndPreview))
        (Camera|RemoveGenericCameraLensEffect (Variables|/Script/Engine.PlayerController|GetPlayerCameraManager (Fortnite|GetFortPlayerControllerfromActor (Variables|FortPlayerParachute|GetPlayerPawn))) (Variables|Default|GetCameraLensInterfaceFX))))))

(event Parachute|EventOnParachuteTrailUpdated
  (params (MovementLengthSquared :type Float%20%28single-precision%29) (ForwardDot :type Float%20%28single-precision%29) (RightDot :type Float%20%28single-precision%29) (RotationalMovementAcceleration :type Float%20%28single-precision%29))
  (bind _returnvalue (< ForwardDot 0.0))
  (bind _returnvalue_2 (Math|Float|Absolute(Float) ForwardDot))
  (bind _returnvalue_3 (select _returnvalue 0.0 _returnvalue_2))
  (bind _returnvalue_4 (Math|Float|Lerp 0.0 1.0 _returnvalue_3))
  (bind _returnvalue_5 (Math|Float|Absolute(Float) RightDot))
  (bind _returnvalue_6 (>= _returnvalue_5 0.2))
  (bind _returnvalue_7 (CallFunction|GetTrailVFX2Component))
  (|CallParentFunction MovementLengthSquared ForwardDot RightDot RotationalMovementAcceleration)
  (if (Variables|Default|GetHypersapce%20Deployed)
    (Rendering|Material|SetScalarParameterValueonMaterials (Variables|Parachute|GetParachuteMesh) "PG_GlowStrength" (Math|Float|Lerp (Variables|Default|GetMin%20Engine%20Emissive) (Variables|Default|GetMax%20Engine%20Emissive) _returnvalue_4))
    (Effects|Components|ParticleSystem|SetFloatParameter _returnvalue_7 "WPOFlare" (Math|Float|Lerp 1e+01 5e+01 _returnvalue_4))
    (Effects|Components|ParticleSystem|SetFloatParameter _returnvalue_7 "WPO_SmallFlare" (Math|Float|Lerp 5.0 2e+01 _returnvalue_4))
    (Effects|Components|ParticleSystem|SetVectorParameter _returnvalue_7 "LightIntensity" (Math|Vector|Lerp(Vector) "0.100000,0.100000,0.100000" "0.500000,0.500000,0.500000" _returnvalue_4))
    (if (not _returnvalue_6)
      (Utilities|FlowControl|Sequence
        (:then_0
          (Utilities|FlowControl|DoOnce false
            (:Completed
              _)))
        (:then_1
          _))
      (Effects|Components|ParticleSystem|SetFloatParameter (CallFunction|GetTrailVFXComponent) "SpawnRate" (select _returnvalue_6 3e+01 0.0)))))

(fn ConstructionScript
  (|CallParentFunction))
