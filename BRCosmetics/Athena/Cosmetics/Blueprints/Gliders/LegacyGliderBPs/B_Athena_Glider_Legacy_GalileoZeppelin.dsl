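; Legacy glider Blueprint for the Galileo Zeppelin cosmetic, driving an engine flame effect on the parent glider's trail VFX component.
; When the chute becomes visible, that component attaches to the parachute mesh's FX_Engine_L socket.
; Full deployment activates it and plays the EngienFlare timeline, feeding FlameControl float and vector particle parameters.
; Once that flare finishes, trail updates set the same parameters from forward movement dot while the animation reports fully deployed.
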(event EventOnChuteVisibilityChanged
  (params (bParachuteVisibility :type Boolean))
  (|CallParentFunction bParachuteVisibility)
  (if bParachuteVisibility
    (Transformation|AttachComponentToComponent (CallFunction|GetTrailVFXComponent) (Variables|Parachute|GetParachuteMesh) "FX_Engine_L" "KeepRelative" "KeepRelative" "KeepRelative" true)))

(event EventOnGliderFullyDeployed
  (bind _returnvalue (CallFunction|GetTrailVFXComponent))
  (|CallParentFunction)
  (Components|Activation|Activate (CallFunction|GetTrailVFXComponent) false)
  (bind (_direction _flareup) (Timeline|EngienFlare PlayFromStart)
    (:Update
      (Effects|Components|ParticleSystem|SetFloatParameter _returnvalue "FlameControl" _flareup)
      (Effects|Components|ParticleSystem|SetVectorParameter _returnvalue "FlameControl_Vector" (Math|Conversions|ToVector(Float) _flareup)))))

(event Parachute|EventOnParachuteTrailUpdated
  (params (MovementLengthSquared :type Float%20%28single-precision%29) (ForwardDot :type Float%20%28single-precision%29) (RightDot :type Float%20%28single-precision%29) (RotationalMovementAcceleration :type Float%20%28single-precision%29))
  (bind _returnvalue (CallFunction|GetTrailVFXComponent))
  (|CallParentFunction MovementLengthSquared ForwardDot RightDot RotationalMovementAcceleration)
  (if (and (not (Components|Timeline|IsPlaying (Variables|B_Athena_Glider_Legacy_GalileoZeppelin|GetEngienFlare))) (CallFunction|IsGliderFullyDeployedAccordingtoAnim))
    (Effects|Components|ParticleSystem|SetFloatParameter _returnvalue "FlameControl" ForwardDot)
    (Effects|Components|ParticleSystem|SetVectorParameter _returnvalue "FlameControl_Vector" (Math|Conversions|ToVector(Float) ForwardDot))))

(fn ConstructionScript
  (|CallParentFunction))
