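; GarageBand glider Blueprint that gives the parachute vehicle-style lights, jet and smoke effects, and reactive music.
; Trail updates turn movement dots and rotational acceleration into particle toggles, headlight visibility, and brake, reversing and blinker states on a dynamic mesh material, animated by timelines.
; Forward movement also drives the audio voice parameter and a low-pass filter sweep.
; Deployment fades in audio and registers it as a music event; front-end preview gets simpler jet and fade handling.
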(event Custom|ResetBreakLightFading
  ;[1 0.247 0.2072] Brake Lights
  (Utilities|FlowControl|DoOnce false
    (:Completed
      (bind (_direction _fade) (Timeline|Timeline_BreakLightFading PlayFromStart)
        (:Update
          (Rendering|Material|SetScalarParameterValue (Variables|Default|GetMID) "Emissive_BrightnessMax" (Math|Float|Lerp 0.0 1.0 _fade) :!class /Script/Engine.MaterialInstanceDynamic)
          (CallFunction|ResetBreakLightup))))))

(event Parachute|EventOnParachuteTrailUpdated
  (params (MovementLengthSquared :type Float%20%28single-precision%29) (ForwardDot :type Float%20%28single-precision%29) (RightDot :type Float%20%28single-precision%29) (RotationalMovementAcceleration :type Float%20%28single-precision%29))
  (bind _returnvalue (CallFunction|IsGliderFullyDeployedAccordingtoAnim))
  (bind _returnvalue_2 (> RightDot 0.2))
  (bind _returnvalue_3 (< RightDot -0.2))
  (bind _returnvalue_4 (Math|Float|InRange(Float) RotationalMovementAcceleration -3.5e+02 -0.5))
  (bind _returnvalue_5 (> RotationalMovementAcceleration 351.0))
  (bind _returnvalue_6 (< RotationalMovementAcceleration -351.0))
  (bind _returnvalue_7 (Math|Float|InRange(Float) RotationalMovementAcceleration 0.5 3.5e+02))
  (bind _returnvalue_8 (or _returnvalue_7 _returnvalue_6))
  (bind _returnvalue_9 (or _returnvalue_3 _returnvalue_8))
  (bind _returnvalue_10 (or _returnvalue_4 _returnvalue_5))
  (bind _returnvalue_11 (or _returnvalue_2 _returnvalue_10))
  (bind _headlightray_l (Variables|Default|GetHeadLightRay_L))
  (bind _mid (Variables|Default|GetMID))
  (bind _headlightray (Variables|Default|GetHeadLightRay))
  (bind _p_jet_idle (Variables|Default|GetP_Jet_Idle))
  (bind _p_jet_frontend (Variables|Default|GetP_Jet_FrontEnd))
  (bind _mid_2 (Variables|Default|GetMID))
  (bind _mid_3 (Variables|Default|GetMID))
  (bind _p_mainjet (Variables|Default|GetP_MainJet))
  (bind _p_smoketrails (Variables|Default|GetP_SmokeTrails))
  ;[1 1 1] Setting for FrontEnd
  (if (Variables|FortPlayerParachute|GetbIsFrontEndPreview)
    (Utilities|FlowControl|Delay 2.7)
    (Rendering|SetVisibility _p_jet_frontend true false)
    (Components|Activation|Activate _p_jet_frontend false)
    (else
      (Utilities|FlowControl|Sequence
        (:then_0
          ;[1.5 1.212 0.6458] HeadlightsRay
          (if _returnvalue
            (Transformation|SetRelativeScale3D _headlightray "1.000000,1.000000,1.000000")
            (Rendering|SetVisibility _headlightray_l true false)
            (else
              (Transformation|SetRelativeScale3D _headlightray "0, 0, 0")
              (Rendering|SetVisibility _headlightray_l false false))))
        (:then_1
          (if _returnvalue
            (Components|Activation|Activate (Variables|Default|GetP_AirFlowWheels_bk_L) false)
            (Utilities|FlowControl|Sequence
              (:then_0
                ;[0.5516 0.7703 1] Jet FX
                (if (>= ForwardDot 0.2)
                  (Components|Activation|Activate _p_mainjet false)
                  (Components|Activation|Deactivate _p_jet_idle)
                  (else
                    (Components|Activation|Deactivate _p_mainjet)
                    (Components|Activation|Activate _p_jet_idle false))))
              (:then_1
                ;[0.5165 1 0.3972] ReducetheCheckRate - Optimization
                (Utilities|FlowControl|DoOnce false
                  (:Completed
                    (Utilities|FlowControl|Sequence
                      (:then_0
                        (Utilities|FlowControl|Delay 0.3))
                      (:then_1
                        _)
                      (:then_2
                        (Development|PrintString (Utilities|String|ToString(Float) RotationalMovementAcceleration))))
                    (Utilities|FlowControl|Sequence
                      (:then_0
                        ;[1 0.247 0.2072] Brake Lights
                        (if (Math|Boolean|ORBoolean _returnvalue_9 _returnvalue_11 (!= ForwardDot 0.0))
                          (Utilities|FlowControl|DoOnce false
                            (:Completed
                              (Timeline|Timeline_BreakLightFading PlayFromStart)))
                          (else
                            (Utilities|FlowControl|DoOnce false
                              (:Completed
                                (Rendering|Material|SetScalarParameterValue _mid_2 "Emissive_BrightnessMax" 3.0 :!class /Script/Engine.MaterialInstanceDynamic)
                                (CallFunction|ResetBreakLightFading))))))
                      (:then_1
                        ;[1 0.8389 0.6389] Reversing Lights
                        (if (and (< ForwardDot 0.0) (not (or _returnvalue_3 _returnvalue_2)))
                          (Rendering|Material|SetScalarParameterValue _mid "HeadlightFalloff" 0.2 :!class /Script/Engine.MaterialInstanceDynamic)
                          (else
                            (Rendering|Material|SetScalarParameterValue _mid "HeadlightFalloff" 0.0 :!class /Script/Engine.MaterialInstanceDynamic))))
                      (:then_2
                        ;[1 0.7774 0.3915] Turning Lights
                        (if _returnvalue_9
                          (CallFunction|StopTurnRight)
                          (Rendering|Material|SetScalarParameterValue _mid_3 "isTurningLeft" 0.0 :!class /Script/Engine.MaterialInstanceDynamic) ; Turn Left
                          (Utilities|FlowControl|DoOnce false
                            (:Completed
                              (bind (_direction _blinker) (Timeline|Timeline_TurnLeft PlayFromStart)
                                (:Update
                                  (Rendering|Material|SetScalarParameterValue _mid_3 "isTurning" _blinker :!class /Script/Engine.MaterialInstanceDynamic)))))
                          (elif _returnvalue_11
                            (CallFunction|StopTurnLeft)
                            (Rendering|Material|SetScalarParameterValue _mid_3 "isTurningLeft" 1.0 :!class /Script/Engine.MaterialInstanceDynamic) ; TurnRight
                            (Utilities|FlowControl|DoOnce false
                              (:Completed
                                (bind (_direction_2 _blinker_2) (Timeline|Timeline_TurnRight PlayFromStart)
                                  (:Update
                                    (Rendering|Material|SetScalarParameterValue _mid_3 "isTurning" _blinker_2 :!class /Script/Engine.MaterialInstanceDynamic)))))
                            (else
                              (Rendering|Material|SetScalarParameterValue _mid_3 "isTurningLeft" 0.5 :!class /Script/Engine.MaterialInstanceDynamic)
                              (Rendering|Material|SetScalarParameterValue _mid_3 "isTurning" 0.0 :!class /Script/Engine.MaterialInstanceDynamic)
                              (CallFunction|StopTurnRight)
                              (CallFunction|StopTurnLeft)))))
                      (:then_3
                        ;[0.375 0.375 0.375] Smoke Trails
                        (if (or _returnvalue_8 _returnvalue_10)
                          (Components|Activation|Activate _p_smoketrails false)
                          (else
                            (Components|Activation|Deactivate _p_smoketrails)))))))))
            (else
              (Components|Activation|Deactivate (Variables|Default|GetP_MainJet)))))
        (:then_2
          (CallFunction|AudioUpdate (> ForwardDot 0.2)))))))

(event Custom|ResetBreakLightup
  ;[1 0.247 0.2072] Brake Lights
  (Utilities|FlowControl|DoOnce false
    (:Completed
      (Rendering|Material|SetScalarParameterValue (Variables|Default|GetMID) "Emissive_BrightnessMax" 3.0 :!class /Script/Engine.MaterialInstanceDynamic)
      (CallFunction|ResetBreakLightFading))))

(event Custom|Audio%20Update
  (params (Moving :type Boolean))
  ;[1 1 1] Audio
  (if (CallFunction|IsGliderFullyDeployedAccordingtoAnim)
    (if Moving
      (Class|FortVehicleAudioVoice|SetParam (Variables|LegacyAudio|GetFortVehicleAudioVoice) 1.0)
      (if (not (Variables|Default|GetbMoving))
        (Variables|Default|SetbMoving true)
        (bind (_direction _lpf_amount) (Timeline|Low%20Pass%20Filter PlayFromStart)
          (:Update
            (Audio|Components|Audio|SetLowPassFilterFrequency (Variables|LegacyAudio|GetFortVehicleAudioVoice) (Math|Float|MapRangeUnclamped _lpf_amount 0.0 1.0 1e+03 2e+04) :!class /Script/Engine.AudioComponent))))
      (else
        (Class|FortVehicleAudioVoice|SetParam (Variables|LegacyAudio|GetFortVehicleAudioVoice) 0.2)
        (if (Variables|Default|GetbMoving)
          (Variables|Default|SetbMoving false)
          (Timeline|Low%20Pass%20Filter ReverseFromEnd))))
    (else
      (Class|FortVehicleAudioVoice|SetParam (Variables|LegacyAudio|GetFortVehicleAudioVoice) 0.0))))

(event Custom|StopTurnLeft
  ;[1 0.7774 0.3915] Turning Lights
  (Utilities|FlowControl|Sequence
    (:then_0
      (Utilities|FlowControl|DoOnce false
        (:Completed
          (Timeline|Timeline_TurnLeft PlayFromStart))))
    (:then_1
      (Timeline|Timeline_TurnLeft Stop))))

(event Custom|StopTurnRight
  ;[1 0.7774 0.3915] Turning Lights
  (Utilities|FlowControl|Sequence
    (:then_0
      (Utilities|FlowControl|DoOnce false
        (:Completed
          (Timeline|Timeline_TurnRight PlayFromStart))))
    (:then_1
      (Timeline|Timeline_TurnRight Stop))))

(event EventOnGliderFullyDeployed
  ;[1 1 1] Audio
  (|CallParentFunction)
  (if (Variables|FortPlayerParachute|GetbIsFrontEndPreview)
    (bind (_direction _fade) (Timeline|FadeInMusic PlayFromStart)
      (:Update
        (Class|FortVehicleAudioVoice|SetParam (Variables|LegacyAudio|GetFortVehicleAudioVoice) _fade)))
    (else
      (Audio|Components|Audio|FadeIn (Variables|LegacyAudio|GetFortVehicleAudioVoice) 0.25 1.0 0.0 "Linear" :!class /Script/Engine.AudioComponent)
      (Audio|AddEventbyComponent (GameInstanceSubsystems|GetMusicEventSubsystem) (Variables|LegacyAudio|GetFortVehicleAudioVoice) "(TagName=\"MusicEventSubsystem.Event.Glider\")" "(GameplayTags=)" 0 ""))))

(fn ConstructionScript
  (|CallParentFunction)
  (bind _returnvalue (Rendering|Material|CreateDynamicMaterialInstance (Variables|Parachute|GetParachuteMesh) 0 0 "None" :!class /Script/Engine.PrimitiveComponent))
  (Variables|Default|SetMID _returnvalue)
  (Rendering|Material|SetScalarParameterValue (Variables|Default|GetMID) "Emissive_BrightnessMax" (Variables|Default|GetBrakeLightBrightness) :!class /Script/Engine.MaterialInstanceDynamic))
