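; Legacy Gingerbread glider Blueprint for Athena cosmetics, extending a parent glider class.
; When the chute becomes visible, it runs the parent visibility handler, then attaches the second trail VFX component to one of the glider's bits mesh components.
; The attachment keeps relative location, rotation and scale with no socket.
; The construction script only calls the parent implementation.
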(event EventOnChuteVisibilityChanged
  (params (bParachuteVisibility :type Boolean))
  (|CallParentFunction bParachuteVisibility)
  (if bParachuteVisibility
    (Transformation|AttachComponentToComponent (CallFunction|GetTrailVFX2Component) (Utilities|Array|Get(acopy) (CallFunction|GetBitsMeshComponents)) "None" "KeepRelative" "KeepRelative" "KeepRelative" true)))

(fn ConstructionScript
  (|CallParentFunction))
