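; Legacy GlowBroBat glider Blueprint that controls the left and right wing particle trails.
; The parachute trail update event calls its parent, then checks forward movement and whether the animation reports the glider fully deployed.
; That check gates trail shutdown, while a DoOnce with no visible reset starts trails between named wing FX sockets.
; Handling is asymmetric: only the left component is deactivated and only the right is activated.
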(event Parachute|EventOnParachuteTrailUpdated
  (params (MovementLengthSquared :type Float%20%28single-precision%29) (ForwardDot :type Float%20%28single-precision%29) (RightDot :type Float%20%28single-precision%29) (RotationalMovementAcceleration :type Float%20%28single-precision%29))
  (|CallParentFunction MovementLengthSquared ForwardDot RightDot RotationalMovementAcceleration)
  (if (not (and (>= ForwardDot 0.2) (CallFunction|IsGliderFullyDeployedAccordingtoAnim)))
    (Components|Activation|Deactivate (Variables|Default|GetP_Wings_Trail_L))
    (Effects|Particles|Trails|EndTrails (Variables|Default|GetP_Wings_Trail_R)))
  (Utilities|FlowControl|DoOnce false
    (:Completed
      (Components|Activation|Activate (Variables|Default|GetP_Wings_Trail_R) false)
      (Effects|Particles|Trails|BeginTrails (Variables|Default|GetP_Wings_Trail_L) "FX_L3" "FX_L4" "ETrailWidthMode_FromCentre" 1.6)
      (Effects|Particles|Trails|BeginTrails (Variables|Default|GetP_Wings_Trail_R) "FX_R3" "FX_R4" "ETrailWidthMode_FromCentre" 1.6))))

(fn ConstructionScript
  (|CallParentFunction))
