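; Legacy Golden Skeleton glider Blueprint that drives the glider's trail VFX and passes movement values to its audio update.
; Trail updates check forward movement and full deployment in the animation, using DoOnce gates around trail activation and timeline playback.
; After a short delay the trail activates and a reversible timeline drives its AlphaScale, deactivating it when forward playback finishes.
; Hiding the parachute also deactivates the trail.
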(event EventOnChuteVisibilityChanged
  (params (bParachuteVisibility :type Boolean))
  (|CallParentFunction bParachuteVisibility)
  (if (not bParachuteVisibility)
    (Components|Activation|Deactivate (CallFunction|GetTrailVFXComponent))))

(event EventOnGliderFullyDeployed
  (|CallParentFunction))

(event Parachute|EventOnParachuteTrailUpdated
  (params (MovementLengthSquared :type Float%20%28single-precision%29) (ForwardDot :type Float%20%28single-precision%29) (RightDot :type Float%20%28single-precision%29) (RotationalMovementAcceleration :type Float%20%28single-precision%29))
  (CallFunction|AudioUpdated ForwardDot RightDot)
  (if (and (>= ForwardDot 0.2) (CallFunction|IsGliderFullyDeployedAccordingtoAnim))
    (Utilities|FlowControl|Sequence
      (:then_0
        (Utilities|FlowControl|DoOnce false :!label _label
          (:Completed
            (Utilities|FlowControl|Delay 0.2)
            (Components|Activation|Activate (CallFunction|GetTrailVFXComponent) true)
            (bind (_direction _alpha) (Timeline|Timeline_0 Reverse)
              (:Update
                (Effects|Components|ParticleSystem|SetFloatParameter (CallFunction|GetTrailVFXComponent) "AlphaScale" _alpha))
              (:Finished
                (if (Utilities|Enum|Equal(Enum) _direction "Forward")
                  (Components|Activation|Deactivate (CallFunction|GetTrailVFXComponent))))))))
      (:then_1
        _))
    (else
      (Utilities|FlowControl|Sequence
        (:then_0
          _)
        (:then_1
          (goto _label Reset)))))
  (Utilities|FlowControl|DoOnce false
    (:Completed
      (Timeline|Timeline_0 Play))))

(fn ConstructionScript
  (|CallParentFunction))
