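; Legacy glider Blueprint for the Graffiti Assassin cosmetic, managing its trail VFX component.
; When the chute becomes visible, it deactivates the trail to work around an auto-activate issue and attaches it to the parachute mesh at Body_Offset; hiding also deactivates it.
; On full deployment, it activates the trail only for front-end previews.
; Both events call their parent implementations first.
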(event EventOnChuteVisibilityChanged
  (params (bParachuteVisibility :type Boolean))
  (bind _returnvalue (CallFunction|GetTrailVFXComponent))
  (|CallParentFunction bParachuteVisibility)
  (if bParachuteVisibility
    (Components|Activation|Deactivate _returnvalue) ; Auto_Activate Issue
    (Transformation|AttachComponentToComponent _returnvalue (Variables|Parachute|GetParachuteMesh) "Body_Offset" "KeepRelative" "KeepRelative" "KeepRelative" true)
    (else
      (Components|Activation|Deactivate (CallFunction|GetTrailVFXComponent)))))

(event EventOnGliderFullyDeployed
  (|CallParentFunction)
  (if (Variables|FortPlayerParachute|GetbIsFrontEndPreview)
    (Components|Activation|Activate (CallFunction|GetTrailVFXComponent) false)))

(fn ConstructionScript
  (|CallParentFunction))
