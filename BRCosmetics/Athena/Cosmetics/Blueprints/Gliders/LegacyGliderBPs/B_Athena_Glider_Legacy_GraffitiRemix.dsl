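; Legacy glider Blueprint for the Graffiti Remix glider cosmetic, which animates a spray-paint transition on the parachute when it deploys.
; The construction script creates a dynamic instance of the Graffiti Remix material, stores it, and applies it to the parachute mesh's first slot.
; On full deployment, a timeline drives the SprayPaintTransition scalar parameter from the start.
; When the parachute is hidden, that parameter resets to zero.
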(event EventOnGliderFullyDeployed
  (|CallParentFunction)
  (bind (_direction _alpha_erosion_control) (Timeline|MaterialTransition PlayFromStart)
    (:Update
      (Rendering|Material|SetScalarParameterValue (Variables|Default|GetGliderMat_01) "SprayPaintTransition" _alpha_erosion_control :!class /Script/Engine.MaterialInstanceDynamic))))

(event EventOnChuteVisibilityChanged
  (params (bParachuteVisibility :type Boolean))
  (|CallParentFunction bParachuteVisibility)
  (if (not bParachuteVisibility)
    (Rendering|Material|SetScalarParameterValue (Variables|Default|GetGliderMat_01) "SprayPaintTransition" 0.0 :!class /Script/Engine.MaterialInstanceDynamic)))

(fn ConstructionScript
  (|CallParentFunction)
  (bind _returnvalue (Rendering|Material|CreateDynamicMaterialInstance "/Game/Gadgets/Assets/VinderTech_GliderChute/Glider_GraffitiRemix/Materials/MI_Glider_GraffitiRemix.MI_Glider_GraffitiRemix" "None" "None" :!class /Script/Engine.KismetMaterialLibrary))
  (bind _output_get (Variables|Default|SetGliderMat_01 _returnvalue))
  (Rendering|Material|SetMaterial (Variables|Parachute|GetParachuteMesh) 0 _output_get))
