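; Legacy glider Blueprint for the GreenJacketFemale cosmetic that drives an airburst particle effect on the parachute.
; On full deployment it sets the effect's streak, smoke and scale-color vectors, using separate values for front-end preview and gameplay, then activates it.
; When the chute becomes visible, it reapplies preview values and activates the effect only in an already-deployed preview; hiding the chute deactivates it.
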(event EventOnChuteVisibilityChanged
  (params (bParachuteVisibility :type Boolean))
  (bind _fx_airburst (Variables|Default|GetFX_AirBurst))
  (|CallParentFunction bParachuteVisibility)
  (if bParachuteVisibility
    (if (and (Variables|FortPlayerParachute|GetbIsFrontEndPreview) (CallFunction|IsGliderFullyDeployedAccordingtoAnim))
      (Effects|Components|ParticleSystem|SetVectorParameter _fx_airburst "StreakOffset_01" "1.000000,-1.000000,1.000000")
      (Effects|Components|ParticleSystem|SetVectorParameter _fx_airburst "SmokeOffset_02" "0, 0, 0")
      (Effects|Components|ParticleSystem|SetVectorParameter _fx_airburst "ScaleColor" "2.000000,2.000000,2.000000")
      (Components|Activation|Activate _fx_airburst false))
    (else
      (Components|Activation|Deactivate (Variables|Default|GetFX_AirBurst)))))

(event EventOnGliderFullyDeployed
  (bind _fx_airburst (Variables|Default|GetFX_AirBurst))
  (|CallParentFunction)
  (if (Variables|FortPlayerParachute|GetbIsFrontEndPreview)
    (Effects|Components|ParticleSystem|SetVectorParameter _fx_airburst "StreakOffset_01" "1.000000,-1.000000,1.000000")
    (Effects|Components|ParticleSystem|SetVectorParameter _fx_airburst "SmokeOffset_02" "0, 0, 0")
    (Effects|Components|ParticleSystem|SetVectorParameter _fx_airburst "ScaleColor" "2.000000,2.000000,2.000000")
    (else
      (Effects|Components|ParticleSystem|SetVectorParameter _fx_airburst "StreakOffset_01" "1.000000,1.000000,1.000000")
      (Effects|Components|ParticleSystem|SetVectorParameter _fx_airburst "SmokeOffset_02" "0.000000,-40.000000,0.000000")
      (Effects|Components|ParticleSystem|SetVectorParameter _fx_airburst "ScaleColor" "1.000000,1.000000,1.000000")))
  (Components|Activation|Activate _fx_airburst false))

(fn ConstructionScript
  (|CallParentFunction))
