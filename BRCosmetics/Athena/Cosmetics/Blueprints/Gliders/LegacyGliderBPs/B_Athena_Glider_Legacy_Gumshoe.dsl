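; Gumshoe glider Blueprint that adds car-style brake lights, turn signals and signal sounds to the parachute's trail VFX.
; Trail updates set tail light brightness and color from forward movement, and toggle cornering emitters from sideways movement and rotational acceleration, ignoring extreme spikes.
; While the chute is visible, particle spawn events from the second trail component trigger the turn signal cue, keeping audio tied to the blinks.
; Full deployment flips that component 180 degrees.
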(event Parachute|EventOnParachuteTrailUpdated
  (params (MovementLengthSquared :type Float%20%28single-precision%29) (ForwardDot :type Float%20%28single-precision%29) (RightDot :type Float%20%28single-precision%29) (RotationalMovementAcceleration :type Float%20%28single-precision%29))
  (bind _returnvalue (CallFunction|GetTrailVFX2Component))
  (bind _returnvalue_2 (CallFunction|GetTrailVFX2Component))
  (|CallParentFunction MovementLengthSquared ForwardDot RightDot RotationalMovementAcceleration)
  (Utilities|FlowControl|Sequence
    (:then_0
      ;[1 1 1] Brake Lights
      (if (>= ForwardDot 0.2)
        (Effects|Components|ParticleSystem|SetFloatParameter _returnvalue_2 "LightBrightness" 1.0)
        (else
          (Effects|Components|ParticleSystem|SetFloatParameter _returnvalue_2 "LightBrightness" 8.0)
          (if (<= ForwardDot -0.2)
            (Effects|Components|ParticleSystem|SetVectorParameter _returnvalue "TailLightColor" "80.000000,48.000000,20.000000")
            (else
              (Effects|Components|ParticleSystem|SetVectorParameter _returnvalue "TailLightColor" "30.000000, 0, 0"))))))
    (:then_1
      ;[1 1 1] Turn Signals
      (if (not (or (< RotationalMovementAcceleration -3.5e+02) (> RotationalMovementAcceleration 3.5e+02)))
        (bind _output_get (Variables|Default|SetTurnOnRightTurnSignalLight (or (< RightDot -0.2) (> RotationalMovementAcceleration 1.5))))
        (Effects|Components|ParticleSystem|SetEmitterEnable (CallFunction|GetTrailVFX2Component) "CorneringLight_L" _output_get)
        (bind _output_get_2 (Variables|Default|SetTurnOnLeftTurnSignalLight (or (> RightDot 0.2) (< RotationalMovementAcceleration -1.5))))
        (Effects|Components|ParticleSystem|SetEmitterEnable (CallFunction|GetTrailVFX2Component) "CorneringLight_R" _output_get_2)))))

(event EventOnGliderFullyDeployed
  (|CallParentFunction)
  (Transformation|SetRelativeRotation (CallFunction|GetTrailVFX2Component) "0,-180.000000, 0" false false))

(event EventOnChuteVisibilityChanged
  (params (bParachuteVisibility :type Boolean))
  (bind _returnvalue (CallFunction|GetTrailVFX2Component))
  (bind _outputdelegate (EventDispatchers|CreateEvent :!event Custom|Signal%20Check))
  (bind _ascascade_particle_system_component (Utilities|Casting|CastToParticleSystemComponent _returnvalue))
  ;[1 1 1] Turn Signal Controlled Sounds
  (|CallParentFunction bParachuteVisibility)
  (if bParachuteVisibility
    (EventDispatchers|BindEventtoOnParticleSpawn _ascascade_particle_system_component _outputdelegate)
    (else
      (EventDispatchers|UnbindEventfromOnParticleSpawn _ascascade_particle_system_component _outputdelegate))))

(event Custom|Signal%20Check
  (params (EventName :type Name) (EmitterTime :type Float%20%28single-precision%29) (Location :type Vector) (Velocity :type Vector))
  (bind _returnvalue (CallFunction|GetTrailVFXComponent))
  ;[1 1 1] Turn Signal Controlled Sounds
  (if (Utilities|Name|Equal(Name) EventName "SpawnL")
    (Audio|SpawnSoundAttached "/BRCosmetics/Sounds/Gliders/Glider_GumShoe_TurnSignal_Cue.Glider_GumShoe_TurnSignal_Cue" _returnvalue)
    (elif (Utilities|Name|Equal(Name) EventName "SpawnR")
      (Audio|SpawnSoundAttached "/BRCosmetics/Sounds/Gliders/Glider_GumShoe_TurnSignal_Cue.Glider_GumShoe_TurnSignal_Cue" _returnvalue))))

(fn ConstructionScript
  (|CallParentFunction))
