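; Legacy Hacivat glider Blueprint that drives propeller spin blur, smoke and music during flight.
; Trail updates switch particle parameters to high spin when deployed and moving forward, with a DoOnce-gated timeline ramping blur opacity, otherwise to low spin without smoke.
; Full deployment fades in music and registers it with the music event subsystem, and forward movement sets its level.
; Undeployed music checks remove the event, and hiding the chute deactivates Niagara components.
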(event Custom|RESET
  (bind _trailcomponents (Variables|FortPlayerParachute|GetTrailComponents))
  (bind _output (Utilities|Array|Get(acopy) _trailcomponents))
  (Utilities|FlowControl|DoOnce false
    (:Completed
      (Effects|Components|ParticleSystem|SetVectorParameter _output "BlurSpinSpeed" (Variables|Default|GetSpinSpeedHigh))
      (Effects|Components|ParticleSystem|SetVectorParameter _output "SoldPropBlurSpinSpeed" (Variables|Default|GetSolidSpinSpeedHigh))
      (Effects|Components|ParticleSystem|SetFloatParameter _output "SmokeRateScale" (Variables|Default|GetSmokeSpawnRageScale))
      (bind (_direction _lerp) (Timeline|Timeline_0 PlayFromStart)
        (:Update
          (Effects|Components|ParticleSystem|SetFloatParameter (CallFunction|GetTrailVFXComponent) "Opacity" (Math|Float|Lerp 0.2 (Variables|Default|GetSpinBlurHigh) _lerp)))))))

(event Parachute|EventOnParachuteTrailUpdated
  (params (MovementLengthSquared :type Float%20%28single-precision%29) (ForwardDot :type Float%20%28single-precision%29) (RightDot :type Float%20%28single-precision%29) (RotationalMovementAcceleration :type Float%20%28single-precision%29))
  (bind _trailcomponents (Variables|FortPlayerParachute|GetTrailComponents))
  (bind _output (Utilities|Array|Get(acopy) _trailcomponents))
  (|CallParentFunction MovementLengthSquared ForwardDot RightDot RotationalMovementAcceleration)
  (Utilities|FlowControl|Sequence
    (:then_0
      (if (and (>= ForwardDot 0.2) (CallFunction|IsGliderFullyDeployedAccordingtoAnim))
        (Utilities|FlowControl|DoOnce false
          (:Completed
            (Effects|Components|ParticleSystem|SetVectorParameter _output "BlurSpinSpeed" (Variables|Default|GetSpinSpeedHigh))
            (Effects|Components|ParticleSystem|SetVectorParameter _output "SoldPropBlurSpinSpeed" (Variables|Default|GetSolidSpinSpeedHigh))
            (Effects|Components|ParticleSystem|SetFloatParameter _output "SmokeRateScale" (Variables|Default|GetSmokeSpawnRageScale))
            (Timeline|Timeline_0 PlayFromStart)))
        (else
          (Effects|Components|ParticleSystem|SetFloatParameter _output "Opacity" (Variables|Default|GetSpinBlurLow))
          (Effects|Components|ParticleSystem|SetVectorParameter _output "BlurSpinSpeed" (Variables|Default|GetSpinSpeedLow))
          (Effects|Components|ParticleSystem|SetVectorParameter _output "SoldPropBlurSpinSpeed" (Variables|Default|GetSolidSpinSpeedLow))
          (Effects|Components|ParticleSystem|SetFloatParameter _output "SmokeRateScale" 0.0)
          (CallFunction|RESET))))
    (:then_1
      (CallFunction|AudioUpdateMusic ForwardDot 0.5 1.0))))

(event EventOnGliderFullyDeployed
  (|CallParentFunction)
  (Audio|Components|Audio|FadeIn (Variables|Default|GetFortVehicleAudioVoiceMusic) 0.25 1.0 0.0 "Linear" :!class /Script/Engine.AudioComponent)
  (Audio|AddEventbyComponent (GameInstanceSubsystems|GetMusicEventSubsystem) (Variables|Default|GetFortVehicleAudioVoiceMusic) "(TagName=\"MusicEventSubsystem.Event.Glider\")" "(GameplayTags=)" 0 ""))

(event EventCheckForMusicEvent
  (params (GliderFullyDeployed :type Boolean))
  (if (not GliderFullyDeployed)
    (Audio|RemoveEvent (GameInstanceSubsystems|GetMusicEventSubsystem) (Variables|Default|GetFortVehicleAudioVoiceMusic))))

(event EventOnChuteVisibilityChanged
  (params (bParachuteVisibility :type Boolean))
  ;[0.001518 0.6308 0.8714] Added Niagara shut off when glider hides
  (|CallParentFunction bParachuteVisibility)
  (if (not bParachuteVisibility)
    (for _array_element (Actor|GetComponentsByClass self "/Script/Niagara.NiagaraComponent")
      (Components|Activation|Deactivate _array_element))))

(fn AudioUpdate
  (params (Forward%20Dot :type Float%20%28double-precision%29))
  (bind _fortvehicleaudiovoice (Variables|LegacyAudio|GetFortVehicleAudioVoice))
  (if (CallFunction|IsGliderFullyDeployedAccordingtoAnim)
    (Class|FortVehicleAudioVoice|SetParam _fortvehicleaudiovoice (select (>= Forward%20Dot 0.2) 1.0 0.1))
    (else
      (Class|FortVehicleAudioVoice|SetParam _fortvehicleaudiovoice 0.0))))

(fn AudioUpdateMusic
  (params (ForwardDot :type Float%20%28double-precision%29) (MusicVolMin :type Float%20%28double-precision%29) (MusicVolMax :type Float%20%28double-precision%29))
  (bind _fortvehicleaudiovoicemusic (Variables|Default|GetFortVehicleAudioVoiceMusic))
  (Variables|/Script/Engine.AudioComponent|GetSound (Variables|Default|GetFortVehicleAudioVoiceMusic)
    (:then
      (if (CallFunction|IsGliderFullyDeployedAccordingtoAnim)
        (Class|FortVehicleAudioVoice|SetParam _fortvehicleaudiovoicemusic (select (>= (select (Variables|LegacyAudio|GetAbsolute%20Movement_Audio) (Math|Float|Absolute(Float) ForwardDot) ForwardDot) 0.2) MusicVolMax MusicVolMin))
        (else
          (Class|FortVehicleAudioVoice|SetParam _fortvehicleaudiovoicemusic 0.0))))
    (:else)))

(fn ConstructionScript
  (|CallParentFunction))
