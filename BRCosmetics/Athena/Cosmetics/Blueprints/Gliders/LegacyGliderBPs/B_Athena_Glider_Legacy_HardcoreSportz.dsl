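; Blueprint for the legacy HardcoreSportz Athena glider cosmetic, adjusting when its movement-driven trail VFX activates.
; On each parachute trail update, it sets the legacy VFX additional movement-activation condition to true when the absolute rightward movement dot is at least 0.2, then calls the parent handler.
; The construction script only defers to the parent, so all other glider behavior is inherited.
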(event Parachute|EventOnParachuteTrailUpdated
  (params (MovementLengthSquared :type Float%20%28single-precision%29) (ForwardDot :type Float%20%28single-precision%29) (RightDot :type Float%20%28single-precision%29) (RotationalMovementAcceleration :type Float%20%28single-precision%29))
  (Variables|LegacyVFX|SetAdditional%20Condition%20for%20Movement%20Activation (>= (Math|Float|Absolute(Float) RightDot) 0.2))
  (|CallParentFunction MovementLengthSquared ForwardDot RightDot RotationalMovementAcceleration))

(fn ConstructionScript
  (|CallParentFunction))
