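; Legacy Hippie glider Blueprint that drives the glider's trail and deploy effects.
; When the parachute shows, it attaches two trail components to the mesh and, if not yet fully deployed, spawns the deploy effect after a retriggerable delay.
; Trail updates set both trails' RandomColorEvent parameter to one of three random colors, throttled by the Check flag and a short delay.
; Hiding the parachute deactivates only the first trail component.
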(event EventOnChuteVisibilityChanged
  (params (bParachuteVisibility :type Boolean))
  (|CallParentFunction bParachuteVisibility)
  (Utilities|FlowControl|Sequence
    (:then_0
      (if bParachuteVisibility
        (Transformation|AttachComponentToComponent (CallFunction|GetTrailVFXComponent) (Variables|Parachute|GetParachuteMesh) "FX_Trail_R" "KeepRelative" "KeepRelative" "KeepRelative" true)
        (Transformation|AttachComponentToComponent (CallFunction|GetTrailVFX2Component) (Variables|Parachute|GetParachuteMesh) "FX_Trail_L" "KeepRelative" "KeepRelative" "KeepRelative" true)
        (Effects|Components|ParticleSystem|SetVectorParameter (CallFunction|GetTrailVFXComponent) "RandomColorEvent" "0, 0, 0")
        (Variables|Default|SetCheck true)
        (else
          (Components|Activation|Deactivate (CallFunction|GetTrailVFXComponent)))))
    (:then_1
      (if (and bParachuteVisibility (not (CallFunction|IsGliderFullyDeployedAccordingtoAnim)))
        (Utilities|FlowControl|RetriggerableDelay 1.1)
        (if bParachuteVisibility
          (bind _returnvalue (Effects|SpawnFXSystemAttached "/BRCosmetics/Gadgets/Assets/VinderTech_GliderChute/Glider_Hippie/FX/P_Glider_Hippie_Deploy.P_Glider_Hippie_Deploy" (Variables|Parachute|GetParachuteMesh) "Disc_Root"))
          (Variables|Default|SetDeploy _returnvalue))))))

(event EventOnGliderFullyDeployed
  (|CallParentFunction))

(event Custom|RandomColorChange
  (bind _returnvalue (Math|Random|RandomIntegerinRange 0 2))
  (bind _returnvalue_2 (Math|Conversions|ToVector(LinearColor) "(R=0.942366,G=1.000000,B=0.537660,A=1.000000)"))
  (bind _returnvalue_3 (Math|Conversions|ToVector(LinearColor) "(R=0.750000,G=0.610276,B=0.317834,A=1.000000)"))
  (bind _returnvalue_4 (Math|Conversions|ToVector(LinearColor) "(R=1.000000,G=1.000000,B=1.000000,A=1.000000)"))
  (bind _array (Utilities|Array|MakeArray _returnvalue_3 _returnvalue_4 _returnvalue_2))
  (bind _output (Utilities|Array|Get(acopy) _array _returnvalue))
  (if (Variables|Default|GetCheck)
    (Variables|Default|SetCheck false)
    (Effects|Components|ParticleSystem|SetVectorParameter (CallFunction|GetTrailVFXComponent) "RandomColorEvent" _output)
    (Effects|Components|ParticleSystem|SetVectorParameter (CallFunction|GetTrailVFX2Component) "RandomColorEvent" _output)
    (Utilities|FlowControl|Delay 0.1)
    (Variables|Default|SetCheck true)))

(event Parachute|EventOnParachuteTrailUpdated
  (params (MovementLengthSquared :type Float%20%28single-precision%29) (ForwardDot :type Float%20%28single-precision%29) (RightDot :type Float%20%28single-precision%29) (RotationalMovementAcceleration :type Float%20%28single-precision%29))
  (|CallParentFunction MovementLengthSquared ForwardDot RightDot RotationalMovementAcceleration)
  (CallFunction|RandomColorChange))

(fn ConstructionScript
  (|CallParentFunction))
