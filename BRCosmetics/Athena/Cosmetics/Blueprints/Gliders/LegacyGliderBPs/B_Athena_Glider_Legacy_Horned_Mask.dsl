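; Legacy glider cosmetic Blueprint for the Horned Mask glider, adding a fire effect to the base glider behavior.
; The fire component activates without reset once the glider fully deploys and deactivates when the parachute becomes hidden.
; Both event overrides call their parent implementations first, and the construction script only runs the parent's version.
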(event EventOnGliderFullyDeployed
  (|CallParentFunction)
  (Components|Activation|Activate (Variables|Default|GetP_HornedMask_Fires) false))

(event EventOnChuteVisibilityChanged
  (params (bParachuteVisibility :type Boolean))
  (|CallParentFunction bParachuteVisibility)
  (if (not bParachuteVisibility)
    (Components|Activation|Deactivate (Variables|Default|GetP_HornedMask_Fires))))

(fn ConstructionScript
  (|CallParentFunction))
