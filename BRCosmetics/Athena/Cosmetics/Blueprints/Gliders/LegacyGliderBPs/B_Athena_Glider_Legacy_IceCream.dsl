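; Legacy Ice Cream glider Blueprint that drives the cosmetic's sprinkle and melting trail effects, material animation and audio.
; On full deploy, front-end preview activates the trail, swaps to preview sprinkles, tunes vertex animation via a dynamic material and sets an audio parameter.
; Otherwise it enables gameplay sprinkles, fades in audio and registers it with the music event subsystem.
; When the chute becomes visible, both trail components attach to left/right melting sockets with an offset.
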(event EventOnGliderFullyDeployed
  (bind _returnvalue (CallFunction|GetTrailVFXComponent))
  (bind _returnvalue_2 (CallFunction|GetTrailVFXComponent))
  (|CallParentFunction)
  (if (Variables|FortPlayerParachute|GetbIsFrontEndPreview)
    (Components|Activation|Activate _returnvalue false)
    (Effects|Components|ParticleSystem|SetEmitterEnable _returnvalue "Sprinkles_sphere_FE" true)
    (Effects|Components|ParticleSystem|SetEmitterEnable _returnvalue "Sprinkles_sphere" false)
    (bind _returnvalue_3 (Rendering|Material|CreateDynamicMaterialInstance (Variables|Parachute|GetParachuteMesh) 1 "/BRCosmetics/Effects/Fort_Effects/Materials/M_Glider_IceCream_VertexAnim_Inst_02.M_Glider_IceCream_VertexAnim_Inst_02" "None" :!class /Script/Engine.PrimitiveComponent))
    (Rendering|Material|SetScalarParameterValue _returnvalue_3 "WPO_Intensity" 0.3 :!class /Script/Engine.MaterialInstanceDynamic)
    (Rendering|Material|SetScalarParameterValue _returnvalue_3 "WPO_PanspeedWhileMoving" 0.0 :!class /Script/Engine.MaterialInstanceDynamic)
    (Class|FortVehicleAudioVoice|SetParam (Variables|LegacyAudio|GetFortVehicleAudioVoice) 1.0)
    (else
      (Effects|Components|ParticleSystem|SetEmitterEnable _returnvalue_2 "Sprinkles_sphere_FE" false)
      (Effects|Components|ParticleSystem|SetEmitterEnable _returnvalue_2 "Sprinkles_sphere" true)
      (Audio|Components|Audio|FadeIn (Variables|LegacyAudio|GetFortVehicleAudioVoice) 0.25 1.0 0.0 "Linear" :!class /Script/Engine.AudioComponent)
      (Audio|AddEventbyComponent (GameInstanceSubsystems|GetMusicEventSubsystem) (Variables|LegacyAudio|GetFortVehicleAudioVoice) "(TagName=\"MusicEventSubsystem.Event.Glider\")" "(GameplayTags=)" 0 ""))))

(event EventOnChuteVisibilityChanged
  (params (bParachuteVisibility :type Boolean))
  (|CallParentFunction bParachuteVisibility)
  (if bParachuteVisibility
    (Transformation|AttachComponentToComponent (CallFunction|GetTrailVFXComponent) (Variables|Parachute|GetParachuteMesh) "FX_Melting_L" "KeepRelative" "KeepRelative" "KeepRelative" true)
    (Transformation|SetRelativeLocation (CallFunction|GetTrailVFXComponent) "65.000000, 0, 0" false false)
    (Transformation|AttachComponentToComponent (CallFunction|GetTrailVFX2Component) (Variables|Parachute|GetParachuteMesh) "FX_Melting_R" "KeepRelative" "KeepRelative" "KeepRelative" true)
    (Transformation|SetRelativeLocation (CallFunction|GetTrailVFX2Component) "65.000000, 0, 0" false false)))

(fn ConstructionScript
  (|CallParentFunction))
