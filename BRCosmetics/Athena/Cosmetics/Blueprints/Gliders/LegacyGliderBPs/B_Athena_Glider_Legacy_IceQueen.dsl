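; Legacy Blueprint for the Ice Queen glider cosmetic in Athena.
; When the glider fully deploys, it runs the parent logic, then sets the relative rotation of the second trail VFX component to a fixed 90-degree turn.
; The construction script only calls the parent implementation, so all other behavior is inherited from the parent glider class.
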(event EventOnGliderFullyDeployed
  (|CallParentFunction)
  (Transformation|SetRelativeRotation (CallFunction|GetTrailVFX2Component) "0,-90.000000, 0" false false))

(fn ConstructionScript
  (|CallParentFunction))
