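; Legacy Krampus glider Blueprint that adds movement-reactive trail VFX and mesh material effects on top of its parent parachute logic.
; When the parachute is hidden, it deactivates the trail VFX component.
; On trail updates, forward and right dot products feed the trail particle system's Velocity vector, while clamped forward dot drives the Wipe scalar on the parachute mesh materials.
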(event EventOnChuteVisibilityChanged
  (params (bParachuteVisibility :type Boolean))
  (|CallParentFunction bParachuteVisibility)
  (if (not bParachuteVisibility)
    (Components|Activation|Deactivate (CallFunction|GetTrailVFXComponent))))

(event Parachute|EventOnParachuteTrailUpdated
  (params (MovementLengthSquared :type Float%20%28single-precision%29) (ForwardDot :type Float%20%28single-precision%29) (RightDot :type Float%20%28single-precision%29) (RotationalMovementAcceleration :type Float%20%28single-precision%29))
  (|CallParentFunction MovementLengthSquared ForwardDot RightDot RotationalMovementAcceleration)
  (Effects|Components|ParticleSystem|SetVectorParameter (CallFunction|GetTrailVFXComponent) "Velocity" (Math|Vector|MakeVector ForwardDot RightDot))
  (Rendering|Material|SetScalarParameterValueonMaterials (Variables|Parachute|GetParachuteMesh) "Wipe" (Math|Float|Clamp(Float) ForwardDot -0.5)))

(fn ConstructionScript
  (|CallParentFunction))
