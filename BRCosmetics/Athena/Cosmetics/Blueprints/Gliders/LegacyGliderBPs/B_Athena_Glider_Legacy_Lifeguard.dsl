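; Blueprint for the legacy Lifeguard glider cosmetic, which toggles its left engine effect with the glider's state.
; When the glider fully deploys, it calls the parent logic and activates the EngineFX_L component.
; When chute visibility changes to hidden, it calls the parent logic and deactivates that effect.
; The construction script only runs the parent's.
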(event EventOnGliderFullyDeployed
  (|CallParentFunction)
  (Components|Activation|Activate (Variables|Default|GetEngineFX_L) false))

(event EventOnChuteVisibilityChanged
  (params (bParachuteVisibility :type Boolean))
  (|CallParentFunction bParachuteVisibility)
  (if (not bParachuteVisibility)
    (Components|Activation|Deactivate (Variables|Default|GetEngineFX_L))))

(fn ConstructionScript
  (|CallParentFunction))
