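; Legacy Mad Commander glider Blueprint that drives a turbine thruster trail effect on the parachute mesh.
; When the chute becomes visible, it snaps the trail VFX to the FX_Turbine_r socket and activates it, deactivating it when hidden.
; Trail updates lower the thrust particle parameters unless the glider is fully deployed and moving forward.
; A DoOnce after that check sets full thrust, with no reset shown.
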(event EventOnChuteVisibilityChanged
  (params (bParachuteVisibility :type Boolean))
  (bind _returnvalue (CallFunction|GetTrailVFXComponent))
  (|CallParentFunction bParachuteVisibility)
  (if bParachuteVisibility
    (Transformation|AttachComponentToComponent _returnvalue (Variables|Parachute|GetParachuteMesh) "FX_Turbine_r" "SnapToTarget" "SnapToTarget" "SnapToTarget" true)
    (Components|Activation|Activate _returnvalue true)
    (else
      (Components|Activation|Deactivate _returnvalue))))

(event Parachute|EventOnParachuteTrailUpdated
  (params (MovementLengthSquared :type Float%20%28single-precision%29) (ForwardDot :type Float%20%28single-precision%29) (RightDot :type Float%20%28single-precision%29) (RotationalMovementAcceleration :type Float%20%28single-precision%29))
  (bind _returnvalue (CallFunction|GetTrailVFXComponent))
  (|CallParentFunction MovementLengthSquared ForwardDot RightDot RotationalMovementAcceleration)
  (if (not (and (>= ForwardDot 0.2) (CallFunction|IsGliderFullyDeployedAccordingtoAnim)))
    (Effects|Components|ParticleSystem|SetFloatParameter _returnvalue "Thrust" 0.25)
    (Effects|Components|ParticleSystem|SetFloatParameter _returnvalue "ThrustCore" 0.0)
    (Effects|Components|ParticleSystem|SetFloatParameter _returnvalue "ThrustSmoke" 0.0))
  (Utilities|FlowControl|DoOnce false
    (:Completed
      (Effects|Components|ParticleSystem|SetFloatParameter _returnvalue "Thrust" 1.0)
      (Effects|Components|ParticleSystem|SetFloatParameter _returnvalue "ThrustCore" 1.0)
      (Effects|Components|ParticleSystem|SetFloatParameter _returnvalue "ThrustSmoke" 1.0))))

(fn ConstructionScript
  (|CallParentFunction))
