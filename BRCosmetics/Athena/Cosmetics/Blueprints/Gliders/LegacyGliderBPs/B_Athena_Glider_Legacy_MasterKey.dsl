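; Legacy Master Key glider Blueprint that drives an orange emissive glow on the parachute mesh.
; The construction script creates a dynamic material instance whose EmissiveMultiplier the events set.
; Trail updates on a fully deployed glider switch the glow off between 6 and 18 time of day, and on otherwise.
; In front-end preview, deployment fades the glow in by timeline after a delay, and hiding the chute resets it.
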(event EventOnChuteVisibilityChanged
  (params (bParachuteVisibility :type Boolean))
  (|CallParentFunction bParachuteVisibility)
  (if bParachuteVisibility
    _
    (elif (Variables|FortPlayerParachute|GetbIsFrontEndPreview)
      (Rendering|Material|SetVectorParameterValue (Variables|Default|GetMID) "EmissiveMultiplier" "(R=0.000000,G=0.000000,B=0.000000,A=1.000000)" :!class /Script/Engine.MaterialInstanceDynamic))))

(event EventOnGliderFullyDeployed
  (|CallParentFunction)
  (if (Variables|FortPlayerParachute|GetbIsFrontEndPreview)
    (Utilities|FlowControl|Delay 2.0)
    (bind (_direction _lerp) (Timeline|Timeline_0 PlayFromStart)
      (:Update
        (Rendering|Material|SetVectorParameterValue (Variables|Default|GetMID) "EmissiveMultiplier" (Math|Color|Lerp(LinearColor) "(R=0.000000,G=0.000000,B=0.000000,A=1.000000)" "(R=7.800000,G=1.428000,B=0.023700,A=1.000000)" _lerp) :!class /Script/Engine.MaterialInstanceDynamic)))))

(event Parachute|EventOnParachuteTrailUpdated
  (params (MovementLengthSquared :type Float%20%28single-precision%29) (ForwardDot :type Float%20%28single-precision%29) (RightDot :type Float%20%28single-precision%29) (RotationalMovementAcceleration :type Float%20%28single-precision%29))
  (bind _mid (Variables|Default|GetMID))
  (|CallParentFunction MovementLengthSquared ForwardDot RightDot RotationalMovementAcceleration)
  (if (CallFunction|IsGliderFullyDeployedAccordingtoAnim)
    (if (Math|Float|InRange(Float) (TimeOfDay|GetTimeOfDay :!class /Script/FortniteGame.FortKismetLibrary) 6.0 18.0)
      (Rendering|Material|SetVectorParameterValue _mid "EmissiveMultiplier" "(R=0.000000,G=0.000000,B=0.000000,A=1.000000)" :!class /Script/Engine.MaterialInstanceDynamic)
      (else
        (Rendering|Material|SetVectorParameterValue _mid "EmissiveMultiplier" "(R=7.800000,G=1.428688,B=0.023734,A=1.000000)" :!class /Script/Engine.MaterialInstanceDynamic)))))

(fn ConstructionScript
  (|CallParentFunction)
  (bind _returnvalue (Rendering|Material|CreateDynamicMaterialInstance (Variables|Parachute|GetParachuteMesh) 0 0 "Glider_MasterKey_MID" :!class /Script/Engine.PrimitiveComponent))
  (Variables|Default|SetMID _returnvalue))
