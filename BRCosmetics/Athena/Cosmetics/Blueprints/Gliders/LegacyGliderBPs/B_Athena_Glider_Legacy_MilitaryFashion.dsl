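; Cosmetic glider Blueprint for the legacy MilitaryFashion glider, driving its jet contrail effects and emissive glow.
; Construction creates a dynamic material instance on the parachute mesh.
; Full deployment starts the main jet after a delay, then repositions the trail VFX or disables it in front-end preview.
; Trail updates toggle the FE contrail and brighten emissive during forward movement; undeployed or hidden states switch both off.
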(event EventOnGliderFullyDeployed
  (bind _returnvalue (CallFunction|GetTrailVFXComponent))
  (|CallParentFunction)
  (Utilities|FlowControl|Delay 0.2)
  (Components|Activation|Activate (Variables|Default|GetP_GliderContrail_MilitaryFashion_JetMain_Active) true)
  (if (Variables|FortPlayerParachute|GetbIsFrontEndPreview)
    (Components|Activation|Deactivate _returnvalue)
    (else
      (Transformation|SetRelativeLocationAndRotation _returnvalue "36.000000,40.000000,16.000000" "0,90.000000, 0" false false))))

(event Parachute|EventOnParachuteTrailUpdated
  (params (MovementLengthSquared :type Float%20%28single-precision%29) (ForwardDot :type Float%20%28single-precision%29) (RightDot :type Float%20%28single-precision%29) (RotationalMovementAcceleration :type Float%20%28single-precision%29))
  (bind _returnvalue (CallFunction|IsGliderFullyDeployedAccordingtoAnim))
  (bind _p_glidercontrail_militaryfashion_fe (Variables|Default|GetP_GliderContrail_MilitaryFashion_FE))
  (|CallParentFunction MovementLengthSquared ForwardDot RightDot RotationalMovementAcceleration)
  (if _returnvalue
    (Components|Activation|Activate _p_glidercontrail_militaryfashion_fe false)
    (if (and _returnvalue (>= ForwardDot 0.2))
      (Rendering|Material|SetVectorParameterValue (Variables|Default|GetMID) "EmissiveMultiplier" "(R=20.000000,G=20.000000,B=20.000000,A=1.000000)" :!class /Script/Engine.MaterialInstanceDynamic)
      (Components|Activation|Deactivate _p_glidercontrail_militaryfashion_fe))
    (else
      (Components|Activation|Deactivate (Variables|Default|GetP_GliderContrail_MilitaryFashion_FE))
      (Rendering|Material|SetVectorParameterValue (Variables|Default|GetMID) "EmissiveMultiplier" "(R=1.000000,G=1.000000,B=1.000000,A=1.000000)" :!class /Script/Engine.MaterialInstanceDynamic))))

(event EventOnChuteVisibilityChanged
  (params (bParachuteVisibility :type Boolean))
  (|CallParentFunction bParachuteVisibility)
  (if (not bParachuteVisibility)
    (Components|Activation|Deactivate (Variables|Default|GetP_GliderContrail_MilitaryFashion_FE))
    (Rendering|Material|SetVectorParameterValue (Variables|Default|GetMID) "EmissiveMultiplier" "(R=1.000000,G=1.000000,B=1.000000,A=1.000000)" :!class /Script/Engine.MaterialInstanceDynamic)))

(fn ConstructionScript
  (|CallParentFunction)
  (bind _returnvalue (Rendering|Material|CreateDynamicMaterialInstance (Variables|Parachute|GetParachuteMesh) 0 0 "GliderMilitaryFashion_MID" :!class /Script/Engine.PrimitiveComponent))
  (Variables|Default|SetMID _returnvalue))
