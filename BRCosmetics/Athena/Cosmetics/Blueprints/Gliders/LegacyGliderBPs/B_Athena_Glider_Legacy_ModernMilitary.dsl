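; Legacy Modern Military glider Blueprint that drives propeller particle effects during deploy and flight.
; Showing the chute activates the deploy propeller, with a timeline ramping spin and fade if moving forward; full deployment deactivates it, delayed when moving forward.
; Trail updates cache ForwardDot and, once deployed, switch between spinning blades and motion blur at a forward threshold.
; DoOnce gates with ResetMove/ResetStop events manage these transitions, and a gliding timeline blends blade speed and fade.
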(event EventOnGliderFullyDeployed
  (|CallParentFunction)
  (if (>= (Variables|Default|GetForward%20Dot) 0.2)
    (Utilities|FlowControl|Delay 0.5))
  (Components|Activation|Deactivate (Variables|Default|GetPropellerDeployL)))

(event EventOnChuteVisibilityChanged
  (params (bParachuteVisibility :type Boolean))
  (|CallParentFunction bParachuteVisibility)
  (if bParachuteVisibility
    (Components|Activation|Activate (Variables|Default|GetPropellerDeployL) false)
    (if (>= (Variables|Default|GetForward%20Dot) 0.2)
      (bind (_direction _spinspeed _erode) (Timeline|BladesDeployingSpinSpeedCurve PlayFromStart)
        (:Update
          (Effects|Components|ParticleSystem|SetVectorParameter (Variables|Default|GetPropellerDeployR) "PropellerSpinSpeed" (Math|Vector|MakeVector 0.0 0.0 (Math|Float|Lerp 3.0 8.0 _spinspeed)))
          (Effects|Components|ParticleSystem|SetFloatParameter (Variables|Default|GetPropellerDeployR) "PropellerFade" (Math|Float|Lerp 1.5 0.0 _erode)))) ; Handle the Speed Changing etc.
      (else
        (Effects|Components|ParticleSystem|SetVectorParameter (Variables|Default|GetPropellerDeployR) "PropellerSpinSpeed" "0, 0,3.000000")))))

(event Custom|ResetMove
  (bind _propellerspinningblurl (Variables|Default|GetPropellerSpinningBlurL))
  (bind _propellerspinningbladesl (Variables|Default|GetPropellerSpinningBladesL))
  (Utilities|FlowControl|DoOnce false
    (:Completed
      (Components|Activation|Activate _propellerspinningblurl false)
      (Effects|Components|ParticleSystem|SetFloatParameter _propellerspinningblurl "Opacity" 1.0)
      (bind (_direction _spinspeed _erode) (Timeline|BladesGlidingSpinSpeedCurve PlayFromStart)
        (:Update
          (Effects|Components|ParticleSystem|SetVectorParameter _propellerspinningbladesl "PropellerSpinSpeed" (Math|Vector|MakeVector 0.0 0.0 (Math|Float|Lerp 3.0 15.0 _spinspeed)))
          (Effects|Components|ParticleSystem|SetFloatParameter _propellerspinningbladesl "PropellerFade" (Math|Float|Lerp 0.0 2.0 _erode)))
        (:Finished
          (Components|Activation|Deactivate (Variables|Default|GetPropellerSpinningBladesL))
          (CallFunction|ResetStop))) ; Handle the Speed Changing etc.
    )))

(event Parachute|EventOnParachuteTrailUpdated
  (params (MovementLengthSquared :type Float%20%28single-precision%29) (ForwardDot :type Float%20%28single-precision%29) (RightDot :type Float%20%28single-precision%29) (RotationalMovementAcceleration :type Float%20%28single-precision%29))
  (bind _propellerspinningblurl (Variables|Default|GetPropellerSpinningBlurL))
  (bind _propellerspinningbladesl (Variables|Default|GetPropellerSpinningBladesL))
  (bind _propellerspinningbladesl_2 (Variables|Default|GetPropellerSpinningBladesL))
  (|CallParentFunction MovementLengthSquared ForwardDot RightDot RotationalMovementAcceleration)
  (bind _output_get (Variables|Default|SetForward%20Dot ForwardDot))
  (if (and (>= _output_get 0.2) (CallFunction|IsGliderFullyDeployedAccordingtoAnim))
    (Utilities|FlowControl|DoOnce false
      (:Completed
        (Components|Activation|Activate _propellerspinningblurl false)
        (Effects|Components|ParticleSystem|SetFloatParameter _propellerspinningblurl "Opacity" 1.0)
        (Timeline|BladesGlidingSpinSpeedCurve PlayFromStart)))
    (else
      (Utilities|FlowControl|DoOnce false
        (:Completed
          (Components|Activation|Deactivate (Variables|Default|GetPropellerSpinningBlurR))
          (Components|Activation|Activate _propellerspinningbladesl false)
          (Effects|Components|ParticleSystem|SetVectorParameter _propellerspinningbladesl "PropellerSpinSpeed" "0, 0,3.000000")
          (Effects|Components|ParticleSystem|SetFloatParameter _propellerspinningbladesl "PropellerFade" 2.0)
          (CallFunction|ResetMove))))))

(event Custom|ResetStop
  (bind _propellerspinningbladesl (Variables|Default|GetPropellerSpinningBladesL))
  (Utilities|FlowControl|DoOnce false
    (:Completed
      (Components|Activation|Deactivate (Variables|Default|GetPropellerSpinningBlurR))
      (Components|Activation|Activate _propellerspinningbladesl false)
      (Effects|Components|ParticleSystem|SetVectorParameter _propellerspinningbladesl "PropellerSpinSpeed" "0, 0,3.000000")
      (Effects|Components|ParticleSystem|SetFloatParameter _propellerspinningbladesl "PropellerFade" 2.0)
      (CallFunction|ResetMove))))

(fn ConstructionScript
  (|CallParentFunction))
