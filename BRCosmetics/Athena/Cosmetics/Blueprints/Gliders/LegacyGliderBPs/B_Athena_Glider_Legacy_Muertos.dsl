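; Glider Blueprint for the legacy Muertos glider, extending its parent's deploy and visibility events with cosmetic trail effects.
; In front-end preview, the secondary trail VFX switches on when fully deployed and off when hidden.
; On trail updates, the anim Blueprint's CanFlap gates DoOnce-guarded anim trails between left and right socket pairs.
; Timelines fade the first trail's Opacity in, then out after a short delay, before shutdown.
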(event EventOnChuteVisibilityChanged
  (params (bParachuteVisibility :type Boolean))
  (|CallParentFunction bParachuteVisibility)
  (if bParachuteVisibility
    (if (and (Variables|FortPlayerParachute|GetbIsFrontEndPreview) (CallFunction|IsGliderFullyDeployedAccordingtoAnim))
      (Components|Activation|Activate (CallFunction|GetTrailVFX2Component) false))
    (elif (Variables|FortPlayerParachute|GetbIsFrontEndPreview)
      (Components|Activation|Deactivate (CallFunction|GetTrailVFX2Component)))))

(event EventOnGliderFullyDeployed
  (|CallParentFunction)
  (if (Variables|FortPlayerParachute|GetbIsFrontEndPreview)
    (Components|Activation|Activate (CallFunction|GetTrailVFX2Component) false)))

(event Parachute|EventOnParachuteTrailUpdated
  (params (MovementLengthSquared :type Float%20%28single-precision%29) (ForwardDot :type Float%20%28single-precision%29) (RightDot :type Float%20%28single-precision%29) (RotationalMovementAcceleration :type Float%20%28single-precision%29))
  (bind _animtrails01 (Variables|Default|GetAnimTrails01))
  (|CallParentFunction MovementLengthSquared ForwardDot RightDot RotationalMovementAcceleration)
  (if (CallFunction|IsGliderFullyDeployedAccordingtoAnim)
    (if (Variables|/BRCosmetics/Gadgets/Assets/VinderTech_GliderChute/Blueprints/Muertos_Glider_AnimBP.Muertos_Glider_AnimBP_C|GetCanFlap (Utilities|Casting|CastToMuertos_Glider_AnimBP (Components|SkeletalMesh|GetAnimInstance (Variables|Parachute|GetParachuteMesh))))
      (Utilities|FlowControl|DoOnce false
        (:Completed
          (Components|Activation|Activate (Variables|Default|GetAnimTrails01) false)
          (Effects|Particles|Trails|BeginTrails (Variables|Default|GetAnimTrails01) "primaryA1_l" "primaryJ1_l" "ETrailWidthMode_FromCentre" 1.0)
          (Effects|Particles|Trails|BeginTrails (Variables|Default|GetAnimTrails02) "primaryA1_r" "primaryJ1_r" "ETrailWidthMode_FromCentre" 1.0)
          (bind (_direction _fadein) (Timeline|Timeline_0 PlayFromStart)
            (:Update
              (Effects|Components|ParticleSystem|SetFloatParameter (Variables|Default|GetAnimTrails01) "Opacity" _fadein))
            (:Finished
              (Utilities|FlowControl|DoOnce false
                (:Completed
                  (Utilities|FlowControl|Delay 0.3)
                  (bind (_direction_2 _fadeout) (Timeline|Timeline_1 PlayFromStart)
                    (:Update
                      (Effects|Components|ParticleSystem|SetFloatParameter _animtrails01 "Opacity" _fadeout))
                    (:Finished
                      (Components|Activation|Deactivate _animtrails01)
                      (Effects|Particles|Trails|EndTrails (Variables|Default|GetAnimTrails02))))))))))
      (else
        _))))

(fn ConstructionScript
  (|CallParentFunction))
