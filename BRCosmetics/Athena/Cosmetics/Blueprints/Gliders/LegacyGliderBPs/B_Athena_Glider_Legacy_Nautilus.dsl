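; Legacy Nautilus glider Blueprint that adds a propeller trail effect and an underwater-style caustic camera lens effect.
; When the parachute becomes visible, it attaches the second trail VFX component to the propeller socket and rotates it; when hidden, it deactivates the cached lens particle.
; Trail updates use forward movement and full-deployment checks, with DoOnce gates, to spawn, cache and toggle the lens effect on the player pawn.
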(event EventOnGliderFullyDeployed
  (|CallParentFunction))

(event EventOnChuteVisibilityChanged
  (params (bParachuteVisibility :type Boolean))
  (bind _returnvalue (CallFunction|GetTrailVFX2Component))
  (|CallParentFunction bParachuteVisibility)
  (if bParachuteVisibility
    (Transformation|AttachComponentToComponent _returnvalue (Variables|Parachute|GetParachuteMesh) "propeller" "SnapToTarget" "SnapToTarget" "KeepWorld" true)
    (Transformation|SetRelativeRotation _returnvalue "0,-90.000000, 0" false false)
    (else
      (bind _causticcamlensfx (Variables|Default|GetCausticCamLensFX)
        (:then
          (Components|Activation|Deactivate _causticcamlensfx))
        (:else)))))

(event Parachute|EventOnParachuteTrailUpdated
  (params (MovementLengthSquared :type Float%20%28single-precision%29) (ForwardDot :type Float%20%28single-precision%29) (RightDot :type Float%20%28single-precision%29) (RotationalMovementAcceleration :type Float%20%28single-precision%29))
  (bind _lenseffectinterface (Variables|Default|GetLensEffectInterface))
  (bind _returnvalue (CameraLensEffect|GetPrimaryParticleComponent _lenseffectinterface))
  (|CallParentFunction MovementLengthSquared ForwardDot RightDot RotationalMovementAcceleration)
  (if (and (>= ForwardDot 0.2) (CallFunction|IsGliderFullyDeployedAccordingtoAnim))
    (Utilities|FlowControl|DoOnce false
      (:Completed
        (Variables|Default|GetCausticCamLensFX
          (:then
            (Components|Activation|Activate _returnvalue false)
            (Utilities|FlowControl|DoOnce false
              (:Completed
                (Components|Activation|Deactivate (CameraLensEffect|GetPrimaryParticleComponent (Variables|Default|GetLensEffectInterface))))))
          (:else
            _))
        (bind _returnvalue_2 (Feedback|SpawnGenericCameraLensEffect (CallFunction|GetFortPlayerPawn) "/BRCosmetics/Effects/Fort_Effects/Camera_Effects/B_CameraLens_CausticPattern.B_CameraLens_CausticPattern_C"))
        (Variables|Default|SetLensEffectInterface _returnvalue_2)
        (Variables|Default|SetCausticCamLensFX _returnvalue)))
    (else
      (Variables|Default|GetCausticCamLensFX
        (:then
          _)
        (:else
          _)))))

(fn ConstructionScript
  (|CallParentFunction))
