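; Legacy glider Blueprint for the female Photographer glider that triggers its burst and particle effects.
; On receiving the item definition, it caches the parachute mesh's Photographer anim instance and binds that instance's BurstFX event to fire the stretch burst.
; Full deployment activates the deployed burst and deploy particles.
; Trail updates call the parent, then check the anim instance's StretchBurstActivated flag alongside a DoOnce-gated stretch burst activation.
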(event Parachute|EventOnReceiveParachuteItemdef
  (bind _asphotographer_glider_anim_bp (Utilities|Casting|CastToPhotographer_Glider_AnimBP (Components|SkeletalMesh|GetAnimInstance (Variables|Parachute|GetParachuteMesh))))
  (bind _output_get (Variables|Default|SetPhotographerAnimInstance _asphotographer_glider_anim_bp))
  (Default|BindEventtoBurstFX _output_get (delegate Custom|BurstFrontendFX)))

(event Custom|BurstFrontendFX
  (Components|Activation|Activate (Variables|Default|GetStretchBurst) false))

(event EventOnGliderFullyDeployed
  (Components|Activation|Activate (Variables|Default|GetDeployedBurst) false)
  (Components|Activation|Activate (Variables|Default|GetDeployParticles) false))

(event Parachute|EventOnParachuteTrailUpdated
  (params (MovementLengthSquared :type Float%20%28single-precision%29) (ForwardDot :type Float%20%28single-precision%29) (RightDot :type Float%20%28single-precision%29) (RotationalMovementAcceleration :type Float%20%28single-precision%29))
  (bind _photographeraniminstance (Variables|Default|GetPhotographerAnimInstance))
  (|CallParentFunction MovementLengthSquared ForwardDot RightDot RotationalMovementAcceleration)
  (Utilities|IsValid _photographeraniminstance
    (:Is%20Valid
      (if (Variables|/BRCosmetics/Gadgets/Assets/VinderTech_GliderChute/Blueprints/Photographer_Glider_AnimBP.Photographer_Glider_AnimBP_C|GetStretchBurstActivated _photographeraniminstance)
        _)
      (Utilities|FlowControl|DoOnce false
        (:Completed
          (Components|Activation|Activate (Variables|Default|GetStretchBurst) false))))))

(fn ConstructionScript
  (|CallParentFunction))
