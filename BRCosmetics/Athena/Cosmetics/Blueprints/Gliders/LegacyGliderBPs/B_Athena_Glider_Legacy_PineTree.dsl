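; Legacy PineTree glider Blueprint that controls the glider's Engine component during parachute trail updates.
; After calling the parent update, it deactivates Engine unless forward movement dot is at least 0.2 and the deploy animation reports fully deployed.
; Engine activation is gated by a DoOnce with no reset shown.
; The construction script only calls its parent.
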(event Parachute|EventOnParachuteTrailUpdated
  (params (MovementLengthSquared :type Float%20%28single-precision%29) (ForwardDot :type Float%20%28single-precision%29) (RightDot :type Float%20%28single-precision%29) (RotationalMovementAcceleration :type Float%20%28single-precision%29))
  (|CallParentFunction MovementLengthSquared ForwardDot RightDot RotationalMovementAcceleration)
  (if (not (and (>= ForwardDot 0.2) (CallFunction|IsGliderFullyDeployedAccordingtoAnim)))
    (Components|Activation|Deactivate (Variables|Default|GetEngine)))
  (Utilities|FlowControl|DoOnce false
    (:Completed
      (Components|Activation|Activate (Variables|Default|GetEngine) false))))

(fn ConstructionScript
  (|CallParentFunction))
