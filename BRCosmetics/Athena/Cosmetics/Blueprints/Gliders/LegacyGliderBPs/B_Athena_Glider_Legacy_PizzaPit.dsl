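; Legacy Pizza Pit glider Blueprint that adds a pizza grease camera lens effect while gliding.
; On full deployment, it reactivates an existing lens effect's primary particle component, or spawns and caches a new effect if the player pawn is locally controlled.
; When the chute hides, it deactivates particles through a separate Camera Lens Effect variable, which this graph never sets, rather than the cached LensEffectInterface.
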(event EventOnGliderFullyDeployed
  (bind _playerpawn (Variables|FortPlayerParachute|GetPlayerPawn))
  (|CallParentFunction)
  (Utilities|IsValid (Utilities|ToObject(Interface) (Variables|Default|GetLensEffectInterface))
    (:Is%20Valid
      (Components|Activation|Activate (CameraLensEffect|GetPrimaryParticleComponent (Variables|Default|GetLensEffectInterface)) false))
    (:Is%20Not%20Valid
      (bind _asfort_player_pawn_athena (Utilities|Casting|CastToFortPlayerPawnAthena _playerpawn))
      (if (Pawn|IsLocallyControlled _playerpawn)
        (bind _returnvalue (Feedback|SpawnGenericCameraLensEffect _asfort_player_pawn_athena "/BRCosmetics/Athena/Prototype/Blueprints/Mooney/FX/BP_CameraLens_PizzaGrease.BP_CameraLens_PizzaGrease_C"))
        (Variables|Default|SetLensEffectInterface _returnvalue)))))

(event EventOnChuteVisibilityChanged
  (params (bParachuteVisibility :type Boolean))
  (|CallParentFunction bParachuteVisibility)
  (if (not bParachuteVisibility)
    (bind _camera_lens_effect (Variables|Default|GetCamera%20Lens%20Effect)
      (:then
        (Components|Activation|Deactivate (Variables|/Script/CascadeRuntime.Emitter|GetParticleSystemComponent _camera_lens_effect)))
      (:else))))

(fn ConstructionScript
  (|CallParentFunction))
