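; PrairiePusher legacy glider Blueprint that triggers directional spill effects and sounds as the parachute moves.
; On each trail update, after the parent call, a Sequence checks ForwardDot and RightDot against thresholds for front, back, right and left.
; Each direction activates its Bullect component and spawns an attached spill sound at it.
; DoOnce gates and short random Delays throttle the triggers.
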(event Parachute|EventOnParachuteTrailUpdated
  (params (MovementLengthSquared :type Float%20%28single-precision%29) (ForwardDot :type Float%20%28single-precision%29) (RightDot :type Float%20%28single-precision%29) (RotationalMovementAcceleration :type Float%20%28single-precision%29))
  (bind _returnvalue (Math|Random|RandomFloatinRange 0.2 0.5))
  (bind _bullect_right (Variables|Default|GetBullect_Right))
  (bind _bullect_left (Variables|Default|GetBullect_Left))
  (bind _bullect_front (Variables|Default|GetBullect_Front))
  (bind _bullect_back (Variables|Default|GetBullect_Back))
  (|CallParentFunction MovementLengthSquared ForwardDot RightDot RotationalMovementAcceleration)
  (Utilities|FlowControl|Sequence
    (:then_0
      (if (not (>= ForwardDot 0.5))
        (Utilities|FlowControl|Delay _returnvalue))
      (Utilities|FlowControl|DoOnce false
        (:Completed
          (Components|Activation|Activate _bullect_front true)
          (Audio|SpawnSoundAttached "/BRCosmetics/Sounds/Gliders/SFX_Glider_PrairiePusher_Spill_Cue.SFX_Glider_PrairiePusher_Spill_Cue" _bullect_front))))
    (:then_1
      (if (not (<= ForwardDot -0.5))
        (Utilities|FlowControl|Delay _returnvalue))
      (Utilities|FlowControl|DoOnce false
        (:Completed
          (Components|Activation|Activate _bullect_back true)
          (Audio|SpawnSoundAttached "/BRCosmetics/Sounds/Gliders/SFX_Glider_PrairiePusher_Spill_Cue.SFX_Glider_PrairiePusher_Spill_Cue" _bullect_back))))
    (:then_2
      (if (not (>= RightDot 0.5))
        (Utilities|FlowControl|Delay _returnvalue))
      (Utilities|FlowControl|DoOnce false
        (:Completed
          (Components|Activation|Activate _bullect_right true)
          (Audio|SpawnSoundAttached "/BRCosmetics/Sounds/Gliders/SFX_Glider_PrairiePusher_Spill_Cue.SFX_Glider_PrairiePusher_Spill_Cue" _bullect_right))))
    (:then_3
      (if (not (<= RightDot -0.5))
        (Utilities|FlowControl|Delay _returnvalue))
      (Utilities|FlowControl|DoOnce false
        (:Completed
          (Components|Activation|Activate _bullect_left true)
          (Audio|SpawnSoundAttached "/BRCosmetics/Sounds/Gliders/SFX_Glider_PrairiePusher_Spill_Cue.SFX_Glider_PrairiePusher_Spill_Cue" _bullect_left))))))

(fn ConstructionScript
  (|CallParentFunction))
