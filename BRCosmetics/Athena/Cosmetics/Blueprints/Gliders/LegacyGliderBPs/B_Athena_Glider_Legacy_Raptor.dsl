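; Legacy Raptor glider Blueprint that drives the glider's particle effects and propeller blur material on top of parent behavior.
; Full deployment activates smoke, plus a propeller effect in front-end preview, then fades in blur by setting opacity after a delay.
; Visibility changes reset blur opacity, and hiding also stops smoke.
; Trail updates activate the Raptor effect after a delay during deployed forward movement, otherwise deactivate it.
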(event EventOnGliderFullyDeployed
  (|CallParentFunction)
  (if (Variables|FortPlayerParachute|GetbIsFrontEndPreview)
    (Components|Activation|Activate (Variables|Default|GetP_Glider_Raptor_Propeller_Frontend) false)
    (else
      _))
  (Components|Activation|Activate (Variables|Default|GetP_Glider_Raptor_Smoke) false)
  (Utilities|FlowControl|Delay 1.0)
  (Rendering|Material|SetScalarParameterValueonMaterials (Variables|Default|GetPropeller_Blur_R) "Opacity" 0.7))

(event EventOnChuteVisibilityChanged
  (params (bParachuteVisibility :type Boolean))
  (|CallParentFunction bParachuteVisibility)
  (if bParachuteVisibility
    (Rendering|Material|SetScalarParameterValueonMaterials (Variables|Default|GetPropeller_Blur_R) "Opacity" 0.0)
    (else
      (Components|Activation|Deactivate (Variables|Default|GetP_Glider_Raptor_Smoke))
      (Rendering|Material|SetScalarParameterValueonMaterials (Variables|Default|GetPropeller_Blur_R) "Opacity" 0.0))))

(event Parachute|EventOnParachuteTrailUpdated
  (params (MovementLengthSquared :type Float%20%28single-precision%29) (ForwardDot :type Float%20%28single-precision%29) (RightDot :type Float%20%28single-precision%29) (RotationalMovementAcceleration :type Float%20%28single-precision%29))
  (bind _p_glider_raptor (Variables|Default|GetP_Glider_Raptor))
  (|CallParentFunction MovementLengthSquared ForwardDot RightDot RotationalMovementAcceleration)
  (if (and (>= ForwardDot 0.2) (CallFunction|IsGliderFullyDeployedAccordingtoAnim))
    (Utilities|FlowControl|Delay 0.6)
    (Components|Activation|Activate _p_glider_raptor false)
    (else
      (Components|Activation|Deactivate _p_glider_raptor))))

(fn ConstructionScript
  (|CallParentFunction))
