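; Legacy glider Blueprint for the Raptor Black Ops glider cosmetic, adjusting its trail VFX when the parachute's visibility changes.
; After the parent handler runs, a visible parachute attaches the second trail VFX component to the parachute mesh's Frame socket with relative rules and welding.
; It then yaws the two trail components 90 degrees in opposite directions.
; The construction script only defers to the parent.
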(event EventOnChuteVisibilityChanged
  (params (bParachuteVisibility :type Boolean))
  (bind _returnvalue (CallFunction|GetTrailVFX2Component))
  (|CallParentFunction bParachuteVisibility)
  (if bParachuteVisibility
    (Transformation|AttachComponentToComponent _returnvalue (Variables|Parachute|GetParachuteMesh) "Frame" "KeepRelative" "KeepRelative" "KeepRelative" true)
    (Transformation|SetRelativeRotation _returnvalue "0,-90.000000, 0" false false)
    (Transformation|SetRelativeRotation (CallFunction|GetTrailVFXComponent) "0,90.000000, 0" false false)))

(fn ConstructionScript
  (|CallParentFunction))
