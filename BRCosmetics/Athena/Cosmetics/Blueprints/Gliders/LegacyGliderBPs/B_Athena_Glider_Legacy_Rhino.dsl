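; Legacy Rhino glider Blueprint that adds flame burst and glow effects on top of its parent glider logic.
; Construction creates a dynamic material instance from the parachute mesh's first material.
; Trail updates switch that material's EmissiveMultiplier on during sufficiently forward movement with the glider animation fully deployed, and off otherwise.
; Full deployment activates the flame burst component; showing the parachute reorients both trail VFX components, scaling the first.
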(event EventOnGliderFullyDeployed
  (|CallParentFunction)
  (Components|Activation|Activate (Variables|Default|GetP_Glider_Rhino_Flame_Burst) false))

(event EventOnChuteVisibilityChanged
  (params (bParachuteVisibility :type Boolean))
  (|CallParentFunction bParachuteVisibility)
  (if bParachuteVisibility
    (Transformation|SetRelativeTransform (CallFunction|GetTrailVFXComponent) (Math|Transform|MakeTransform "0, 0, 0" "0,-90.000000, 0" "1.100000,1.000000,1.100000") false false)
    (Transformation|SetRelativeRotation (CallFunction|GetTrailVFX2Component) "0,-90.000000, 0" false false)))

(event Parachute|EventOnParachuteTrailUpdated
  (params (MovementLengthSquared :type Float%20%28single-precision%29) (ForwardDot :type Float%20%28single-precision%29) (RightDot :type Float%20%28single-precision%29) (RotationalMovementAcceleration :type Float%20%28single-precision%29))
  (bind _mid (Variables|Default|GetMID))
  (|CallParentFunction MovementLengthSquared ForwardDot RightDot RotationalMovementAcceleration)
  (if (and (>= ForwardDot 0.2) (CallFunction|IsGliderFullyDeployedAccordingtoAnim))
    (Rendering|Material|SetVectorParameterValue _mid "EmissiveMultiplier" "(R=3.000000,G=3.000000,B=3.000000,A=1.000000)" :!class /Script/Engine.MaterialInstanceDynamic)
    (else
      (Rendering|Material|SetVectorParameterValue _mid "EmissiveMultiplier" "(R=0.000000,G=0.000000,B=0.000000,A=1.000000)" :!class /Script/Engine.MaterialInstanceDynamic))))

(fn ConstructionScript
  (bind _parachutemesh (Variables|Parachute|GetParachuteMesh))
  (|CallParentFunction)
  (Utilities|IsValid _parachutemesh
    (:Is%20Valid
      (bind _returnvalue (Rendering|Material|CreateDynamicMaterialInstance _parachutemesh 0 0 "ParachuteMesh0_MID" :!class /Script/Engine.PrimitiveComponent))
      (Variables|Default|SetMID _returnvalue))))
