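; Blueprint for the legacy Rust Lord Remix glider cosmetic, driving its chainsaw, engine flare, exhaust and thrust audio.
; Visibility changes attach trail VFX to the parachute mesh or shut effects down, while full deployment activates the chainsaw.
; In front-end preview, a timeline triggers and ends thrust effects.
; Trail updates use forward movement on the deployed glider to fire thrust, with DoOnce nodes gating thrust and exhaust shutoff.
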(event EventOnChuteVisibilityChanged
  (params (bParachuteVisibility :type Boolean))
  (|CallParentFunction bParachuteVisibility)
  (if bParachuteVisibility
    (Transformation|AttachComponentToComponent (CallFunction|GetTrailVFXComponent) (Variables|Parachute|GetParachuteMesh) "Body_Offset" "KeepRelative" "KeepRelative" "KeepRelative" true)
    (else
      (Components|Activation|Deactivate (Variables|Default|GetChainsaw))
      (if (Variables|FortPlayerParachute|GetbIsFrontEndPreview)
        (Components|Activation|Deactivate (CallFunction|GetTrailVFX2Component))
        (Components|Timeline|Stop (Variables|B_Athena_Glider_Legacy_RustLordRemix|GetFrontend%20Preview%20Timeline))))))

(event EventOnGliderFullyDeployed
  (|CallParentFunction)
  (Components|Activation|Activate (Variables|Default|GetChainsaw) false)
  (if (Variables|FortPlayerParachute|GetbIsFrontEndPreview)
    (Components|Activation|Activate (CallFunction|GetTrailVFX2Component) false)
    (Timeline|Frontend%20Preview%20Timeline PlayFromStart
      (:Trigger
        (Components|Activation|Activate (Variables|Default|GetEnginesFlareOn) false)
        (Audio|Components|Audio|Play (Variables|Default|GetGlider_RustLordRemix_Thrust_Cue) 0.0)
        (Utilities|FlowControl|Delay 0.1)
        (Components|Activation|Activate (Variables|Default|GetExhaust) false))
      (:Deactivate
        (Components|Activation|Deactivate (Variables|Default|GetExhaust))))))

(event Parachute|EventOnParachuteTrailUpdated
  (params (MovementLengthSquared :type Float%20%28single-precision%29) (ForwardDot :type Float%20%28single-precision%29) (RightDot :type Float%20%28single-precision%29) (RotationalMovementAcceleration :type Float%20%28single-precision%29))
  (|CallParentFunction MovementLengthSquared ForwardDot RightDot RotationalMovementAcceleration)
  (if (and (>= ForwardDot 0.2) (CallFunction|IsGliderFullyDeployedAccordingtoAnim))
    (Utilities|FlowControl|DoOnce false
      (:Completed
        (Components|Activation|Activate (Variables|Default|GetEnginesFlareOn) false)
        (Audio|Components|Audio|Play (Variables|Default|GetGlider_RustLordRemix_Thrust_Cue) 0.0)
        (Utilities|FlowControl|Delay 0.1)
        (Components|Activation|Activate (Variables|Default|GetExhaust) false)))
    (else
      _))
  (Utilities|FlowControl|DoOnce false
    (:Completed
      (Components|Activation|Deactivate (Variables|Default|GetExhaust)))))

(fn ConstructionScript
  (|CallParentFunction))
