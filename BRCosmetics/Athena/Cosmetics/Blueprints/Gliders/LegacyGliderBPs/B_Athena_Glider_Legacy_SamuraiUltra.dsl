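; Legacy SamuraiUltra glider Blueprint that drives the glider's red eye trails and audio from its movement.
; On each parachute trail update it calls the parent, then treats a forward dot product at or above a threshold as forward movement.
; The red eye trails activate only while moving forward and fully deployed according to animation, otherwise they deactivate.
; A custom event sets the vehicle audio voice parameter by forward movement, or zero until fully deployed.
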(event Parachute|EventOnParachuteTrailUpdated
  (params (MovementLengthSquared :type Float%20%28single-precision%29) (ForwardDot :type Float%20%28single-precision%29) (RightDot :type Float%20%28single-precision%29) (RotationalMovementAcceleration :type Float%20%28single-precision%29))
  (bind _returnvalue (>= ForwardDot 0.2))
  (bind _redeyetrails (Variables|Default|GetRedEyeTrails))
  (|CallParentFunction MovementLengthSquared ForwardDot RightDot RotationalMovementAcceleration)
  (CallFunction|AudioVoice1Update _returnvalue)
  (if (and _returnvalue (CallFunction|IsGliderFullyDeployedAccordingtoAnim))
    (Components|Activation|Activate _redeyetrails false)
    (else
      (Components|Activation|Deactivate _redeyetrails))))

(event Custom|AudioVoice1Update
  (params (Is%20Moving%20Forward :type Boolean))
  (bind _fortvehicleaudiovoice1 (Variables|Default|GetFortVehicleAudioVoice1))
  (if (CallFunction|IsGliderFullyDeployedAccordingtoAnim)
    (Class|FortVehicleAudioVoice|SetParam _fortvehicleaudiovoice1 (select Is%20Moving%20Forward 1.0 0.5))
    (else
      (Class|FortVehicleAudioVoice|SetParam _fortvehicleaudiovoice1 0.0))))

(fn ConstructionScript
  (|CallParentFunction))
