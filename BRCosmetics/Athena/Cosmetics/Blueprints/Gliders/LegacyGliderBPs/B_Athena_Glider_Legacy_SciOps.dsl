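; SciOps legacy glider Blueprint that adds twin jet exhaust trail effects to the parachute mesh.
; When the chute becomes visible, both trail components attach to the left and right exhaust sockets.
; The first forward movement while fully deployed fires a DoOnce-gated, two-stage timeline sequence driving particle parameters for flare, length, ring speed, opacity and tail size.
; The sequence animates the first trail component; the second only receives a RefSpawn parameter.
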(event EventOnChuteVisibilityChanged
  (params (bParachuteVisibility :type Boolean))
  (|CallParentFunction bParachuteVisibility)
  (if bParachuteVisibility
    (Transformation|AttachComponentToComponent (CallFunction|GetTrailVFXComponent) (Variables|Parachute|GetParachuteMesh) "FX_ExhaustL" "KeepRelative" "KeepRelative" "KeepRelative" true)
    (Transformation|AttachComponentToComponent (CallFunction|GetTrailVFX2Component) (Variables|Parachute|GetParachuteMesh) "FX_ExhaustR" "KeepRelative" "KeepRelative" "KeepRelative" true)))

(event Parachute|EventOnParachuteTrailUpdated
  (params (MovementLengthSquared :type Float%20%28single-precision%29) (ForwardDot :type Float%20%28single-precision%29) (RightDot :type Float%20%28single-precision%29) (RotationalMovementAcceleration :type Float%20%28single-precision%29))
  (bind _returnvalue (CallFunction|GetTrailVFXComponent))
  (bind _returnvalue_2 (CallFunction|GetTrailVFXComponent))
  (bind _returnvalue_3 (CallFunction|GetTrailVFXComponent))
  (|CallParentFunction MovementLengthSquared ForwardDot RightDot RotationalMovementAcceleration)
  (if (and (>= ForwardDot 0.2) (CallFunction|IsGliderFullyDeployedAccordingtoAnim))
    (Utilities|FlowControl|DoOnce false
      (:Completed
        (Effects|Components|ParticleSystem|SetFloatParameter _returnvalue_2 "RingShapeSpeed" 0.0)
        (Effects|Components|ParticleSystem|SetFloatParameter _returnvalue_2 "OpacityOverall" 1.0)
        (Effects|Components|ParticleSystem|SetFloatParameter _returnvalue_2 "FlareSpawn" 1.0)
        (Effects|Components|ParticleSystem|SetFloatParameter _returnvalue_2 "JetTailSize" 1.2)
        (Effects|Components|ParticleSystem|SetFloatParameter (CallFunction|GetTrailVFX2Component) "RefSpawn" 1.0)
        (bind (_direction _length) (Timeline|Timeline_0 PlayFromStart)
          (:Update
            (Effects|Components|ParticleSystem|SetVectorParameter (CallFunction|GetTrailVFXComponent) "LengthCurve" (Math|Vector|MakeVector 1.0 (Math|Float|Lerp 1.0 2.0 _length) 1.0)))
          (:Finished
            (Utilities|FlowControl|DoOnce false
              (:Completed
                (Effects|Components|ParticleSystem|SetFloatParameter _returnvalue_3 "FlareSpawn" 0.0)
                (Effects|Components|ParticleSystem|SetFloatParameter _returnvalue_3 "RefSpawn" 0.0)
                (bind (_direction_2 _length_2) (Timeline|Timeline_1 PlayFromStart)
                  (:Update
                    (Effects|Components|ParticleSystem|SetVectorParameter _returnvalue "LengthCurve" (Math|Vector|MakeVector 1.0 (Math|Float|Lerp 1.0 2.0 _length_2) 1.0))
                    (Effects|Components|ParticleSystem|SetFloatParameter _returnvalue "RingShapeSpeed" (Math|Float|Lerp 2.0 0.0 _length_2))
                    (Effects|Components|ParticleSystem|SetFloatParameter _returnvalue "OpacityOverall" (Math|Float|Lerp 0.4 2.0 _length_2))
                    (Effects|Components|ParticleSystem|SetFloatParameter _returnvalue "JetTailSize" (Math|Float|Lerp 1.55 1.2 (* _length_2 2.0))))
                  (:Finished))))))))
    (else
      _)))

(fn ConstructionScript
  (|CallParentFunction))
