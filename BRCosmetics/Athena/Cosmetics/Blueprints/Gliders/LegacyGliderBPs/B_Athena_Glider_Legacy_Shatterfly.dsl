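; Legacy Shatterfly glider Blueprint that layers cosmetic material and trail-effect tweaks on inherited parachute behavior.
; On BeginPlay it reads r.SeparateTranslucency and passes a matching SepTrans scalar to the parachute mesh materials.
; When the chute becomes visible, it yaws the second trail VFX component 90 degrees.
; Trail updates include a forward-movement and deployment check, plus a DoOnce-gated timeline that drives BP_Alpha on the main trail particles.
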(event EventBeginPlay
  (bind (_outvalue _returnvalue) (Fortnite|GetNumericCVar "r.SeparateTranslucency"))
  (|CallParentFunction)
  (if _returnvalue
    (Rendering|Material|SetScalarParameterValueonMaterials (Variables|Parachute|GetParachuteMesh) "SepTrans" (Utilities|Select :Index (Math|Float|Truncate _outvalue) :Option%200 0.0 :Option%201 1.0))))

(event EventOnChuteVisibilityChanged
  (params (bParachuteVisibility :type Boolean))
  (|CallParentFunction bParachuteVisibility)
  (if bParachuteVisibility
    (Transformation|SetRelativeRotation (CallFunction|GetTrailVFX2Component) "0,90.000000, 0" false false)))

(event Parachute|EventOnParachuteTrailUpdated
  (params (MovementLengthSquared :type Float%20%28single-precision%29) (ForwardDot :type Float%20%28single-precision%29) (RightDot :type Float%20%28single-precision%29) (RotationalMovementAcceleration :type Float%20%28single-precision%29))
  (|CallParentFunction MovementLengthSquared ForwardDot RightDot RotationalMovementAcceleration)
  (if (and (>= ForwardDot 0.2) (CallFunction|IsGliderFullyDeployedAccordingtoAnim))
    _)
  (Utilities|FlowControl|DoOnce false
    (:Completed
      (bind (_direction _newtrack_0) (Timeline|Timeline_0 PlayFromStart)
        (:Update
          (Effects|Components|ParticleSystem|SetFloatParameter (CallFunction|GetTrailVFXComponent) "BP_Alpha" _newtrack_0))))))

(fn ConstructionScript
  (|CallParentFunction))
