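; Legacy Sheath glider Blueprint that drives the glider's engine and air trail particle effects.
; The idle engine effect activates on full deployment and deactivates when the chute is hidden.
; Trail updates with enough forward movement while fully deployed trigger DoOnce-guarded engine boost and air trails, with a timeline animating trail alpha.
; Another DoOnce path ends the right trail and runs a timeline on boost alpha, then deactivates the boost through a Gate.
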(event EventOnChuteVisibilityChanged
  (params (bParachuteVisibility :type Boolean))
  (|CallParentFunction bParachuteVisibility)
  (if (not bParachuteVisibility)
    (Components|Activation|Deactivate (Variables|Default|GetEngine_Idle_L))))

(event EventOnGliderFullyDeployed
  (|CallParentFunction)
  (Components|Activation|Activate (Variables|Default|GetEngine_Idle_L) false))

(event Parachute|EventOnParachuteTrailUpdated
  (params (MovementLengthSquared :type Float%20%28single-precision%29) (ForwardDot :type Float%20%28single-precision%29) (RightDot :type Float%20%28single-precision%29) (RotationalMovementAcceleration :type Float%20%28single-precision%29))
  (bind _air_trail_r (Variables|Default|GetAir_Trail_R))
  (bind _engine_boost_l (Variables|Default|GetEngine_Boost_L))
  (|CallParentFunction MovementLengthSquared ForwardDot RightDot RotationalMovementAcceleration)
  (if (and (>= ForwardDot 0.2) (CallFunction|IsGliderFullyDeployedAccordingtoAnim))
    (Utilities|FlowControl|DoOnce false
      (:Completed
        (Components|Activation|Activate _engine_boost_l false)
        (Effects|Components|ParticleSystem|SetFloatParameter _engine_boost_l "AlphaScale02" 1.0)
        (Components|Activation|Activate (Variables|Default|GetAir_Trail_R) false)
        (Effects|Particles|Trails|BeginTrails (Variables|Default|GetAir_Trail_L) "fx_airtrail_l1" "fx_airtrail_l2" "ETrailWidthMode_FromCentre" 1.0)
        (Effects|Particles|Trails|BeginTrails (Variables|Default|GetAir_Trail_R) "fx_airtrail_r1" "fx_airtrail_r2" "ETrailWidthMode_FromCentre" 1.0)
        (Utilities|FlowControl|Sequence
          (:then_0
            (bind (_direction _alpha) (Timeline|Timeline_0 PlayFromStart)
              (:Update
                (Effects|Components|ParticleSystem|SetFloatParameter (CallFunction|GetTrailVFXComponent) "AlphaScale" _alpha))))
          (:then_1
            _)
          (:then_2
            (Utilities|FlowControl|DoOnce false
              (:Completed
                (Components|Activation|Deactivate _air_trail_r)
                (Effects|Particles|Trails|EndTrails _air_trail_r)
                (Effects|Components|ParticleSystem|SetFloatParameter (Variables|Default|GetEngine_Boost_L) "AlphaScale02" 1.0)
                (Utilities|FlowControl|Sequence
                  (:then_0
                    (bind (_direction_2 _alpha_2 _alpha02) (Timeline|Timeline_1 PlayFromStart)
                      (:Update
                        (Effects|Components|ParticleSystem|SetFloatParameter (Variables|Default|GetEngine_Boost_L) "AlphaScale02" _alpha02))
                      (:Finished
                        (Utilities|FlowControl|Gate true
                          (:Exit
                            (Components|Activation|Deactivate (Variables|Default|GetEngine_Boost_L)))))))
                  (:then_1
                    _)
                  (:then_2
                    _))))))))
    (else
      _)))

(fn ConstructionScript
  (|CallParentFunction))
