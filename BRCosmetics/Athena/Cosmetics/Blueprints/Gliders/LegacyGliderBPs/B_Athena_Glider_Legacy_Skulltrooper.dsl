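; Skulltrooper legacy glider blueprint that drives coffin-themed visuals: a light, eye trail VFX and grave smoke.
; Deployment attaches the trail, reveals the smoke and colors the light, using the emissive cosmetic variant in game.
; Trail updates smoothly interpolate a material vector from movement direction to sway the smoke and toggle forward-motion particles.
; Hiding the chute hides smoke and light; low effects quality or mobile destroys the light at startup.
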(event EventBeginPlay
  (|CallParentFunction)
  (if (or (<= (Settings|GetVisualEffectQuality (Settings|GetFortGameUserSettings)) 1) (TouchPlatform|IsMobileGame))
    (Components|DestroyComponent (Variables|Default|GetCoffin%20Light))))

(event EventOnGliderFullyDeployed
  (bind (_asmaterial_instance_dynamic _bsuccess) (Utilities|Casting|CastToMaterialInstanceDynamic (Rendering|Material|GetMaterial (Variables|Default|GetS_FX_SkullTrooper_Glider_Panel) 0)))
  (bind _coffin_light (Variables|Default|GetCoffin%20Light))
  (|CallParentFunction)
  (Transformation|AttachComponentToComponent (CallFunction|GetTrailVFX2Component) (Variables|Parachute|GetParachuteMesh) "FX_Coffin_Eyes" "SnapToTarget" "SnapToTarget" "SnapToTarget" true)
  (Development|SetHiddeninGame (Variables|Default|GetS_FX_GraveSmoke) false false)
  (if (Variables|FortPlayerParachute|GetbIsFrontEndPreview)
    (Rendering|Material|SetVectorParameterValueonMaterials (Variables|Default|GetS_FX_GraveSmoke) "LocalPostioning" (Variables|Default|GetFrontEndSmoke))
    (else
      _))
  (bind _coffin_light_2 (Variables|Default|GetCoffin%20Light)
    (:then
      (if (Variables|FortPlayerParachute|GetbIsFrontEndPreview)
        (if _bsuccess
          (Rendering|Components|Light|SetLightColor _coffin_light "(R=0.261077,G=1.000000,B=0.502138,A=1.000000)" true)
          (else
            (Rendering|Components|Light|SetLightColor _coffin_light "(R=0.124772,G=0.822786,B=1.000000,A=1.000000)" true)))
        (else
          (bind _index (CosmeticLoadout|GetVariantChannelActiveIndexforSlotActor "(TagName=\"LoadoutSlot.Glider\")" (CallFunction|GetFortPlayerPawn) "(TagName=\"Cosmetics.Variant.Channel.Emissive\")" (Utilities|Array|MakeArray "(TagName=\"Cosmetics.Variant.Property.Emissive1\")" "(TagName=\"Cosmetics.Variant.Property.Emissive2\")")))
          (Rendering|Components|Light|SetLightColor _coffin_light_2 (Utilities|Select :Index _index :Option%200 "(R=0.124772,G=0.822786,B=1.000000,A=1.000000)" :Option%201 "(R=0.262251,G=1.000000,B=0.502886,A=1.000000)") true)
          (Development|SetHiddeninGame _coffin_light_2 false false))))
    (:else)))

(event Parachute|EventOnParachuteTrailUpdated
  (params (MovementLengthSquared :type Float%20%28single-precision%29) (ForwardDot :type Float%20%28single-precision%29) (RightDot :type Float%20%28single-precision%29) (RotationalMovementAcceleration :type Float%20%28single-precision%29))
  (|CallParentFunction MovementLengthSquared ForwardDot RightDot RotationalMovementAcceleration)
  (Utilities|FlowControl|Sequence
    (:then_0
      (bind _output_get (Variables|Default|SetCurrentSmoke (Math|Interpolation|VInterpTo (Variables|Default|GetCurrentSmoke) (CallFunction|DetermineSmokeMovement ForwardDot RightDot) (Utilities|Time|GetWorldDeltaSeconds) (Variables|Default|GetSmoke%20Movement%20Interp%20Speed))))
      (Rendering|Material|SetVectorParameterValueonMaterials (Variables|Default|GetS_FX_GraveSmoke) "LocalPostioning" _output_get))
    (:then_1
      (Effects|Components|ParticleSystem|SetEmitterEnable (CallFunction|GetTrailVFXComponent) "SmokeLoopVelocity" (and (>= ForwardDot 0.2) (CallFunction|IsGliderFullyDeployedAccordingtoAnim))))))

(event EventOnChuteVisibilityChanged
  (params (bParachuteVisibility :type Boolean))
  (|CallParentFunction bParachuteVisibility)
  (if bParachuteVisibility
    (CallFunction|ApplyVariants)
    (if (not (CallFunction|IsGliderFullyDeployedAccordingtoAnim))
      (Components|Activation|Deactivate (CallFunction|GetTrailVFX2Component)))
    (else
      (Development|SetHiddeninGame (Variables|Default|GetS_FX_GraveSmoke) true false)
      (bind _coffin_light (Variables|Default|GetCoffin%20Light)
        (:then
          (Development|SetHiddeninGame _coffin_light true false))
        (:else)))))

(fn DetermineSmokeMovement
  (params (Forward%20Dot :type Float%20%28double-precision%29) (Right%20Dot :type Float%20%28double-precision%29))
  (return (+ (* (Math|Vector|Lerp(Vector) "0,-1.000000, 0" "0,1.000000, 0" (Math|Float|MapRangeClamped Forward%20Dot -1.0 1.0 0.0 1.0)) (Variables|Default|GetForward_Intensity)) (* (Math|Vector|Lerp(Vector) "1.000000,0.000000, 0" "-1.000000,0.000000, 0" (Math|Float|MapRangeClamped Right%20Dot -1.0 1.0 0.0 1.0)) (Variables|Default|GetRight_Intensity)))))

(fn ConstructionScript
  (|CallParentFunction))
