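; Legacy glider Blueprint for SleepyTime that drives its trail particles, blue tail material and emissive core.
; Construction creates dynamic materials for the tail and core.
; On chute show, it attaches both trail VFX components to the parachute mesh, and ResetTimeline animates tail appearance and mesh scales.
; Trail updates use DoOnce gates to raise streak and star speeds and play a material speed-up timeline during deployed forward flight, or restore slower speeds.
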(event EventOnChuteVisibilityChanged
  (params (bParachuteVisibility :type Boolean))
  (|CallParentFunction bParachuteVisibility)
  (if bParachuteVisibility
    (Utilities|FlowControl|Sequence
      (:then_0
        (Components|Activation|Deactivate (CallFunction|GetTrailVFXComponent)) ; AutoActivate Issue
        (Transformation|AttachComponentToComponent (CallFunction|GetTrailVFXComponent) (Variables|Parachute|GetParachuteMesh) "Handles_Offset" "KeepRelative" "KeepRelative" "KeepRelative" true)
        (Transformation|AttachComponentToComponent (CallFunction|GetTrailVFX2Component) (Variables|Parachute|GetParachuteMesh) "Handles_Offset" "KeepRelative" "KeepRelative" "KeepRelative" true)
        (Transformation|SetRelativeLocationAndRotation (CallFunction|GetTrailVFXComponent) "0,40.000000, 0" "0, 0,90.000000" false false)
        (Transformation|SetRelativeLocationAndRotation (CallFunction|GetTrailVFX2Component) "0,40.000000, 0" "0, 0,90.000000" false false)
        (Variables|VFX|SetTrailVFX%20Deployed%20Activation true)
        (Variables|VFX|SetTrailVFX2%20Deployed%20Activation true))
      (:then_1
        (if (Components|Timeline|IsPlaying (Variables|B_Athena_Glider_Legacy_SleepyTime|GetResetTimeline))
          (Components|Timeline|Stop (Variables|B_Athena_Glider_Legacy_SleepyTime|GetResetTimeline)))
        (Rendering|Material|SetScalarParameterValue (Variables|Default|GetBlueTail) "Appearing" -2.0 :!class /Script/Engine.MaterialInstanceDynamic)
        (Transformation|SetRelativeScale3D (Variables|Default|GetEmissive%20Projectile%20Mesh) "0, 0, 0")
        (bind (_direction _appearing _grow _scalesizereduced) (Timeline|ResetTimeline PlayFromStart)
          (:Update
            (Rendering|Material|SetScalarParameterValue (Variables|Default|GetBlueTail) "Appearing" _appearing :!class /Script/Engine.MaterialInstanceDynamic)
            (Transformation|SetRelativeScale3D (Variables|Default|GetEmissive%20Projectile%20Mesh) (Math|Conversions|ToVector(Float) _grow))
            (Transformation|SetRelativeScale3D (Variables|Default|GetGliderTrail) _scalesizereduced)))))))

(event Parachute|EventOnParachuteTrailUpdated
  (params (MovementLengthSquared :type Float%20%28single-precision%29) (ForwardDot :type Float%20%28single-precision%29) (RightDot :type Float%20%28single-precision%29) (RotationalMovementAcceleration :type Float%20%28single-precision%29))
  (bind _returnvalue (CallFunction|GetTrailVFX2Component))
  (bind _returnvalue_2 (CallFunction|GetTrailVFX2Component))
  (|CallParentFunction MovementLengthSquared ForwardDot RightDot RotationalMovementAcceleration)
  (if (and (>= ForwardDot 0.2) (CallFunction|IsGliderFullyDeployedAccordingtoAnim))
    (Utilities|FlowControl|DoOnce false
      (:Completed
        (Effects|Components|ParticleSystem|SetVectorParameter _returnvalue_2 "StreakSpeed" "0, 0,60.000000")
        (Effects|Components|ParticleSystem|SetVectorParameter _returnvalue_2 "StarSpeed" "0,0.000000,300.000000")
        (Utilities|FlowControl|Sequence
          (:then_0
            _)
          (:then_1
            (bind (_direction _speedup) (Timeline|ActivatedTimeline PlayFromStart)
              (:Update
                (Rendering|Material|SetScalarParameterValue (Variables|Default|GetBlueTail) "SpeedUp" _speedup :!class /Script/Engine.MaterialInstanceDynamic)
                (Rendering|Material|SetScalarParameterValue (Variables|Default|GetCoreSpeed) "SpeedY" (- 1.0 _speedup) :!class /Script/Engine.MaterialInstanceDynamic)))))))
    (else
      _))
  (Utilities|FlowControl|DoOnce false
    (:Completed
      (Utilities|FlowControl|Sequence
        (:then_0
          _)
        (:then_1
          _))
      (if (Components|Timeline|IsPlaying (Variables|B_Athena_Glider_Legacy_SleepyTime|GetActivatedTimeline))
        (Components|Timeline|Stop (Variables|B_Athena_Glider_Legacy_SleepyTime|GetActivatedTimeline)))
      (Effects|Components|ParticleSystem|SetVectorParameter _returnvalue "StreakSpeed" "0, 0,30.000000")
      (Effects|Components|ParticleSystem|SetVectorParameter _returnvalue "StarSpeed" "0, 0,100.000000"))))

(fn ConstructionScript
  (|CallParentFunction)
  (bind _returnvalue (Rendering|Material|CreateDynamicMaterialInstance (Variables|Default|GetGliderTrail) 0 "/BRCosmetics/Gadgets/Assets/VinderTech_GliderChute/Glider_SleepyTime/FX/M_SleepyTime_Gliader_Tail.M_SleepyTime_Gliader_Tail" "Appearing" :!class /Script/Engine.PrimitiveComponent))
  (bind _output_get (Variables|Default|SetBlueTail _returnvalue))
  (Rendering|Material|SetScalarParameterValue _output_get "Appearing" -1.5 :!class /Script/Engine.MaterialInstanceDynamic)
  (bind _returnvalue_2 (Rendering|Material|CreateDynamicMaterialInstance (Variables|Default|GetEmissive%20Projectile%20Mesh) 0 "/BRCosmetics/Gadgets/Assets/VinderTech_GliderChute/Glider_SleepyTime/FX/MI_SleepyTime_BlueTailCore_Inst.MI_SleepyTime_BlueTailCore_Inst" "CoreSpeed_MID" :!class /Script/Engine.PrimitiveComponent))
  (Variables|Default|SetCoreSpeed _returnvalue_2))
