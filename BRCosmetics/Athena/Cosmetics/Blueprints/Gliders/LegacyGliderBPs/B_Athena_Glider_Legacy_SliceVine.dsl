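; Legacy SliceVine glider Blueprint that plays an uncloak and glow effect when the parachute becomes visible.
; It creates a dynamic material on the parachute mesh, then a timeline drives its uncloak, pulse, end and glow parameters.
; Each timeline update also reads the player's glider particle variant from the cosmetic loadout.
; A DoOnce spawns the triangle Niagara system at FX_Base, setting its Variant parameter from that index.
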(event EventOnGliderFullyDeployed)

(event EventOnChuteVisibilityChanged
  (params (bParachuteVisibility :type Boolean))
  (bind _parachutemesh (Variables|Parachute|GetParachuteMesh))
  (bind _mid (Variables|Default|GetMID))
  (|CallParentFunction bParachuteVisibility)
  (if bParachuteVisibility
    (Variables|Default|SetMID (Class|BTechArtGlobalLibrary|GetDynamicMaterialInstance (Variables|Parachute|GetParachuteMesh)))
    (bind (_direction _uncloak _endactive _deployedglow _uncloakpulse) (Timeline|Timeline_0 PlayFromStart)
      (:Update
        (Rendering|Material|SetScalarParameterValue _mid "BP_UnCloaking" _uncloak :!class /Script/Engine.MaterialInstanceDynamic)
        (Rendering|Material|SetScalarParameterValue _mid "EndActive" _endactive :!class /Script/Engine.MaterialInstanceDynamic)
        (Rendering|Material|SetScalarParameterValue _mid "DeployedGlow" _deployedglow :!class /Script/Engine.MaterialInstanceDynamic)
        (Rendering|Material|SetScalarParameterValue _mid "BP_UnCloakingPulse" _uncloakpulse :!class /Script/Engine.MaterialInstanceDynamic)
        ; Get active variant and creat MID
        (bind _asplayer_pawn_athena (Utilities|Casting|CastToPlayerPawn_Athena (Variables|FortPlayerParachute|GetPlayerPawn)))
        (bind _index (Class|FortCosmeticFunctionLibrary|%28Deprecated%29GetVariantChannelActiveIndexForSlot "(TagName=\"LoadoutSlot.Glider\")" (Variables|/Script/FortniteGame.FortPlayerPawn|GetCosmeticLoadout _asplayer_pawn_athena) "(TagName=\"Cosmetics.Variant.Channel.Particle\")" (Utilities|Array|MakeArray "(TagName=\"Cosmetics.Variant.Property.Particle1\")" "(TagName=\"Cosmetics.Variant.Property.Particle2\")") false false))
        (bind _output_get (Variables|Default|SetActiveVariantIndex _index)))))
  (Utilities|FlowControl|DoOnce false
    (:Completed
      (switch int _output_get
        (:0
          (bind _returnvalue (Niagara|SpawnSystemAttached "/BRCosmetics/Gadgets/Assets/VinderTech_GliderChute/Glider_SliceVine/FX/NS_Glider_Slicevine_Triangles.NS_Glider_Slicevine_Triangles" _parachutemesh "FX_Base"))
          (Effects|Components|ParticleSystem|SetFloatParameter _returnvalue "Variant" 0.0))
        (:1
          (bind _returnvalue_2 (Niagara|SpawnSystemAttached "/BRCosmetics/Gadgets/Assets/VinderTech_GliderChute/Glider_SliceVine/FX/NS_Glider_Slicevine_Triangles.NS_Glider_Slicevine_Triangles" _parachutemesh "FX_Base"))
          (Effects|Components|ParticleSystem|SetFloatParameter _returnvalue_2 "Variant" 1.0))))))

(fn ConstructionScript)
