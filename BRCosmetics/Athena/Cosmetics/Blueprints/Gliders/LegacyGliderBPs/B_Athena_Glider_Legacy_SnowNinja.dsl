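; Legacy Snow Ninja glider Blueprint that manages the glider's trail and fan effects.
; When the chute shows, both trail components attach to the parachute mesh's Frame socket with offsets.
; In front-end preview, full deployment starts P_Fan2 and, after a delay, the first trail, which stops on hide.
; Trail updates use DoOnce gates to start P_Fan2 during deployed forward movement and to stop P_Fan1.
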(event EventOnChuteVisibilityChanged
  (params (bParachuteVisibility :type Boolean))
  (|CallParentFunction bParachuteVisibility)
  (if bParachuteVisibility
    (Transformation|AttachComponentToComponent (CallFunction|GetTrailVFXComponent) (Variables|Parachute|GetParachuteMesh) "Frame" "KeepRelative" "KeepRelative" "KeepRelative" true)
    (Transformation|SetRelativeLocation (CallFunction|GetTrailVFXComponent) "0.000000,-32.334381,9.228863" false false)
    (Transformation|AttachComponentToComponent (CallFunction|GetTrailVFX2Component) (Variables|Parachute|GetParachuteMesh) "Frame" "KeepRelative" "KeepRelative" "KeepRelative" true)
    (Transformation|SetRelativeLocation (CallFunction|GetTrailVFX2Component) "0.000000,-35.634995,11.736790" false false)
    (else
      (if (Variables|FortPlayerParachute|GetbIsFrontEndPreview)
        (Components|Activation|Deactivate (CallFunction|GetTrailVFXComponent))))))

(event EventOnGliderFullyDeployed
  (|CallParentFunction)
  (if (Variables|FortPlayerParachute|GetbIsFrontEndPreview)
    (Components|Activation|Activate (Variables|Default|GetP_Fan2) false)
    (Utilities|FlowControl|Delay 1.0)
    (Components|Activation|Activate (CallFunction|GetTrailVFXComponent) false)))

(event Parachute|EventOnParachuteTrailUpdated
  (params (MovementLengthSquared :type Float%20%28single-precision%29) (ForwardDot :type Float%20%28single-precision%29) (RightDot :type Float%20%28single-precision%29) (RotationalMovementAcceleration :type Float%20%28single-precision%29))
  (|CallParentFunction MovementLengthSquared ForwardDot RightDot RotationalMovementAcceleration)
  (if (and (>= ForwardDot 0.2) (CallFunction|IsGliderFullyDeployedAccordingtoAnim))
    (Utilities|FlowControl|DoOnce false
      (:Completed
        (Components|Activation|Activate (Variables|Default|GetP_Fan2) false)))
    (else
      _))
  (Utilities|FlowControl|DoOnce false
    (:Completed
      (Components|Activation|Deactivate (Variables|Default|GetP_Fan1)))))

(fn ConstructionScript
  (|CallParentFunction))
