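; Legacy Space Shuttle glider Blueprint that adds thrust-start audio to the parachute.
; On each trail update it calls the parent, then checks for forward movement and a fully deployed glider animation.
; When both hold, a DoOnce spawns the shuttle thrust cue attached to the root component.
; A separate DoOnce fades that audio component out, and neither gate shows a reset.
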(event Parachute|EventOnParachuteTrailUpdated
  (params (MovementLengthSquared :type Float%20%28single-precision%29) (ForwardDot :type Float%20%28single-precision%29) (RightDot :type Float%20%28single-precision%29) (RotationalMovementAcceleration :type Float%20%28single-precision%29))
  (|CallParentFunction MovementLengthSquared ForwardDot RightDot RotationalMovementAcceleration)
  (if (and (>= ForwardDot 0.2) (CallFunction|IsGliderFullyDeployedAccordingtoAnim))
    (Utilities|FlowControl|DoOnce false
      (:Completed
        (bind _returnvalue (Audio|SpawnSoundAttached "/BRCosmetics/Sounds/Gliders/Athena_Glider_Shuttle_ThrustStart_Cue.Athena_Glider_Shuttle_ThrustStart_Cue" (Variables|Transformation|GetRootComponent)))))
    (else
      _))
  (Utilities|FlowControl|DoOnce false
    (:Completed
      (Audio|Components|Audio|FadeOut _returnvalue 1.0 0.0 "Linear" :!class /Script/Engine.AudioComponent))))

(fn ConstructionScript
  (|CallParentFunction))
