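; Legacy Speedy glider Blueprint that controls engine trail particles on the player's glider.
; On full deployment in front-end preview, it switches on the trail's SpawnRateScale and EngineSpawn float parameters.
; On trail updates, it zeroes those parameters based on a check of forward movement and the animation's fully-deployed state, with a DoOnce node guarding the switch-on.
; Both events call their parent implementations, and the construction script only calls its parent.
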(event EventOnGliderFullyDeployed
  (bind _returnvalue (CallFunction|GetTrailVFXComponent))
  (|CallParentFunction)
  (if (Variables|FortPlayerParachute|GetbIsFrontEndPreview)
    (Effects|Components|ParticleSystem|SetFloatParameter _returnvalue "SpawnRateScale" 1.0)
    (Effects|Components|ParticleSystem|SetFloatParameter _returnvalue "EngineSpawn" 1.0)))

(event Parachute|EventOnParachuteTrailUpdated
  (params (MovementLengthSquared :type Float%20%28single-precision%29) (ForwardDot :type Float%20%28single-precision%29) (RightDot :type Float%20%28single-precision%29) (RotationalMovementAcceleration :type Float%20%28single-precision%29))
  (bind _returnvalue (CallFunction|GetTrailVFXComponent))
  (|CallParentFunction MovementLengthSquared ForwardDot RightDot RotationalMovementAcceleration)
  (if (not (and (>= ForwardDot 0.2) (CallFunction|IsGliderFullyDeployedAccordingtoAnim)))
    (Effects|Components|ParticleSystem|SetFloatParameter _returnvalue "SpawnRateScale" 0.0)
    (Effects|Components|ParticleSystem|SetFloatParameter _returnvalue "EngineSpawn" 0.0))
  (Utilities|FlowControl|DoOnce false
    (:Completed
      (Effects|Components|ParticleSystem|SetFloatParameter _returnvalue "SpawnRateScale" 1.0)
      (Effects|Components|ParticleSystem|SetFloatParameter _returnvalue "EngineSpawn" 1.0))))

(fn ConstructionScript
  (|CallParentFunction))
