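; Storm Soldier legacy glider Blueprint that animates a material fade and movement-driven effect scaling on the parachute.
; The construction script creates a dynamic material for parachute mesh slot 2, whose Opacity follows a timeline played on full deployment and zeroed when the chute hides.
; Trail updates use forward movement and deployment state to drive the Shrink timeline, stretching FX_L and FX_M along X.
; Showing the chute also rotates the trail VFX.
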(event EventOnChuteVisibilityChanged
  (params (bParachuteVisibility :type Boolean))
  (|CallParentFunction bParachuteVisibility)
  (if bParachuteVisibility
    (Transformation|SetRelativeRotation (CallFunction|GetTrailVFXComponent) "0,90.000000, 0" false false)
    (else
      (Utilities|FlowControl|Sequence
        (:then_0
          (Rendering|Material|SetScalarParameterValue (Variables|Default|GetNewVar_0) "Opacity" 0.0 :!class /Script/Engine.MaterialInstanceDynamic))
        (:then_1
          (Utilities|FlowControl|DoOnce false
            (:Completed
              (bind (_direction _size _opacity) (Timeline|Timeline_0 Play)
                (:Update
                  (Rendering|Material|SetScalarParameterValue (Variables|Default|GetNewVar_0) "Opacity" _opacity :!class /Script/Engine.MaterialInstanceDynamic))))))))))

(event EventOnGliderFullyDeployed
  (|CallParentFunction)
  (Utilities|FlowControl|DoOnce false
    (:Completed
      (Timeline|Timeline_0 Play))))

(event Parachute|EventOnParachuteTrailUpdated
  (params (MovementLengthSquared :type Float%20%28single-precision%29) (ForwardDot :type Float%20%28single-precision%29) (RightDot :type Float%20%28single-precision%29) (RotationalMovementAcceleration :type Float%20%28single-precision%29))
  (|CallParentFunction MovementLengthSquared ForwardDot RightDot RotationalMovementAcceleration)
  (if (not (and (>= ForwardDot 0.2) (CallFunction|IsGliderFullyDeployedAccordingtoAnim)))
    (Utilities|FlowControl|Sequence
      (:then_0
        _)
      (:then_1
        (bind (_direction _size) (Timeline|Shrink Reverse)
          (:Update
            (Transformation|SetRelativeScale3D (Variables|Components|GetFX_L) (Math|Vector|MakeVector (Math|Float|Lerp 1.0 4.0 _size) 0.5 0.5))
            (Transformation|SetRelativeScale3D (Variables|Components|GetFX_M) (Math|Vector|MakeVector (Math|Float|Lerp 2.0 6.0 _size) 0.8 0.8)))))))
  (Utilities|FlowControl|DoOnce false
    (:Completed
      (Timeline|Shrink Play))))

(fn ConstructionScript
  (|CallParentFunction)
  (bind _returnvalue (Rendering|Material|CreateDynamicMaterialInstance (Variables|Parachute|GetParachuteMesh) 2 0 "None" :!class /Script/Engine.PrimitiveComponent))
  (Variables|Default|SetNewVar_0 _returnvalue))
