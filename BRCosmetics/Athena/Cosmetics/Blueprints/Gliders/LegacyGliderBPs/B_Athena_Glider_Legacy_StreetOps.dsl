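; Legacy StreetOps glider Blueprint that animates an emissive blinking light while the parachute is visible.
; Construction creates a dynamic material instance on slot 0 of the parachute mesh and stores it as MID.
; When the chute shows, a timeline moved to the during-physics tick group plays from start, lerping the MID's EmissiveMultiplier toward the blinking light multiplier.
; When the chute hides, the timeline stops.
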(event EventOnChuteVisibilityChanged
  (params (bParachuteVisibility :type Boolean))
  (|CallParentFunction bParachuteVisibility)
  (if bParachuteVisibility
    (Components|Tick|SetTickGroup (Variables|B_Athena_Glider_Legacy_StreetOps|GetTimeline_0) "TG_DuringPhysics")
    (bind (_direction _emissivecolor) (Timeline|Timeline_0 PlayFromStart)
      (:Update
        (Rendering|Material|SetVectorParameterValue (Variables|Default|GetMID) "EmissiveMultiplier" (Math|Color|Lerp(LinearColor) :B (Variables|Default|GetBlinking%20Light%20Emissive%20Multiplier) :Alpha _emissivecolor) :!class /Script/Engine.MaterialInstanceDynamic)))
    (else
      (Timeline|Timeline_0 Stop))))

(fn ConstructionScript
  (|CallParentFunction)
  (bind _returnvalue (Rendering|Material|CreateDynamicMaterialInstance (Variables|Parachute|GetParachuteMesh) 0 (Variables|Default|GetMID%20Base%20Material) "StreetOpsGlider - EmissiveMultiplier" :!class /Script/Engine.PrimitiveComponent))
  (Variables|Default|SetMID _returnvalue))
