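; Glider Blueprint for the Street Racer Drift Remix cosmetic, adding an erode reveal, lightning and socket-to-socket drift beams.
; Showing the chute plays an erode timeline on a dynamic material; full deployment later swaps in the final material and starts idle lightning.
; BeamSetup updates beam endpoints, checking effects quality to limit beam count, from trail updates or a front-end preview timer.
; Preview also flickers emissive color; forward movement offsets the lightning.
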(event EventOnGliderFullyDeployed
  (|CallParentFunction)
  (Utilities|FlowControl|Delay 0.45)
  (Rendering|Material|SetMaterial (Variables|Parachute|GetParachuteMesh) 0 "/BRCosmetics/Gadgets/Assets/VinderTech_GliderChute/Glider_Street_Racer_Drift_Remix/Materials/MI_Glider_Street_Racer_Drift_Remix.MI_Glider_Street_Racer_Drift_Remix")
  (Components|Activation|Activate (Variables|Default|GetLightning_Idle) false)
  (if (Variables|FortPlayerParachute|GetbIsFrontEndPreview)
    (bind _returnvalue (Utilities|Time|SetTimerbyEvent (EventDispatchers|CreateEvent :!event Custom|BeamSetup) 0.1 true))
    (Variables|Default|SetFrontendTimer _returnvalue)
    (bind (_direction _flick) (Timeline|Flicker PlayFromStart)
      (:Update
        (Rendering|Material|SetVectorParameterValueonMaterials (Variables|Parachute|GetParachuteMesh) "EmissiveMultiplier" (Math|Vector|Lerp(Vector) (Math|Vector|Normalize (Math|Conversions|ToVector(LinearColor) (Variables|Default|GetGliderEmissiveColor))) (Math|Conversions|ToVector(LinearColor) (Variables|Default|GetGliderEmissiveColor)) _flick))))
    (else
      (Timeline|Flicker PlayFromStart))))

(event Custom|BeamSetup
  (bind _drift_beam (Variables|Default|GetDrift_Beam))
  (bind _parachutemesh (Variables|Parachute|GetParachuteMesh))
  (bind _returnvalue (Settings|GetGameUserSettings))
  (Variables|Default|SetQualityCheck (>= (Settings|GetVisualEffectQuality _returnvalue) 2))
  (bind (_array_element _array_index) (Utilities|Array|ForEachLoopwithBreak (Variables|Default|GetBeam%20Target%20Points))
    (:LoopBody
      (Effects|Components|ParticleSystem|SetBeamTargetPoint _drift_beam _array_index (Transformation|GetSocketLocation _parachutemesh _array_element) 0)
      (Effects|Components|ParticleSystem|SetBeamSourcePoint _drift_beam _array_index (Transformation|GetSocketLocation _parachutemesh (Utilities|Array|Get(acopy) (Variables|Default|GetBeam%20Source%20Points) _array_index)) 0 :!class /Script/CascadeRuntime.ParticleSystemComponent)
      (if (and (not (Variables|Default|GetQualityCheck)) (>= _array_index 7))
        _))
    (:Completed
      _))
  (if (CallFunction|IsGliderFullyDeployedAccordingtoAnim)
    (Components|Activation|Activate (Variables|Default|GetDrift_Beam) false)))

(event EventOnChuteVisibilityChanged
  (params (bParachuteVisibility :type Boolean))
  (bind _mid (Variables|Default|GetMID))
  (|CallParentFunction bParachuteVisibility)
  (if bParachuteVisibility
    (Rendering|Material|SetMaterial (Variables|Parachute|GetParachuteMesh) 0 _mid)
    (Rendering|Material|SetVectorParameterValue _mid "PointOffset" (select (Variables|FortPlayerParachute|GetbIsFrontEndPreview) (Math|Conversions|ToLinearColor(Vector) (Variables|Default|GetFE_ErodeOffset)) (Math|Conversions|ToLinearColor(Vector) (Variables|Default|GetInGame_ErodeOffset))) :!class /Script/Engine.MaterialInstanceDynamic)
    (bind (_direction _uncloak) (Timeline|Timeline_0 PlayFromStart)
      (:Update
        (Rendering|Material|SetScalarParameterValue (Variables|Default|GetMID) "Erode" _uncloak :!class /Script/Engine.MaterialInstanceDynamic)))
    (else
      (Components|Activation|Deactivate (Variables|Default|GetLightning_Idle)))))

(event Parachute|EventOnParachuteTrailUpdated
  (params (MovementLengthSquared :type Float%20%28single-precision%29) (ForwardDot :type Float%20%28single-precision%29) (RightDot :type Float%20%28single-precision%29) (RotationalMovementAcceleration :type Float%20%28single-precision%29))
  (bind _returnvalue (CallFunction|IsGliderFullyDeployedAccordingtoAnim))
  (bind _lightning_idle (Variables|Default|GetLightning_Idle))
  (|CallParentFunction MovementLengthSquared ForwardDot RightDot RotationalMovementAcceleration)
  (if _returnvalue
    (CallFunction|BeamSetup)
    (if (and (>= ForwardDot 0.2) _returnvalue)
      (Effects|Components|ParticleSystem|SetVectorParameter _lightning_idle "LocOffsetL" "0,-18.000000, 0")
      (else
        (Effects|Components|ParticleSystem|SetVectorParameter _lightning_idle "LocOffsetL" "0, 0, 0")))))

(fn ConstructionScript
  (|CallParentFunction)
  (bind _returnvalue (Rendering|Material|CreateDynamicMaterialInstance "/BRCosmetics/Gadgets/Assets/VinderTech_GliderChute/Glider_Street_Racer_Drift_Remix/FX/M_Glider_StreetRacer_DriftRemix_Test_Inst.M_Glider_StreetRacer_DriftRemix_Test_Inst" "StreetRacerDriftRemixGlider_MID" "None" :!class /Script/Engine.KismetMaterialLibrary))
  (Variables|Default|SetMID _returnvalue))
