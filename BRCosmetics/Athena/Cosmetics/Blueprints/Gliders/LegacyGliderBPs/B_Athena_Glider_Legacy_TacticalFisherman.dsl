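; Legacy glider Blueprint for the Tactical Fisherman raft glider, driving its trail, drip and engine effects.
; When shown, it attaches both trail VFX components to the Raft_L and Raft_R sockets with mirrored offsets; hiding stops the drips.
; Full deployment starts the drips, plus the engine in front-end preview.
; Trail updates stop the engine without sufficient forward movement or full deployment, with activation gated by a DoOnce.
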(event EventOnChuteVisibilityChanged
  (params (bParachuteVisibility :type Boolean))
  (bind _returnvalue (CallFunction|GetTrailVFX2Component))
  (bind _returnvalue_2 (CallFunction|GetTrailVFXComponent))
  (|CallParentFunction bParachuteVisibility)
  (if bParachuteVisibility
    (Transformation|AttachComponentToComponent _returnvalue_2 (Variables|Parachute|GetParachuteMesh) "Raft_L" "KeepRelative" "KeepRelative" "KeepRelative" true)
    (Transformation|SetRelativeLocationAndRotation _returnvalue_2 "-10.000000,10.000000, 0" "0, 0, 0" false false)
    (Transformation|AttachComponentToComponent _returnvalue (Variables|Parachute|GetParachuteMesh) "Raft_R" "KeepRelative" "KeepRelative" "KeepRelative" true)
    (Transformation|SetRelativeLocationAndRotation _returnvalue "10.000000,-10.000000, 0" "0,180.000000, 0" false false)
    (else
      (Components|Activation|Deactivate (Variables|Default|GetDrips_R))
      (if (Variables|FortPlayerParachute|GetbIsFrontEndPreview)
        (Components|Activation|Deactivate (Variables|Default|GetEngine))))))

(event Parachute|EventOnParachuteTrailUpdated
  (params (MovementLengthSquared :type Float%20%28single-precision%29) (ForwardDot :type Float%20%28single-precision%29) (RightDot :type Float%20%28single-precision%29) (RotationalMovementAcceleration :type Float%20%28single-precision%29))
  (|CallParentFunction MovementLengthSquared ForwardDot RightDot RotationalMovementAcceleration)
  (if (not (and (>= ForwardDot 0.2) (CallFunction|IsGliderFullyDeployedAccordingtoAnim)))
    (Components|Activation|Deactivate (Variables|Default|GetEngine)))
  (Utilities|FlowControl|DoOnce false
    (:Completed
      (Components|Activation|Activate (Variables|Default|GetEngine) false))))

(event EventOnGliderFullyDeployed
  (|CallParentFunction)
  (Components|Activation|Activate (Variables|Default|GetDrips_R) false)
  (if (Variables|FortPlayerParachute|GetbIsFrontEndPreview)
    (Components|Activation|Activate (Variables|Default|GetEngine) false)))

(fn ConstructionScript
  (|CallParentFunction))
