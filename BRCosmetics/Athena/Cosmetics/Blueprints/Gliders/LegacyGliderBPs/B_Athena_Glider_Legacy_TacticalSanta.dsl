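; Legacy Tactical Santa glider Blueprint that drives jet flame, center and light particle effects.
; On full deployment it snaps flame and center effects to the parachute mesh's inner jet joint and activates the light.
; Visibility changes deactivate the flame when the chute shows and the light when it hides.
; Trail updates zero flame thrust parameters unless deployed and moving forward, with a DoOnce setting them to full.
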(event EventOnGliderFullyDeployed
  (Transformation|AttachComponentToComponent (Variables|Default|GetP_Glider_TacticalSanta_flame) (Variables|Parachute|GetParachuteMesh) "jetInner_jnt" "SnapToTarget" "SnapToTarget" "KeepRelative" true)
  (Transformation|AttachComponentToComponent (Variables|Default|GetP_Glider_TacticalSanta_center) (Variables|Parachute|GetParachuteMesh) "jetInner_jnt" "SnapToTarget" "SnapToTarget" "KeepRelative" true)
  (Transformation|SetRelativeLocationAndRotation (Variables|Default|GetP_Glider_TacticalSanta_center) "0.000000,-2.387752,0.000000" "0,90.000000, 0" false false)
  (|CallParentFunction)
  (Components|Activation|Activate (Variables|Default|GetP_Glider_TacticalSanta_light) true))

(event EventOnChuteVisibilityChanged
  (params (bParachuteVisibility :type Boolean))
  (|CallParentFunction bParachuteVisibility)
  (if bParachuteVisibility
    (Components|Activation|Deactivate (Variables|Default|GetP_Glider_TacticalSanta_flame))
    (else
      (Components|Activation|Deactivate (Variables|Default|GetP_Glider_TacticalSanta_light)))))

(event Parachute|EventOnParachuteTrailUpdated
  (params (MovementLengthSquared :type Float%20%28single-precision%29) (ForwardDot :type Float%20%28single-precision%29) (RightDot :type Float%20%28single-precision%29) (RotationalMovementAcceleration :type Float%20%28single-precision%29))
  (bind _p_glider_tacticalsanta_flame (Variables|Default|GetP_Glider_TacticalSanta_flame))
  (|CallParentFunction MovementLengthSquared ForwardDot RightDot RotationalMovementAcceleration)
  (if (not (and (>= ForwardDot 0.2) (CallFunction|IsGliderFullyDeployedAccordingtoAnim)))
    (Effects|Components|ParticleSystem|SetFloatParameter _p_glider_tacticalsanta_flame "Thrust" 0.0)
    (Effects|Components|ParticleSystem|SetFloatParameter _p_glider_tacticalsanta_flame "ThrustCore" 0.0)
    (Effects|Components|ParticleSystem|SetFloatParameter _p_glider_tacticalsanta_flame "ThrustSmoke" 0.0))
  (Utilities|FlowControl|DoOnce false
    (:Completed
      (Effects|Components|ParticleSystem|SetFloatParameter _p_glider_tacticalsanta_flame "Thrust" 1.0)
      (Effects|Components|ParticleSystem|SetFloatParameter _p_glider_tacticalsanta_flame "ThrustCore" 1.0)
      (Effects|Components|ParticleSystem|SetFloatParameter _p_glider_tacticalsanta_flame "ThrustSmoke" 1.0))))

(fn ConstructionScript
  (|CallParentFunction))
