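; Legacy TechMage glider Blueprint that plays an uncloaking effect when the chute becomes visible.
; Construction creates an uncloak dynamic material, applied on show with direction flipped for front-end preview.
; A timeline animates its cloak and glow parameters and stores the particle variant index; full deployment swaps to the standard material.
; A DoOnce on visibility change attaches borrowed SliceVine triangle Niagara FX with a fixed variant, independent of that index.
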(event EventOnChuteVisibilityChanged
  (params (bParachuteVisibility :type Boolean))
  (bind _mid (Variables|Default|GetMID))
  (bind _mid_2 (Variables|Default|GetMID))
  (|CallParentFunction bParachuteVisibility)
  (if bParachuteVisibility
    (Rendering|Material|SetMaterial (Variables|Parachute|GetParachuteMesh) 0 _mid)
    (Rendering|Material|SetScalarParameterValue _mid "UnCloakingDirection" (Math|Conversions|ToFloat(Integer) (select (Variables|FortPlayerParachute|GetbIsFrontEndPreview) 1 -1)) :!class /Script/Engine.MaterialInstanceDynamic)
    (bind (_direction _uncloak _endactive _deployedglow _newtrack_1) (Timeline|Timeline_0 PlayFromStart)
      (:Update
        (Rendering|Material|SetScalarParameterValue _mid_2 "BP_UnCloaking" _uncloak :!class /Script/Engine.MaterialInstanceDynamic)
        (Rendering|Material|SetScalarParameterValue _mid_2 "EndActive" _endactive :!class /Script/Engine.MaterialInstanceDynamic)
        (Rendering|Material|SetScalarParameterValue _mid_2 "DeployedGlow" _deployedglow :!class /Script/Engine.MaterialInstanceDynamic)
        (Rendering|Material|SetScalarParameterValue _mid_2 "BP_UnCloakingPulse" _newtrack_1 :!class /Script/Engine.MaterialInstanceDynamic)
        ; Setting Variant Parm
        (bind _asplayer_pawn_athena (Utilities|Casting|CastToPlayerPawn_Athena (Variables|FortPlayerParachute|GetPlayerPawn)))
        (bind _index (CosmeticLoadout|GetVariantChannelActiveIndexforSlotActor "(TagName=\"LoadoutSlot.Glider\")" _asplayer_pawn_athena "(TagName=\"Cosmetics.Variant.Channel.Particle\")" (Utilities|Array|MakeArray "(TagName=\"Cosmetics.Variant.Property.Particle1\")" "(TagName=\"\")") false false))
        (Variables|Default|SetActive%20Index _index))))
  (Utilities|FlowControl|DoOnce false
    (:Completed
      (bind _returnvalue (Niagara|SpawnSystemAttached "/BRCosmetics/Gadgets/Assets/VinderTech_GliderChute/Glider_SliceVine/FX/NS_Glider_Slicevine_Triangles.NS_Glider_Slicevine_Triangles" (Variables|Parachute|GetParachuteMesh) "FX_Base"))
      (Effects|Components|ParticleSystem|SetFloatParameter _returnvalue "Variant" 2.0))))

(event EventOnGliderFullyDeployed
  (|CallParentFunction)
  (Rendering|Material|SetMaterial (Variables|Parachute|GetParachuteMesh) 0 "/BRCosmetics/Gadgets/Assets/VinderTech_GliderChute/Glider_TechMage/Materials/MI_Glider_TechMage.MI_Glider_TechMage"))

(fn ConstructionScript
  (|CallParentFunction)
  (bind _returnvalue (Rendering|Material|CreateDynamicMaterialInstance "/BRCosmetics/Gadgets/Assets/VinderTech_GliderChute/Glider_TechMage/FX/MI_Glider_TechMage_Uncloak_Final.MI_Glider_TechMage_Uncloak_Final" "TechMageGlider_CloakingMID" "None" :!class /Script/Engine.KismetMaterialLibrary))
  (Variables|Default|SetMID _returnvalue))
