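; Legacy TechOps glider Blueprint that positions its two trail VFX components on the glider model.
; When the parachute becomes visible, it calls the parent handler, then sets fixed relative transforms on both trail components, rotated 180 degrees in yaw, with the first also scaled up.
; The fully-deployed and trail-update overrides call their parents and evaluate preview, forward-movement and deployment conditions, but their branches are empty.
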(event EventOnGliderFullyDeployed
  (|CallParentFunction)
  (if (Variables|FortPlayerParachute|GetbIsFrontEndPreview)
    _))

(event EventOnChuteVisibilityChanged
  (params (bParachuteVisibility :type Boolean))
  (|CallParentFunction bParachuteVisibility)
  (if bParachuteVisibility
    (Transformation|SetRelativeTransform (CallFunction|GetTrailVFXComponent) (struct :Location "51.500000,-34.000000,-7.000000" :Rotation "0.000000,180.000000,0.000000" :Scale "1.500000,1.500000,1.500000") false false)
    (Transformation|SetRelativeLocationAndRotation (CallFunction|GetTrailVFX2Component) "55.000000,22.000000,-1.000000" "0.000000,180.000000,0.000000" false false)))

(event Parachute|EventOnParachuteTrailUpdated
  (params (MovementLengthSquared :type Float%20%28single-precision%29) (ForwardDot :type Float%20%28single-precision%29) (RightDot :type Float%20%28single-precision%29) (RotationalMovementAcceleration :type Float%20%28single-precision%29))
  (|CallParentFunction MovementLengthSquared ForwardDot RightDot RotationalMovementAcceleration)
  (if (and (> ForwardDot -0.85) (CallFunction|IsGliderFullyDeployedAccordingtoAnim))
    _
    (else
      _)))

(fn ConstructionScript
  (|CallParentFunction))
