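; Legacy glider cosmetic Blueprint for TeriyakiFish that adds movement-reactive music and manages trail VFX.
; On full deploy it fades in the secondary vehicle audio voice and registers it with the MusicEventSubsystem, removing it when undeployed.
; Trail updates switch music volume between two levels by thresholding forward movement, optionally absolute, and mute it unless the animation reports full deployment.
; Trail VFX activates on deploy in front-end preview and deactivates when the chute hides.
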(event Parachute|EventOnParachuteTrailUpdated
  (params (MovementLengthSquared :type Float%20%28single-precision%29) (ForwardDot :type Float%20%28single-precision%29) (RightDot :type Float%20%28single-precision%29) (RotationalMovementAcceleration :type Float%20%28single-precision%29))
  (|CallParentFunction MovementLengthSquared ForwardDot RightDot RotationalMovementAcceleration)
  (Default|AudioUpdateMusic ForwardDot 0.5 1.0))

(event EventOnGliderFullyDeployed
  (bind _fortvehicleaudiovoicesecondary (Variables|Default|GetFortVehicleAudioVoiceSecondary))
  (|CallParentFunction)
  (Utilities|FlowControl|Sequence
    (:then_0
      (if (Variables|FortPlayerParachute|GetbIsFrontEndPreview)
        (Components|Activation|Activate (CallFunction|GetTrailVFXComponent) false)))
    (:then_1
      (Audio|Components|Audio|FadeIn _fortvehicleaudiovoicesecondary 0.25 1.0 0.0 "Linear" :!class /Script/Engine.AudioComponent)
      (Audio|AddEventbyComponent (GameInstanceSubsystems|GetMusicEventSubsystem) _fortvehicleaudiovoicesecondary "(TagName=\"MusicEventSubsystem.Event.Glider\")" "(GameplayTags=)" 0 ""))))

(event EventCheckForMusicEvent
  (params (GliderFullyDeployed :type Boolean))
  (if (not GliderFullyDeployed)
    (Audio|RemoveEvent (GameInstanceSubsystems|GetMusicEventSubsystem) (Variables|Default|GetFortVehicleAudioVoiceSecondary))))

(event EventOnChuteVisibilityChanged
  (params (bParachuteVisibility :type Boolean))
  (|CallParentFunction bParachuteVisibility)
  (if (not bParachuteVisibility)
    (Components|Activation|Deactivate (CallFunction|GetTrailVFXComponent))))

(fn AudioUpdateMusic
  (params (ForwardDot :type Float%20%28double-precision%29) (MusicVolMin :type Float%20%28double-precision%29) (MusicVolMax :type Float%20%28double-precision%29))
  (bind _fortvehicleaudiovoicesecondary (Variables|Default|GetFortVehicleAudioVoiceSecondary))
  (Variables|/Script/Engine.AudioComponent|GetSound (Variables|Default|GetFortVehicleAudioVoiceSecondary)
    (:then
      (if (CallFunction|IsGliderFullyDeployedAccordingtoAnim)
        (Class|FortVehicleAudioVoice|SetParam _fortvehicleaudiovoicesecondary (select (>= (select (Variables|LegacyAudio|GetAbsolute%20Movement_Audio) (Math|Float|Absolute(Float) ForwardDot) ForwardDot) 0.2) MusicVolMax MusicVolMin))
        (else
          (Class|FortVehicleAudioVoice|SetParam _fortvehicleaudiovoicesecondary 0.0))))
    (:else)))

(fn ConstructionScript
  (|CallParentFunction))
