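; Legacy Athena glider Blueprint for the TwinDark glider cosmetic.
; It overrides the fully deployed, chute visibility and parachute trail update events, plus the construction script, but each only forwards to the parent implementation.
; The graphs add no logic of their own, so all glider behavior shown here comes from the parent class.
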(event EventOnGliderFullyDeployed
  (|CallParentFunction))

(event EventOnChuteVisibilityChanged
  (params (bParachuteVisibility :type Boolean))
  (|CallParentFunction bParachuteVisibility))

(event Parachute|EventOnParachuteTrailUpdated
  (params (MovementLengthSquared :type Float%20%28single-precision%29) (ForwardDot :type Float%20%28single-precision%29) (RightDot :type Float%20%28single-precision%29) (RotationalMovementAcceleration :type Float%20%28single-precision%29))
  (|CallParentFunction MovementLengthSquared ForwardDot RightDot RotationalMovementAcceleration))

(fn ConstructionScript
  (|CallParentFunction))
