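; Legacy Season 10 umbrella glider Blueprint that adjusts where a trail effect sits when the glider's canopy appears.
; When chute visibility changes, it runs the parent handler, and if the parachute is shown it attaches the second trail VFX component to the blades_mstr socket on the parachute mesh.
; It then nudges that component to a small relative offset along Z; the construction script only calls the parent.
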(event EventOnChuteVisibilityChanged
  (params (bParachuteVisibility :type Boolean))
  (bind _returnvalue (CallFunction|GetTrailVFX2Component))
  (|CallParentFunction bParachuteVisibility)
  (if bParachuteVisibility
    (Transformation|AttachComponentToComponent _returnvalue (Variables|Parachute|GetParachuteMesh) "blades_mstr" "KeepRelative" "KeepRelative" "KeepRelative" true)
    (Transformation|SetRelativeLocation _returnvalue "0, 0,5.000000" false false)))

(fn ConstructionScript
  (|CallParentFunction))
