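; Legacy Season 11 umbrella glider Blueprint for Athena, handling its trail effect when the canopy's visibility changes.
; On the chute visibility event it runs the parent logic first, then, only if the parachute is visible, attaches the trail VFX component to the root socket of the parachute mesh and activates it without a reset.
; The construction script only calls the parent implementation.
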(event EventOnChuteVisibilityChanged
  (params (bParachuteVisibility :type Boolean))
  (|CallParentFunction bParachuteVisibility)
  (if bParachuteVisibility
    (Transformation|AttachComponentToComponent (CallFunction|GetTrailVFXComponent) (Variables|Parachute|GetParachuteMesh) "root" "SnapToTarget" "SnapToTarget" "KeepRelative" true)
    (Components|Activation|Activate (CallFunction|GetTrailVFXComponent) false)))

(fn ConstructionScript
  (|CallParentFunction))
