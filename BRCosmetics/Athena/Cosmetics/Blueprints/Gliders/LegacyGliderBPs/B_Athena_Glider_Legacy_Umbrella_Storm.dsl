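; Legacy glider Blueprint for the Storm umbrella cosmetic, adding a dissolve effect to the parachute mesh on top of the parent glider behavior.
; When the glider fully deploys, a timeline plays from the start and drives the DissolveGradientPercentage material parameter on the parachute mesh.
; When the chute becomes hidden, it stops the timeline, resets the dissolve parameter to zero and deactivates the trail VFX component.
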(event EventOnGliderFullyDeployed
  (|CallParentFunction)
  (bind (_direction _opacity) (Timeline|Timeline_0 PlayFromStart)
    (:Update
      (Rendering|Material|SetScalarParameterValueonMaterials (Variables|Parachute|GetParachuteMesh) "DissolveGradientPercentage" _opacity))))

(event EventOnChuteVisibilityChanged
  (params (bParachuteVisibility :type Boolean))
  (|CallParentFunction bParachuteVisibility)
  (if (not bParachuteVisibility)
    (Utilities|FlowControl|Sequence
      (:then_0
        (Timeline|Timeline_0 Stop))
      (:then_1
        (Rendering|Material|SetScalarParameterValueonMaterials (Variables|Parachute|GetParachuteMesh) "DissolveGradientPercentage" 0.0)
        (Components|Activation|Deactivate (CallFunction|GetTrailVFXComponent))))))

(fn ConstructionScript
  (|CallParentFunction))
