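; Legacy Valentines glider cosmetic Blueprint that plays a ribbon and tassel VFX sequence when the glider fully deploys.
; On full deployment it activates the heart ribbon effect and runs the Deploy Sequence timeline from the start.
; Rounded, clamped timeline tracks index bone-name arrays, snapping the heart ribbon and left/right tassel effects along parachute mesh sockets each update.
; When the timeline finishes, only the heart ribbon effect is deactivated.
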(event EventOnGliderFullyDeployed
  (|CallParentFunction)
  (Components|Activation|Activate (Variables|Default|GetDeploySequenceVFX_HeartRibbon) false)
  (Variables|Default|SetLocation 0)
  (bind (_direction _controler _heartribbonno _tasselribbonno) (Timeline|Deploy%20Sequence PlayFromStart)
    (:Update
      (Transformation|SetWorldLocationAndRotation (Variables|Default|GetDeploySequenceVFX_HeartRibbon) (Transformation|GetSocketLocation (Variables|Parachute|GetParachuteMesh) (Utilities|Array|Get(acopy) (Variables|Default|GetRibbonBoneNames) (Utilities|ClampMaxInteger (Math|Float|Round _heartribbonno) 49))) (Transformation|GetSocketRotation (Variables|Parachute|GetParachuteMesh) (Utilities|Array|Get(acopy) (Variables|Default|GetRibbonBoneNames) (Utilities|ClampMaxInteger (Math|Float|Round _heartribbonno) 49))) false false)
      (Transformation|SetWorldLocationAndRotation (Variables|Default|GetDeploySequenceVFX_TasselL) (Transformation|GetSocketLocation (Variables|Parachute|GetParachuteMesh) (Utilities|Array|Get(acopy) (Variables|Default|GetTassel_L_BoneNames) (Utilities|ClampMaxInteger (Math|Float|Round _tasselribbonno) 9))) (Transformation|GetSocketRotation (Variables|Parachute|GetParachuteMesh) (Utilities|Array|Get(acopy) (Variables|Default|GetTassel_L_BoneNames) (Utilities|ClampMaxInteger (Math|Float|Round _tasselribbonno) 9))) false false)
      (Transformation|SetWorldLocationAndRotation (Variables|Default|GetDeploySequenceVFX_TasselR) (Transformation|GetSocketLocation (Variables|Parachute|GetParachuteMesh) (Utilities|Array|Get(acopy) (Variables|Default|GetTassel_R_BoneNames) (Utilities|ClampMaxInteger (Math|Float|Round _tasselribbonno) 9))) (Transformation|GetSocketRotation (Variables|Parachute|GetParachuteMesh) (Utilities|Array|Get(acopy) (Variables|Default|GetTassel_R_BoneNames) (Utilities|ClampMaxInteger (Math|Float|Round _tasselribbonno) 9))) false false))
    (:Finished
      (Components|Activation|Deactivate (Variables|Default|GetDeploySequenceVFX_HeartRibbon)))))

(fn ConstructionScript
  (|CallParentFunction))
