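; Legacy Vampire glider Blueprint that drives movement-reactive audio and manages how the parachute mesh hides.
; Trail updates feed absolute forward movement and deployment state into an audio voice parameter, silent until fully deployed.
; The first hide event only caches mesh materials; later ones hide the mesh, swap in a hide material, then restore the cache after a short delay.
; Construction applies that hide material to one mesh slot.
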(event Custom|CE_DeployUpdate)

(event EventBeginPlay)

(event Parachute|EventOnParachuteTrailUpdated
  (params (MovementLengthSquared :type Float%20%28single-precision%29) (ForwardDot :type Float%20%28single-precision%29) (RightDot :type Float%20%28single-precision%29) (RotationalMovementAcceleration :type Float%20%28single-precision%29))
  (CallFunction|AudioUpdate (>= (Math|Float|Absolute(Float) ForwardDot) 0.2) (CallFunction|IsGliderFullyDeployedAccordingtoAnim)))

(event EventOnGliderFullyDeployed
  (|CallParentFunction))

(event Custom|Audio%20Update
  (params (Moving%20Forward :type Boolean) (Is%20Fully%20Deployed :type Boolean))
  (bind _fortvehicleaudiovoice (Variables|LegacyAudio|GetFortVehicleAudioVoice))
  (if Is%20Fully%20Deployed
    (Class|FortVehicleAudioVoice|SetParam _fortvehicleaudiovoice (select Moving%20Forward 1.0 0.5))
    (else
      (Class|FortVehicleAudioVoice|SetParam _fortvehicleaudiovoice 0.0))))

(event EventOnChuteVisibilityChanged
  (params (bParachuteVisibility :type Boolean))
  (bind _parachutemesh (Variables|Parachute|GetParachuteMesh))
  (bind _returnvalue (Rendering|Material|GetMaterials _parachutemesh))
  (|CallParentFunction bParachuteVisibility)
  (if bParachuteVisibility
    (Development|SetHiddeninGame (Variables|Parachute|GetParachuteMesh) false false)
    (elif (Variables|Default|GetbMaterialsCaptured)
      (Development|SetHiddeninGame _parachutemesh true false)
      (for (_ _array_index) _returnvalue
        (Rendering|Material|SetMaterial _parachutemesh _array_index "/BRCosmetics/Accessories/FORT_Backpacks/M_MED_GhostPortal_Backpack/FX/M_HideTemp.M_HideTemp"))
      (Utilities|FlowControl|Delay 0.4)
      (for (_array_element _array_index_2) (Variables|Default|GetGliderMaterials)
        (Rendering|Material|SetMaterial (Variables|Parachute|GetParachuteMesh) _array_index_2 _array_element))
      (else
        (Variables|Default|SetGliderMaterials _returnvalue)
        (Variables|Default|SetbMaterialsCaptured true)))))

(event Custom|CE_EndTrail)

(event Custom|CE_StartTrail)

(fn ConstructionScript
  (|CallParentFunction)
  (Rendering|Material|SetMaterial (Variables|Parachute|GetParachuteMesh) 3 "/BRCosmetics/Accessories/FORT_Backpacks/M_MED_GhostPortal_Backpack/FX/M_HideTemp.M_HideTemp"))
