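; Glider Blueprint for the legacy Viper cosmetic, driving material fades and idle effects through deployment.
; Construction creates dynamic materials for the parachute mesh and four Viper components, starting the mesh at zero opacity.
; Showing the chute runs a timeline driving Growth and Opacity; full deployment raises the first smoke material's Alpha and activates the gameplay or front-end idle component.
; Hiding zeroes that Alpha and Opacity and deactivates only the gameplay idle.
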(event EventOnChuteVisibilityChanged
  (params (bParachuteVisibility :type Boolean))
  (bind _mid (Variables|Default|GetMID))
  (|CallParentFunction bParachuteVisibility)
  (if bParachuteVisibility
    (bind (_direction _growth _opacity) (Timeline|Timeline_0 PlayFromStart)
      (:Update
        (Rendering|Material|SetScalarParameterValue _mid "Growth" _growth :!class /Script/Engine.MaterialInstanceDynamic)
        (Rendering|Material|SetScalarParameterValue _mid "Opacity" _opacity :!class /Script/Engine.MaterialInstanceDynamic)))
    (else
      (Rendering|Material|SetScalarParameterValue (Variables|Default|GetSmoke%20Alpha_01) "Alpha" 0.0 :!class /Script/Engine.MaterialInstanceDynamic)
      (Components|Activation|Deactivate (Variables|Default|GetIdle))
      (Rendering|Material|SetScalarParameterValue (Variables|Default|GetMID) "Opacity" 0.0 :!class /Script/Engine.MaterialInstanceDynamic))))

(event EventOnGliderFullyDeployed
  (|CallParentFunction)
  (Rendering|Material|SetScalarParameterValue (Variables|Default|GetSmoke%20Alpha_01) "Alpha" 2.0 :!class /Script/Engine.MaterialInstanceDynamic)
  (Components|Activation|Activate (select (Variables|FortPlayerParachute|GetbIsFrontEndPreview) (Variables|Default|GetIdle_Frontend) (Variables|Default|GetIdle)) false))

(fn ConstructionScript
  (|CallParentFunction)
  (bind _returnvalue (Rendering|Material|CreateDynamicMaterialInstance (Variables|Parachute|GetParachuteMesh) 1 0 "None" :!class /Script/Engine.PrimitiveComponent))
  (bind _output_get (Variables|Default|SetMID _returnvalue))
  (Rendering|Material|SetScalarParameterValue _output_get "Opacity" 0.0 :!class /Script/Engine.MaterialInstanceDynamic)
  (bind _returnvalue_2 (Rendering|Material|CreateDynamicMaterialInstance (Variables|Default|GetGlider_Viper_01) 0 0 "None" :!class /Script/Engine.PrimitiveComponent))
  (Variables|Default|SetSmoke%20Alpha_01 _returnvalue_2)
  (bind _returnvalue_3 (Rendering|Material|CreateDynamicMaterialInstance (Variables|Default|GetGlider_Viper_03) 0 0 "None" :!class /Script/Engine.PrimitiveComponent))
  (Variables|Default|SetSmoke%20Alpha_02 _returnvalue_3)
  (bind _returnvalue_4 (Rendering|Material|CreateDynamicMaterialInstance (Variables|Default|GetGlider_Viper_02) 0 0 "None" :!class /Script/Engine.PrimitiveComponent))
  (Variables|Default|SetSmoke%20Alpha_03 _returnvalue_4)
  (bind _returnvalue_5 (Rendering|Material|CreateDynamicMaterialInstance (Variables|Default|GetGlider_Viper_04) 0 0 "None" :!class /Script/Engine.PrimitiveComponent))
  (Variables|Default|SetSmoke%20Alpha_04 _returnvalue_5))
