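; Legacy Witch glider Blueprint that adds a sparkle contrail effect tied to the visibility of the glider's chute.
; On a chute visibility change it runs the parent handling, then activates the witch sparkle component when the chute is visible and deactivates it when hidden.
; The construction script only calls the parent, so this toggle is the only logic the graph adds.
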(event EventOnChuteVisibilityChanged
  (params (bParachuteVisibility :type Boolean))
  (bind _p_glidercontrail_witch_sparkles (Variables|Default|GetP_GliderContrail_Witch_Sparkles))
  (|CallParentFunction bParachuteVisibility)
  (if bParachuteVisibility
    (Components|Activation|Activate _p_glidercontrail_witch_sparkles false)
    (else
      (Components|Activation|Deactivate _p_glidercontrail_witch_sparkles))))

(fn ConstructionScript
  (|CallParentFunction))
