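; Legacy Yeti glider Blueprint that drives the glider's snow particle effects on top of its parent glider logic.
; The idle snow effect starts on full deployment and stops when the parachute is hidden.
; Trail updates deactivate the second snow effect unless the glider moves sufficiently forward while fully deployed by animation, with its activation gated by a DoOnce.
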(event EventOnGliderFullyDeployed
  (|CallParentFunction)
  (Components|Activation|Activate (Variables|Default|GetP_Glider_Yeti_Snow_Idle) false))

(event EventOnChuteVisibilityChanged
  (params (bParachuteVisibility :type Boolean))
  (|CallParentFunction bParachuteVisibility)
  (if (not bParachuteVisibility)
    (Components|Activation|Deactivate (Variables|Default|GetP_Glider_Yeti_Snow_Idle))))

(event Parachute|EventOnParachuteTrailUpdated
  (params (MovementLengthSquared :type Float%20%28single-precision%29) (ForwardDot :type Float%20%28single-precision%29) (RightDot :type Float%20%28single-precision%29) (RotationalMovementAcceleration :type Float%20%28single-precision%29))
  (|CallParentFunction MovementLengthSquared ForwardDot RightDot RotationalMovementAcceleration)
  (if (not (and (>= ForwardDot 0.2) (CallFunction|IsGliderFullyDeployedAccordingtoAnim)))
    (Components|Activation|Deactivate (Variables|Default|GetP_Glider_Yeti_Snow)))
  (Utilities|FlowControl|DoOnce false
    (:Completed
      (Components|Activation|Activate (Variables|Default|GetP_Glider_Yeti_Snow) false))))

(fn ConstructionScript
  (|CallParentFunction))
